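# Control Rig for the CyberTentacle female cosmetic body, adding procedural deformation and accessory motion on top of the skeleton.
# A shared DeformRig function runs first, then weighted parent constraints place the bag strap nulls, spherical pose readers nudge them, and Spline IK fits front and back bone chains.
# More pose readers drive sleeve curves and offset hip buttons and a bracelet, while blink and jumping variables animate the hedgehog's eyelids and a front strap bone.
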ControlRigBlueprint /BRCosmetics/Characters/Player/Female/Medium/Bodies/F_MED_CyberTentacle/Meshes/F_MED_CyberTentacle_CtrlRig
graphs [
 RigVMFunctionLibrary {links [], nodes []}
 RigVMFunction_Sequence_1_ContainedGraph {nodes [
   n1 = RigVMFunctionEntryNode()
   RigVMFunction_Sequence = &RigVMFunction_Sequence(ExecuteContext=@n1.ExecuteContext)
   n2 = &RigVMFunction_Sequence(ExecuteContext=@RigVMFunction_Sequence.B)
   RigVMFunctionReturnNode(A=@RigVMFunction_Sequence.A, B=@n2.A, C=@n2.B)
  ]}
 RigVMModel {nodes [
   RigUnit_BeginExecution = &RigUnit_BeginExecution()
   FN_DeformRig_FNC = RigVMFunctionReferenceNode(
    DeformRig_ Curve="(deformRig_l_upperarm_up_alpha_72_#1=False,deformRig_r_upperarm_up_alpha_73_#2=False,deformRig_l_upperarm_fr_alpha_74_#3=False,deformRig_r_upperarm_fr_alpha_75_CB51B8 ...(truncated)"
    DeformRig_Bool="(upperarm_1_#4=True,elbow_3_#5=True,wrist_7_#6=True,groin_8_#7=True,glute_17_#8=True,knee_9_40C408324CA261 ...(truncated)"
    DeformRig_Debug=(upperarm_debug_18_#4=False,wrist_debug_20_#6=False,lat_debug_26_#9=False,groinGlute_debug_27_#7=False)
    DeformRig_Value="(upperarm_up_18_#10=(Rotation=(X=0,Y=0,Z=0,W=1),Translation=(X=0,Y=0,Z=8),Scale3D=(X=0.7,Y=1,Z=1)),upperarm_fwd_fr_19_#11=(Rotation=(X=0,Y=0,Z=0,W=1),Translation=(X=0,Y=10,Z=0),Scale3D=(X=0.5,Y=1,Z= ...(truncated)"
    ExecuteContext=@RigUnit_BeginExecution.ExecutePin
   )
   > RigVMFunction_Sequence = RigVMAggregateNode()
   &RigUnit_ParentConstraint(
    AdvancedSettings=(RotationOrderForFilter=XZY)
    Child=(Type=Bone,Name="bag_root")
    Parents=((Item=(Type=Bone,Name="upperarm_twist_01_l"),Weight=0.0075),(Item=(Type=Bone,Name="deform_upperarm_up_l"),Weight=0.0084),(Item=(Type=Bone,Name="spine_05"),Weight=0.2275),(Item=(Type=Bone,Name="clavicle_l"),Weight=0.7566))
    Weight=1
    ExecutePin=@RigVMFunction_Sequence.A
   )
   > &RigUnit_ParentConstraint(
    AdvancedSettings=(RotationOrderForFilter=XZY)
    Child=(Type=Null,Name="bag_back_01_ctrl_null")
    Parents="((Item=(Type=Bone,Name=\"deform_upperarm_bk_l\"),Weight=0.0195),(Item=(Type=Bone,Name=\"upperarm_twist_01_l\"),Weight=0.0302),(Item=(Type=Bone,Name=\"deform_upperarm_up_l\"),Weight=0.0325),(Item=(Type=Bone,Name=\"spine_05\"),Weight=0.1997),(Item=(Type=Bone,Name=\"c ...(truncated)"
    Weight=1
   )
   > &RigUnit_ParentConstraint(
    AdvancedSettings=(RotationOrderForFilter=XZY)
    Child=(Type=Null,Name="bag_back_02_ctrl_null")
    Parents="((Item=(Type=Bone,Name=\"upperarm_twist_01_l\"),Weight=0.0052),(Item=(Type=Bone,Name=\"spine_04\"),Weight=0.0336),(Item=(Type=Bone,Name=\"clavicle_r\"),Weight=0.0385),(Item=(Type=Bone,Name=\"clavicle_l\"),Weight=0.4587),(Item=(Type=Bone,Name=\"spine_05\"),Weight=0.4 ...(truncated)"
    Weight=1
   )
   > &RigUnit_ParentConstraint(
    AdvancedSettings=(RotationOrderForFilter=XZY)
    Child=(Type=Null,Name="bag_back_03_ctrl_null")
    Parents="((Item=(Type=Bone,Name=\"upperarm_twist_01_l\"),Weight=0.0048),(Item=(Type=Bone,Name=\"spine_02\"),Weight=0.0148),(Item=(Type=Bone,Name=\"spine_03\"),Weight=0.0173),(Item=(Type=Bone,Name=\"clavicle_r\"),Weight=0.0681),(Item=(Type=Bone,Name=\"spine_04\"),Weight=0.214 ...(truncated)"
    Weight=1
   )
   > &RigUnit_ParentConstraint(
    AdvancedSettings=(RotationOrderForFilter=XZY)
    Child=(Type=Null,Name="bag_back_04_ctrl_null")
    Parents="((Item=(Type=Bone,Name=\"crop_top_3\"),Weight=0.0001),(Item=(Type=Bone,Name=\"deform_upperarm_bk_l\"),Weight=0.0001),(Item=(Type=Bone,Name=\"crop_top_6\"),Weight=0.0007),(Item=(Type=Bone,Name=\"upperarm_twist_01_l\"),Weight=0.0013),(Item=(Type=Bone,Name=\"pelvis\"), ...(truncated)"
    Weight=1
   )
   > &RigUnit_ParentConstraint(
    AdvancedSettings=(RotationOrderForFilter=XZY)
    Child=(Type=Null,Name="bag_back_05_ctrl_null")
    Parents="((Item=(Type=Bone,Name=\"crop_top_6\"),Weight=0.007),(Item=(Type=Bone,Name=\"crop_top_5\"),Weight=0.008),(Item=(Type=Bone,Name=\"upperarm_twist_01_r\"),Weight=0.012),(Item=(Type=Bone,Name=\"pelvis\"),Weight=0.012),(Item=(Type=Bone,Name=\"clavicle_l\"),Weight=0.0255) ...(truncated)"
    Weight=1
   )
   > &RigUnit_ParentConstraint(
    AdvancedSettings=(RotationOrderForFilter=XZY)
    Child=(Type=Null,Name="bag_back_06_ctrl_null")
    Parents="((Item=(Type=Bone,Name=\"spine_05\"),Weight=0.0052),(Item=(Type=Bone,Name=\"clavicle_r\"),Weight=0.0056),(Item=(Type=Bone,Name=\"spine_01\"),Weight=0.0953),(Item=(Type=Bone,Name=\"spine_03\"),Weight=0.112),(Item=(Type=Bone,Name=\"pelvis\"),Weight=0.137),(Item=(Type= ...(truncated)"
    Weight=1
   )
   > &RigUnit_ParentConstraint(
    AdvancedSettings=(RotationOrderForFilter=XZY)
    Child=(Type=Null,Name="bag_front_01_ctrl_null")
    Parents="((Item=(Type=Bone,Name=\"deform_upperarm_fr_l\"),Weight=0.0042),(Item=(Type=Bone,Name=\"upperarm_twist_01_l\"),Weight=0.0071),(Item=(Type=Bone,Name=\"spine_04\"),Weight=0.0501),(Item=(Type=Bone,Name=\"spine_05\"),Weight=0.3756),(Item=(Type=Bone,Name=\"clavicle_l\"), ...(truncated)"
    Weight=1
   )
   > &RigUnit_ParentConstraint(
    AdvancedSettings=(RotationOrderForFilter=XZY)
    Child=(Type=Null,Name="bag_front_02_ctrl_null")
    Parents="((Item=(Type=Bone,Name=\"upperarm_twist_01_l\"),Weight=0.0029),(Item=(Type=Bone,Name=\"clavicle_l\"),Weight=0.0785),(Item=(Type=Bone,Name=\"deform_pec_l\"),Weight=0.0918),(Item=(Type=Bone,Name=\"spine_04\"),Weight=0.23),(Item=(Type=Bone,Name=\"spine_05\"),Weight=0.5 ...(truncated)"
    Weight=1
   )
   > &RigUnit_ParentConstraint(
    AdvancedSettings=(RotationOrderForFilter=XZY)
    Child=(Type=Null,Name="bag_front_03_ctrl_null")
    Parents=((Item=(Type=Bone,Name="clavicle_l"),Weight=0.0049),(Item=(Type=Bone,Name="deform_pec_l"),Weight=0.0556),(Item=(Type=Bone,Name="spine_05"),Weight=0.3849),(Item=(Type=Bone,Name="spine_04"),Weight=0.5545))
    Weight=1
   )
   > &RigUnit_ParentConstraint(
    AdvancedSettings=(RotationOrderForFilter=XZY)
    Child=(Type=Null,Name="bag_front_04_ctrl_null")
    Parents="((Item=(Type=Bone,Name=\"pelvis\"),Weight=0.0001),(Item=(Type=Bone,Name=\"crop_top_8\"),Weight=0.0002),(Item=(Type=Bone,Name=\"spine_01\"),Weight=0.0006),(Item=(Type=Bone,Name=\"deform_pec_l\"),Weight=0.0025),(Item=(Type=Bone,Name=\"crop_top_1\"),Weight=0.0125),(Ite ...(truncated)"
    Weight=1
   )
   > &RigUnit_ParentConstraint(
    AdvancedSettings=(RotationOrderForFilter=XZY)
    Child=(Type=Null,Name="bag_front_05_ctrl_null")
    Parents="((Item=(Type=Bone,Name=\"crop_top_1\"),Weight=0.0001),(Item=(Type=Bone,Name=\"spine_05\"),Weight=0.0005),(Item=(Type=Bone,Name=\"pelvis\"),Weight=0.0017),(Item=(Type=Bone,Name=\"crop_top_8\"),Weight=0.0048),(Item=(Type=Bone,Name=\"spine_01\"),Weight=0.0133),(Item=(T ...(truncated)"
    Weight=1
   )
   > &RigUnit_ParentConstraint(
    AdvancedSettings=(RotationOrderForFilter=XZY)
    Child=(Type=Null,Name="bag_front_06_ctrl_null")
    Parents=((Item=(Type=Bone,Name="spine_01"),Weight=0.0436),(Item=(Type=Bone,Name="spine_03"),Weight=0.0978),(Item=(Type=Bone,Name="pelvis"),Weight=0.163),(Item=(Type=Bone,Name="spine_04"),Weight=0.2547),(Item=(Type=Bone,Name="spine_02"),Weight=0.4409))
    Weight=1
   )
   > n1 = &RigUnit_SphericalPoseReader(
    ActiveRegionScaleFactors=(PositiveWidth=2.5,NegativeWidth=3,PositiveHeight=3,NegativeHeight=3)
    ActiveRegionSize=0.1
    Debug=(bDrawDebug=false,DebugScale=25,DebugSegments=20,DebugThickness=0.25)
    DriverAxis=(X=1,Y=0,Z=0)
    DriverItem=(Type=Bone,Name="upperarm_l")
    FalloffRegionScaleFactors=(PositiveWidth=1,NegativeWidth=1,PositiveHeight=1,NegativeHeight=1)
    FalloffSize=0.5
    OptionalParentItem=(Type=Bone,Name="clavicle_l")
    RotationOffset=(X=0,Y=0,Z=0)
   )
   > &RigUnit_ModifyTransforms(
    ItemToModify="((Item=(Type=Bone,Name=\"bag_root\"),Transform=(Rotation=(X=0,Y=0,Z=0,W=1),Translation=(X=0,Y=0,Z=0),Scale3D=(X=1,Y=1,Z=1))),(Item=(Type=Null,Name=\"bag_front_02_ctrl_null\"),Transform=(Rotation=(X=0,Y=0,Z=0,W=1),Translation=(X=0,Y=-2,Z=0),Scale3D=(X=1,Y=1,Z=1 ...(truncated)"
    WeightMinimum=0
    ItemToModify.2.Transform.Rotation=&RigVMFunction_MathQuaternionFromEuler(Euler=(X=0,Y=10,Z=0)).Result
    Weight=@n1.OutputParam
   )
   > n2 = &RigUnit_SphericalPoseReader(
    ActiveRegionScaleFactors=(PositiveWidth=1,NegativeWidth=0.1,PositiveHeight=1,NegativeHeight=0.5)
    ActiveRegionSize=0.3
    Debug=(bDrawDebug=true,bDraw2D=false,bDrawLocalAxes=false,DebugScale=30,DebugSegments=30,DebugThickness=0.2)
    DriverAxis=(X=1,Y=0,Z=0)
    DriverItem=(Type=Bone,Name="clavicle_l")
    FalloffRegionScaleFactors=(PositiveWidth=1,NegativeWidth=0.5,PositiveHeight=1,NegativeHeight=0.5)
    FalloffSize=0.5
    OptionalParentItem=(Type=Bone,Name="None")
    RotationOffset=(X=0,Y=40,Z=180)
   )
   AlphaInterp = &RigVMFunction_AlphaInterp(InterpSpeedDecreasing=10, InterpSpeedIncreasing=10, Scale=1, bClampResult=False, Value=@n2.OutputParam)
   > &RigUnit_ModifyTransforms(
    ItemToModify="((Item=(Type=Bone,Name=\"bag_root\"),Transform=(Rotation=(X=0,Y=0,Z=0,W=1),Translation=(X=3,Y=0,Z=1.8),Scale3D=(X=1,Y=1,Z=1))),(Item=(Type=Null,Name=\"bag_back_02_ctrl_null\"),Transform=(Rotation=(X=0,Y=0,Z=0,W=1),Translation=(X=0,Y=1,Z=0),Scale3D=(X=1,Y=1,Z=1 ...(truncated)"
    WeightMinimum=0
    Weight=@AlphaInterp.Result
    ItemToModify.0.Transform.Rotation=&RigVMFunction_MathQuaternionFromEuler(Euler=(X=0,Y=-25,Z=0)).Result
    ItemToModify.1.Transform.Rotation=&RigVMFunction_MathQuaternionFromEuler(Euler=(X=0,Y=0,Z=0)).Result
   )
   > SplineIK_2 = RigVMFunctionReferenceNode(
    Primary Axis=(X=1,Y=0,Z=0)
    Samples Per Segment=4
    Secondary Spline Direction=(X=0,Y=-1,Z=0)
    Spline Mode=BSpline
    Stretch=false
    Up Axis=(X=0,Y=-1,Z=0)
    debug=false
    Controls=RigVMVariableNode() {variable=bag_rear_ctrls_RBAN}.Value
    Bones=RigVMVariableNode() {variable=bag_rear_joints_RBAN}.Value
   )
   > SplineIK_1_1 = RigVMFunctionReferenceNode(
    Primary Axis=(X=1,Y=0,Z=0)
    Samples Per Segment=4
    Secondary Spline Direction=(X=0,Y=1,Z=0)
    Spline Mode=BSpline
    Stretch=false
    Up Axis=(X=0,Y=1,Z=0)
    debug=false
    Controls=RigVMVariableNode() {variable=bag_front_ctrls_RBAN}.Value
    Bones=RigVMVariableNode() {variable=bag_front_joints_RBAN}.Value
   )
   n3 = &RigUnit_SphericalPoseReader(
    ActiveRegionScaleFactors=(PositiveWidth=0.4,NegativeWidth=0.2,PositiveHeight=0.4,NegativeHeight=0.4)
    ActiveRegionSize=0.2
    Debug=(bDrawDebug=False,bDraw2D=false,bDrawLocalAxes=false,DebugScale=30,DebugSegments=30,DebugThickness=0.2)
    DriverAxis=(X=-1,Y=0,Z=0)
    DriverItem=(Type=Bone,Name="upperarm_r")
    FalloffRegionScaleFactors=(PositiveWidth=0.5,NegativeWidth=0.5,PositiveHeight=0.5,NegativeHeight=0.5)
    FalloffSize=0.5
    OptionalParentItem=(Type=Bone,Name="None")
    RotationOffset=(X=-30,Y=-310,Z=10)
    ExecutePin=@RigVMFunction_Sequence.B
   )
   > /Script/ControlRig.RigUnit_SetCurveValue(Curve=sleeve_R, Value=&RigVMFunction_MathDoubleSub(A=1, B=@n3.OutputParam).Result) {is=struct RigVMUnitNode}
   > n4 = &RigUnit_SphericalPoseReader(
    ActiveRegionScaleFactors=(PositiveWidth=1,NegativeWidth=0.2,PositiveHeight=1,NegativeHeight=0.4)
    ActiveRegionSize=0.4
    Debug=(bDrawDebug=False,bDraw2D=false,bDrawLocalAxes=false,DebugScale=30,DebugSegments=30,DebugThickness=0.2)
    DriverAxis=(X=1,Y=0,Z=0)
    DriverItem=(Type=Bone,Name="upperarm_l")
    FalloffRegionScaleFactors=(PositiveWidth=0.5,NegativeWidth=0.1,PositiveHeight=0.5,NegativeHeight=0.5)
    FalloffSize=0.7
    OptionalParentItem=(Type=Bone,Name="None")
    RotationOffset=(X=0,Y=-150,Z=0)
   )
   > n5 = &RigUnit_SetCurveValue(Curve=sleeve_L, Value=&RigVMFunction_MathDoubleSub(A=1, B=@n4.OutputParam).Result)
   n6 = &RigVMFunction_Sequence(ExecuteContext=@n5.ExecutePin)
   &RigUnit_ParentConstraint(
    AdvancedSettings=(RotationOrderForFilter=XZY)
    Child=(Type=Bone,Name="button_l_2")
    Parents="((Item=(Type=Bone,Name=\"spine_02\"),Weight=0.0006),(Item=(Type=Bone,Name=\"thigh_twist_01_r\"),Weight=0.0033),(Item=(Type=Bone,Name=\"thigh_l\"),Weight=0.0067),(Item=(Type=Bone,Name=\"thigh_twist_01_l\"),Weight=0.2453),(Item=(Type=Bone,Name=\"deform_groin_l\"),Weig ...(truncated)"
    Weight=1
    ExecutePin=@n6.A
   )
   > &RigUnit_ParentConstraint(
    AdvancedSettings=(RotationOrderForFilter=XZY)
    Child=(Type=Bone,Name="button_l_1")
    Parents="((Item=(Type=Bone,Name=\"thigh_twist_01_r\"),Weight=0.0002),(Item=(Type=Bone,Name=\"spine_02\"),Weight=0.0263),(Item=(Type=Bone,Name=\"thigh_twist_01_l\"),Weight=0.0604),(Item=(Type=Bone,Name=\"deform_groin_l\"),Weight=0.1406),(Item=(Type=Bone,Name=\"pelvis\"),Weigh ...(truncated)"
    Weight=1
   )
   > n7 = &RigUnit_SphericalPoseReader(
    ActiveRegionScaleFactors=(PositiveWidth=1,NegativeWidth=0.1,PositiveHeight=1,NegativeHeight=1)
    ActiveRegionSize=0.3
    Debug=(bDrawDebug=false,bDraw2D=false,bDrawLocalAxes=false,DebugScale=30,DebugSegments=30,DebugThickness=0.2)
    DriverAxis=(X=-1,Y=0,Z=0)
    DriverItem=(Type=Bone,Name="thigh_l")
    FalloffRegionScaleFactors=(PositiveWidth=1,NegativeWidth=0.6,PositiveHeight=1,NegativeHeight=1)
    FalloffSize=0.5
    OptionalParentItem=(Type=Bone,Name="pelvis")
    RotationOffset=(X=90,Y=0,Z=0)
   )
   > &RigUnit_ModifyTransforms(
    ItemToModify=((Item=(Type=Bone,Name="button_l_1"),Transform=(Rotation=(X=0,Y=0,Z=0,W=1),Translation=(X=1,Y=0,Z=2),Scale3D=(X=1,Y=1,Z=1))))
    WeightMinimum=0
    Weight=@n7.OutputParam
   )
   RigVMFunction_MathQuaternionFromEuler = &RigVMFunction_MathQuaternionFromEuler(Euler=(X=0,Y=-75,Z=10))
   > &RigUnit_ModifyTransforms(
    ItemToModify=((Item=(Type=Bone,Name="button_l_2"),Transform=(Rotation=(X=0,Y=0,Z=0,W=1),Translation=(X=2,Y=-1,Z=0),Scale3D=(X=1,Y=1,Z=1))))
    WeightMinimum=0
    ItemToModify.0.Transform.Rotation=@RigVMFunction_MathQuaternionFromEuler.Result
    Weight=@n7.OutputParam
   )
   &RigUnit_ParentConstraint(
    AdvancedSettings=(RotationOrderForFilter=XZY)
    Child=(Type=Bone,Name="button_r_2")
    Parents="((Item=(Type=Bone,Name=\"deform_groin_l\"),Weight=0.0001),(Item=(Type=Bone,Name=\"spine_02\"),Weight=0.002),(Item=(Type=Bone,Name=\"thigh_twist_01_l\"),Weight=0.0041),(Item=(Type=Bone,Name=\"thigh_r\"),Weight=0.011),(Item=(Type=Bone,Name=\"thigh_twist_01_r\"),Weight ...(truncated)"
    Weight=1
    ExecutePin=@n6.B
   )
   > &RigUnit_ParentConstraint(
    AdvancedSettings=(RotationOrderForFilter=XZY)
    Child=(Type=Bone,Name="button_r_1")
    Parents="((Item=(Type=Bone,Name=\"spine_04\"),Weight=0.0001),(Item=(Type=Bone,Name=\"thigh_twist_01_l\"),Weight=0.0006),(Item=(Type=Bone,Name=\"spine_02\"),Weight=0.0301),(Item=(Type=Bone,Name=\"thigh_twist_01_r\"),Weight=0.0641),(Item=(Type=Bone,Name=\"deform_groin_r\"),Wei ...(truncated)"
    Weight=1
   )
   > n8 = &RigUnit_SphericalPoseReader(
    ActiveRegionScaleFactors=(PositiveWidth=1,NegativeWidth=0,PositiveHeight=1,NegativeHeight=1)
    ActiveRegionSize=0.3
    Debug=(bDrawDebug=True,bDraw2D=false,bDrawLocalAxes=false,DebugScale=30,DebugSegments=30,DebugThickness=0.2)
    DriverAxis=(X=1,Y=0,Z=0)
    DriverItem=(Type=Bone,Name="thigh_r")
    FalloffRegionScaleFactors=(PositiveWidth=1,NegativeWidth=0.5,PositiveHeight=1,NegativeHeight=1)
    FalloffSize=0.5
    FlipWidthScaling=false
    OptionalParentItem=(Type=Bone,Name="pelvis")
    RotationOffset=(X=-90,Y=180,Z=30)
   )
   > &RigUnit_ModifyTransforms(
    ItemToModify=((Item=(Type=Bone,Name="button_r_1"),Transform=(Rotation=(X=0,Y=0,Z=0,W=1),Translation=(X=0.5,Y=0,Z=3),Scale3D=(X=1,Y=1,Z=1))))
    WeightMinimum=0
    Weight=@n8.OutputParam
   )
   > &RigUnit_ModifyTransforms(
    ItemToModify=((Item=(Type=Bone,Name="button_r_2"),Transform=(Rotation=(X=0,Y=0,Z=0,W=1),Translation=(X=0.5,Y=0,Z=1),Scale3D=(X=1,Y=1,Z=1))))
    WeightMinimum=0
    ItemToModify.0.Transform.Rotation=&RigVMFunction_MathQuaternionFromEuler(Euler=(X=0,Y=0,Z=0)).Result
    Weight=@n8.OutputParam
   )
   &RigUnit_ModifyTransforms(
    ItemToModify=((Item=(Type=Bone,Name="eye_1"),Transform=(Rotation=(X=0,Y=0,Z=0,W=1),Translation=(X=0.05,Y=0,Z=0),Scale3D=(X=1,Y=1,Z=1))))
    Weight=1
    WeightMinimum=0
    ExecutePin=@RigVMFunction_Sequence.C
   )
   > &RigUnit_ModifyTransforms(
    ItemToModify=((Item=(Type=Bone,Name="eye_2"),Transform=(Rotation=(X=0,Y=0,Z=0,W=1),Translation=(X=0.05,Y=0,Z=0),Scale3D=(X=1,Y=1,Z=1))))
    Weight=1
    WeightMinimum=0
   )
   n9 = RigVMVariableNode() {variable=blink}
   > &RigUnit_ModifyTransforms(
    ItemToModify="((Item=(Type=Bone,Name=\"eyelid_upper_1\"),Transform=(Rotation=(X=0.0431,Y=0.1477,Z=-0.0064,W=0.9881),Translation=(X=0.1,Y=0,Z=0),Scale3D=(X=1,Y=1,Z=1))),(Item=(Type=Bone,Name=\"eyelid_lower_1\"),Transform=(Rotation=(X=-0.1294,Y=-0.1294,Z=-0.017,W=0.983),Trans ...(truncated)"
    WeightMinimum=0
    Weight=@n9.Value
   )
   > &RigUnit_ModifyTransforms(
    ItemToModify="((Item=(Type=Bone,Name=\"eyelid_upper_2\"),Transform=(Rotation=(X=-0.0431,Y=0.1477,Z=0.0064,W=0.9881),Translation=(X=0.1,Y=0,Z=0),Scale3D=(X=1,Y=1,Z=1))),(Item=(Type=Bone,Name=\"eyelid_lower_2\"),Transform=(Rotation=(X=0,Y=-0.1305,Z=0,W=0.9914),Translation=(X= ...(truncated)"
    WeightMinimum=0
    Weight=@n9.Value
   )
   n10 = &RigVMFunction_AlphaInterp(InterpSpeedDecreasing=1, InterpSpeedIncreasing=10, Scale=1, bClampResult=False, Value=RigVMVariableNode() {variable=junping}.Value)
   > &RigUnit_ModifyTransforms(
    ItemToModify=((Item=(Type=Bone,Name="bag_front_04"),Transform=(Rotation=(X=0,Y=0,Z=0,W=1),Translation=(X=0,Y=0,Z=0),Scale3D=(X=1,Y=1,Z=1))))
    WeightMinimum=0
    Weight=@n10.Result
    ItemToModify.0.Transform.Rotation=&RigVMFunction_MathQuaternionFromEuler(Euler=(X=0,Y=0,Z=-40)).Result
   )
   > n11 = &RigUnit_SphericalPoseReader(
    ActiveRegionScaleFactors=(PositiveWidth=0.4,NegativeWidth=0.2,PositiveHeight=0.4,NegativeHeight=0.4)
    ActiveRegionSize=0.2
    Debug=(bDrawDebug=True,bDraw2D=false,bDrawLocalAxes=false,DebugScale=30,DebugSegments=30,DebugThickness=0.2)
    DriverAxis=(X=1,Y=0,Z=0)
    DriverItem=(Type=Bone,Name="hand_l")
    FalloffRegionScaleFactors=(PositiveWidth=0.5,NegativeWidth=0.5,PositiveHeight=0.5,NegativeHeight=0.5)
    FalloffSize=0.5
    OptionalParentItem=(Type=Bone,Name="None")
    RotationOffset=(X=60,Y=-90,Z=0)
   )
   > &RigUnit_ModifyTransforms(
    ItemToModify=((Item=(Type=Bone,Name="BASE_dyn_bracelet"),Transform=(Rotation=(X=0,Y=0,Z=0,W=1),Translation=(X=-1,Y=0,Z=0),Scale3D=(X=1,Y=1,Z=1))))
    WeightMinimum=0
    Weight=@n11.OutputParam
   )
  ]}
]
hierarchy [
 root {loc=<0 0 0>, rot=<0 0 0 1>}
 pelvis {parent=root, loc=<0 -0.5301 98.0571>, rot=<0.0127 0.707 -0.0127 -0.707>}
 spine_01 {parent=pelvis, loc=<6.1208 -0.9328 0>, rot=<0 0 0.034 -0.9994>}
 spine_02 {parent=spine_01, loc=<6.5997 -0.2095 0>, rot=<0 0 0.0238 -0.9997>}
 spine_03 {parent=spine_02, loc=<6.4914 -0.1237 0>, rot=<0 0 0.0149 -0.9999>}
 spine_04 {parent=spine_03, loc=<6.9598 0.0794 0>, rot=<0 0 -0.0132 -0.9999>}
 spine_05 {parent=spine_04, loc=<6.9987 0 0>, rot=<0 0 -0.0015 -1>}
 neck_01 {parent=spine_05, loc=<11.3138 0.1529 0>, rot=<0 0 -0.2003 -0.9797>}
 neck_02 {parent=neck_01, loc=<4.5627 0 0>, rot=<0 0 0.0504 -0.9987>}
 head {parent=neck_02, loc=<5.3914 0 0>, rot=<0 0 0.115 -0.9934>}
 clavicle_l {parent=spine_05, loc=<5.4831 -0.1053 -3.2946>, rot=<0.0465 -0.7289 -0.009 -0.6829>}
 upperarm_l {parent=clavicle_l, loc=<9.8658 -1.2136 -0.4105>, rot=<0.0164 -0.3245 -0.0269 -0.9454>}
 lowerarm_l {parent=upperarm_l, loc=<26.465 -0.8345 0.259>, rot=<0.0036 0.0161 -0.2298 -0.9731>}
 hand_l {parent=lowerarm_l, loc=<22.7355 -0.3203 -0.6105>, rot=<-0.6784 -0.0223 0.0735 -0.7307>}
 thumb_01_l {parent=hand_l, loc=<1.3118 -0.3191 -2.459>, rot=<0.6546 -0.4042 0.0664 -0.6354>}
 thumb_02_l {parent=thumb_01_l, loc=<5.9336 -0.0448 0.0401>, rot=<-0.0037 0.0213 0.298 -0.9543>}
 thumb_03_l {parent=thumb_02_l, loc=<2.2223 0.3682 -0.0907>, rot=<0.0118 -0.013 0.0066 -0.9998>}
 index_metacarpal_l {parent=hand_l, loc=<1.6535 0.0183 -2.96>, rot=<0.0158 -0.0124 -0.0173 -0.9996>}
 index_01_l {parent=index_metacarpal_l, loc=<8.3008 0.0357 -0.0524>, rot=<-0.0044 0.0485 0.0924 -0.9945>}
 index_02_l {parent=index_01_l, loc=<3.3587 -0.3004 0.0452>, rot=<-0.002 0.0027 0.3225 -0.9466>}
 index_03_l {parent=index_02_l, loc=<2.5578 -0.123 -0.0252>, rot=<-0.002 0.0067 0.2558 -0.9667>}
 middle_metacarpal_l {parent=hand_l, loc=<2.1856 -0.1725 -0.5984>, rot=<0.006 0.0168 -0.0329 -0.9993>}
 middle_01_l {parent=middle_metacarpal_l, loc=<7.3275 -0.2447 0.0129>, rot=<-0.0208 0.0042 0.0858 -0.9961>}
 middle_02_l {parent=middle_01_l, loc=<4.0123 0.1401 0.1175>, rot=<-0.0218 0.0258 0.3512 -0.9357>}
 middle_03_l {parent=middle_02_l, loc=<2.6077 0.0534 0.1373>, rot=<0.0082 0.0188 0.1708 -0.9851>}
 ring_metacarpal_l {parent=hand_l, loc=<2.2074 -0.3625 1.3638>, rot=<-0.0284 0.0722 -0.0458 -0.9959>}
 ring_01_l {parent=ring_metacarpal_l, loc=<7.0295 -0.6197 -0.1446>, rot=<-0.0489 -0.0252 0.1892 -0.9804>}
 ring_02_l {parent=ring_01_l, loc=<3.6174 0.4647 -0.1027>, rot=<0 0 0.2806 -0.9598>}
 ring_03_l {parent=ring_02_l, loc=<2.4152 0.114 0.0199>, rot=<0 0 0.1189 -0.9929>}
 pinky_metacarpal_l {parent=hand_l, loc=<1.7616 -0.9753 3.3814>, rot=<-0.0397 0.0814 -0.069 -0.9935>}
 pinky_01_l {parent=pinky_metacarpal_l, loc=<7.1709 -0.4831 -0.1094>, rot=<-0.0654 -0.0035 0.2242 -0.9723>}
 pinky_02_l {parent=pinky_01_l, loc=<2.8274 0.0893 -0.1043>, rot=<0.0037 0.0011 0.2854 -0.9584>}
 pinky_03_l {parent=pinky_02_l, loc=<2.2046 0.0307 -0.0078>, rot=<-0.0002 0.0098 0.1445 -0.9895>}
 deform_wrist_dn_base_l {parent=hand_l, loc=<0 -1.5326 0>, rot=<0 0 0 -1>}
 deform_wrist_dn_l {parent=deform_wrist_dn_base_l, loc=<0 0 0>, rot=<0 0 0 1>}
 deform_wrist_up_base_l {parent=hand_l, loc=<0 2.3009 0>, rot=<0 0 0 -1>}
 deform_wrist_up_l {parent=deform_wrist_up_base_l, loc=<0 0 0>, rot=<0 0 0 1>}
 lowerarm_twist_02_l {parent=lowerarm_l, loc=<10.6707 0 0>, rot=<0 0 0 1>}
 lowerarm_twist_01_l {parent=lowerarm_l, loc=<18.316 0 0>, rot=<0 0 0 1>}
 BASE_dyn_bracelet {parent=lowerarm_twist_01_l, loc=<3.1379 -0.1486 -0.9945>, rot=<0 0 0 1>}
 dyn_bracelet {parent=BASE_dyn_bracelet, loc=<0 0 0>, rot=<0 0 0 1>}
 deform_elbow_in_base_l {parent=lowerarm_l, loc=<0.1834 3.6062 -0.2572>, rot=<0 0 0 -1>}
 deform_elbow_in_l {parent=deform_elbow_in_base_l, loc=<0 0 0>, rot=<0 0 0 1>}
 deform_elbow_out_base_l {parent=lowerarm_l, loc=<0 -4 0>, rot=<0 0 0 1>}
 deform_elbow_out_l {parent=deform_elbow_out_base_l, loc=<0 0 0>, rot=<0 0 0 1>}
 upperarm_twist_01_l {parent=upperarm_l, loc=<7.2135 -0.1134 0>, rot=<0 0 0 1>}
 deform_upperarm_bk_base_l {parent=upperarm_twist_01_l, loc=<-6.4053 -4.3746 0.314>, rot=<0 0 0 -1>}
 deform_upperarm_bk_l {parent=deform_upperarm_bk_base_l, loc=<0 0 0>, rot=<0 0 0 1>}
 deform_upperarm_fr_base_l {parent=upperarm_twist_01_l, loc=<-5.6312 5.092 0.3354>, rot=<0 0 0 -1>}
 deform_upperarm_fr_l {parent=deform_upperarm_fr_base_l, loc=<0 0 0>, rot=<0 0 0 1>}
 deform_upperarm_up_base_l {parent=upperarm_twist_01_l, loc=<-8.891 0.4825 3.9517>, rot=<0 0 0 -1>}
 deform_upperarm_up_l {parent=deform_upperarm_up_base_l, loc=<0 0 0>, rot=<0 0 0 1>}
 sleeve_4_l {parent=upperarm_twist_01_l, loc=<-0.1275 5.064 -1.3625>, rot=<1 0 0 0>}
 sleeve_2_l {parent=upperarm_twist_01_l, loc=<-0.427 -5.3259 -2.2863>, rot=<0 0 0 1>}
 sleeve_1_l {parent=upperarm_twist_01_l, loc=<-1.7811 0.1134 4.5249>, rot=<0.7071 0 0 -0.7071>}
 sleeve_3_l {parent=upperarm_twist_01_l, loc=<1.2115 0.1134 -6.5822>, rot=<-0.766 0 0 -0.6428>}
 upperarm_twist_02_l {parent=upperarm_l, loc=<19.7998 -0.1134 0>, rot=<0 0 0 1>}
 clavicle_r {parent=spine_05, loc=<5.4831 -0.1053 3.2946>, rot=<0.7289 0.0465 -0.6829 0.009>}
 upperarm_r {parent=clavicle_r, loc=<-9.8658 1.2136 0.4105>, rot=<0.0164 -0.3245 -0.0269 -0.9454>}
 lowerarm_r {parent=upperarm_r, loc=<-26.465 0.8345 -0.259>, rot=<0.0036 0.0161 -0.2298 -0.9731>}
 hand_r {parent=lowerarm_r, loc=<-22.7355 0.3203 0.6106>, rot=<-0.6784 -0.0223 0.0735 -0.7307>}
 thumb_01_r {parent=hand_r, loc=<-1.3118 0.3191 2.459>, rot=<0.6546 -0.4042 0.0664 -0.6354>}
 thumb_02_r {parent=thumb_01_r, loc=<-5.9336 0.0448 -0.0401>, rot=<-0.0037 0.0213 0.298 -0.9543>}
 thumb_03_r {parent=thumb_02_r, loc=<-2.2223 -0.3682 0.0907>, rot=<0.0118 -0.013 0.0066 -0.9998>}
 index_metacarpal_r {parent=hand_r, loc=<-1.6535 -0.0183 2.96>, rot=<0.0158 -0.0124 -0.0173 -0.9996>}
 index_01_r {parent=index_metacarpal_r, loc=<-8.3008 -0.0357 0.0524>, rot=<-0.0044 0.0485 0.0924 -0.9945>}
 index_02_r {parent=index_01_r, loc=<-3.3587 0.3004 -0.0452>, rot=<-0.002 0.0027 0.3225 -0.9466>}
 index_03_r {parent=index_02_r, loc=<-2.5578 0.123 0.0252>, rot=<-0.002 0.0067 0.2558 -0.9667>}
 middle_metacarpal_r {parent=hand_r, loc=<-2.1856 0.1725 0.5984>, rot=<0.006 0.0168 -0.0329 -0.9993>}
 middle_01_r {parent=middle_metacarpal_r, loc=<-7.3275 0.2447 -0.0129>, rot=<-0.0208 0.0042 0.0858 -0.9961>}
 middle_02_r {parent=middle_01_r, loc=<-4.0123 -0.1401 -0.1175>, rot=<-0.0218 0.0258 0.3512 -0.9357>}
 middle_03_r {parent=middle_02_r, loc=<-2.6077 -0.0534 -0.1373>, rot=<0.0082 0.0188 0.1708 -0.9851>}
 ring_metacarpal_r {parent=hand_r, loc=<-2.2074 0.3625 -1.3638>, rot=<-0.0284 0.0722 -0.0458 -0.9959>}
 ring_01_r {parent=ring_metacarpal_r, loc=<-7.0295 0.6197 0.1446>, rot=<-0.0489 -0.0252 0.1892 -0.9804>}
 ring_02_r {parent=ring_01_r, loc=<-3.6174 -0.4647 0.1027>, rot=<0 0 0.2806 -0.9598>}
 ring_03_r {parent=ring_02_r, loc=<-2.4152 -0.114 -0.0199>, rot=<0 0 0.1189 -0.9929>}
 pinky_metacarpal_r {parent=hand_r, loc=<-1.7616 0.9753 -3.3814>, rot=<-0.0397 0.0814 -0.069 -0.9935>}
 pinky_01_r {parent=pinky_metacarpal_r, loc=<-7.1709 0.4831 0.1094>, rot=<-0.0654 -0.0035 0.2242 -0.9723>}
 pinky_02_r {parent=pinky_01_r, loc=<-2.8274 -0.0893 0.1043>, rot=<0.0037 0.0011 0.2854 -0.9584>}
 pinky_03_r {parent=pinky_02_r, loc=<-2.2046 -0.0307 0.0078>, rot=<-0.0002 0.0098 0.1445 -0.9895>}
 deform_wrist_dn_base_r {parent=hand_r, loc=<-0.0001 1.5328 0>, rot=<0 0 0 -1>}
 deform_wrist_dn_r {parent=deform_wrist_dn_base_r, loc=<0 0 0>, rot=<0 0 0 1>}
 deform_wrist_up_base_r {parent=hand_r, loc=<-0.0002 -2.3007 -0.0001>, rot=<0 0 0 -1>}
 deform_wrist_up_r {parent=deform_wrist_up_base_r, loc=<0 0 0>, rot=<0 0 0 1>}
 lowerarm_twist_02_r {parent=lowerarm_r, loc=<-10.6707 0 0>, rot=<0 0 0 -1>}
 lowerarm_twist_01_r {parent=lowerarm_r, loc=<-18.316 0 0>, rot=<0 0 0 -1>}
 deform_elbow_in_base_r {parent=lowerarm_r, loc=<-0.1834 -3.6062 0.2572>, rot=<0 0 0 -1>}
 deform_elbow_in_r {parent=deform_elbow_in_base_r, loc=<0 0 0>, rot=<0 0 0 1>}
 deform_elbow_out_base_r {parent=lowerarm_r, loc=<0 4 -0.0001>, rot=<0 0 0 -1>}
 deform_elbow_out_r {parent=deform_elbow_out_base_r, loc=<0 0 0>, rot=<0 0 0 1>}
 upperarm_twist_01_r {parent=upperarm_r, loc=<-7.2135 0.1134 0>, rot=<0 0 0 -1>}
 deform_upperarm_bk_base_r {parent=upperarm_twist_01_r, loc=<6.4053 4.3746 -0.314>, rot=<0 0 0 -1>}
 deform_upperarm_bk_r {parent=deform_upperarm_bk_base_r, loc=<0 0 0>, rot=<0 0 0 1>}
 deform_upperarm_fr_base_r {parent=upperarm_twist_01_r, loc=<5.6312 -5.092 -0.3354>, rot=<0 0 0 -1>}
 deform_upperarm_fr_r {parent=deform_upperarm_fr_base_r, loc=<0 0 0>, rot=<0 0 0 1>}
 deform_upperarm_up_base_r {parent=upperarm_twist_01_r, loc=<8.891 -0.4825 -3.9517>, rot=<0 0 0 -1>}
 deform_upperarm_up_r {parent=deform_upperarm_up_base_r, loc=<0 0 0>, rot=<0 0 0 1>}
 sleeve_1_r {parent=upperarm_twist_01_r, loc=<1.7808 -0.1134 -4.5246>, rot=<0 -0.7071 -0.7071 0>}
 sleeve_2_r {parent=upperarm_twist_01_r, loc=<0.4269 5.3259 2.2865>, rot=<0 -1 0 0>}
 sleeve_3_r {parent=upperarm_twist_01_r, loc=<-1.2116 -0.1134 6.5823>, rot=<0 -0.766 0.6428 0>}
 sleeve_4_r {parent=upperarm_twist_01_r, loc=<0.1277 -5.064 1.3623>, rot=<0 0 1 0>}
 upperarm_twist_02_r {parent=upperarm_r, loc=<-19.7998 0.1134 0>, rot=<0 0 0 -1>}
 attach_backpack {parent=spine_05, loc=<-0.123 -13.473 0>, rot=<-0.0093 -0.707 0.0093 -0.707>}
 Backpack_BR {parent=attach_backpack, type=Null, loc=<0 0 0>, rot=<0 0 0 1>}
 attach_cape {parent=spine_05, loc=<9.017 -11.714 0>, rot=<0.0616 -0.7044 -0.0616 -0.7044>}
 cape {parent=attach_cape, type=Null, loc=<0 0 0>, rot=<0 0 0 1>}
 bag_root {parent=spine_05, loc=<11.1051 -1.6608 -9.3698>, rot=<0.0284 -0.7065 -0.0284 -0.7065>}
 bag_front_01 {parent=bag_root, loc=<-0.8717 5.0598 -1.8569>, rot=<-0.4739 -0.2497 0.7977 0.2769>}
 bag_front_02 {parent=bag_front_01, loc=<7.8242 0 0>, rot=<-0.0125 0.0259 -0.0631 -0.9976>}
 bag_front_03 {parent=bag_front_02, loc=<7.9872 0 0>, rot=<0.061 0.0171 -0.146 -0.9873>}
 bag_front_04 {parent=bag_front_03, loc=<5.7188 0 0>, rot=<-0.1047 0.0202 -0.1879 -0.9764>}
 bag_front_05 {parent=bag_front_04, loc=<8.1103 0 0>, rot=<-0.0658 0.0061 -0.0917 -0.9936>}
 bag_front_06 {parent=bag_front_05, loc=<8.022 0 0>, rot=<0 0 0 1>}
 bag_back_01 {parent=bag_root, loc=<-1.5508 -3.7462 -1.7913>, rot=<-0.4594 0.1832 0.8598 -0.1272>}
 bag_back_02 {parent=bag_back_01, loc=<5.5424 0 0>, rot=<0 0 0.1808 -0.9835>}
 bag_back_03 {parent=bag_back_02, loc=<6.3957 0 0>, rot=<0.0074 0.0005 0.0666 -0.9978>}
 bag_back_04 {parent=bag_back_03, loc=<5.8846 0 0>, rot=<0.1516 0.0023 0.0149 -0.9883>}
 bag_back_05 {parent=bag_back_04, loc=<6.3129 0 0>, rot=<-0.0263 -0.0011 0.0437 -0.9987>}
 bag_back_06 {parent=bag_back_05, loc=<9.4374 0 0>, rot=<0 0 0 1>}
 deform_pec_aim_base_l {parent=spine_04, loc=<3.2456 9.3309 -1.2237>, rot=<0.0603 -0.4033 0.3318 -0.8507>}
 deform_pec_aim_l {parent=deform_pec_aim_base_l, loc=<0 0 0>, rot=<0 0 0 1>}
 deform_pec_base_l {parent=deform_pec_aim_l, loc=<9.3665 0 0>, rot=<0 0 0 1>}
 deform_pec_l {parent=deform_pec_base_l, loc=<0 0 0>, rot=<0 0 0 1>}
 deform_pec_aim_base_r {parent=spine_04, loc=<3.2452 9.3309 1.2237>, rot=<-0.4033 -0.0603 0.8507 0.3318>}
 deform_pec_aim_r {parent=deform_pec_aim_base_r, loc=<0 0 0>, rot=<0 0 0 1>}
 deform_pec_base_r {parent=deform_pec_aim_r, loc=<-9.3667 -0.0003 -0.0004>, rot=<0 0 0 1>}
 deform_pec_r {parent=deform_pec_base_r, loc=<0 0 0>, rot=<0 0 0 1>}
 dyn_bag_sim_chain_01_MSTR {parent=spine_04, loc=<-3.0683 1.178 3.6634>, rot=<0.2209 -0.6271 0.7355 0.1304>}
 dyn_bag_sim_chain_01 {parent=dyn_bag_sim_chain_01_MSTR, loc=<0 0 0>, rot=<0 0 0 1>}
 dyn_bag_sim_chain_02 {parent=dyn_bag_sim_chain_01, loc=<11.1795 0 0>, rot=<0 0 0 1>}
 dyn_bag_sim_chain_03 {parent=dyn_bag_sim_chain_02, loc=<0.3745 -0.2247 -0.0633>, rot=<0.1018 0.0185 0.2338 -0.9668>}
 hedgehog {parent=dyn_bag_sim_chain_03, loc=<15.1963 5.04 9.9258>, rot=<0.082 0.4136 -0.3511 -0.836>}
 eye_1 {parent=hedgehog, loc=<6.0428 -0.9119 0.398>, rot=<0.7705 -0.4222 0.3168 -0.3574>}
 eye_2 {parent=hedgehog, loc=<6.0428 -0.9119 0.398>, rot=<0.884 0.1025 0.4435 -0.1065>}
 eyelid_upper_2 {parent=hedgehog, loc=<6.0428 -0.9119 0.398>, rot=<-0.3065 0.4541 -0.1679 -0.8196>}
 eyelid_upper_1 {parent=hedgehog, loc=<6.0428 -0.9119 0.398>, rot=<-0.1832 0.2913 0.4882 -0.802>}
 eyelid_lower_2 {parent=hedgehog, loc=<6.0428 -0.9119 0.398>, rot=<-0.2585 0.2304 -0.2219 -0.9115>}
 eyelid_lower_1 {parent=hedgehog, loc=<6.0428 -0.9119 0.398>, rot=<-0.3301 0.0912 0.4257 -0.8375>}
 zipper {parent=dyn_bag_sim_chain_03, loc=<8.8014 2.572 -10.3533>, rot=<0.6318 0.2623 -0.2672 -0.6787>}
 BASE_crop_top_1 {parent=spine_03, loc=<4.0646 13.0723 -4.9185>, rot=<0.0547 -0.9985 0 0>}
 crop_top_1 {parent=BASE_crop_top_1, loc=<0 0 0>, rot=<0 0 0 1>}
 BASE_crop_top_2 {parent=spine_03, loc=<4.6487 7.7601 -11.1887>, rot=<0.0468 -0.8545 0.5165 -0.0283>}
 crop_top_2 {parent=BASE_crop_top_2, loc=<0 0 0>, rot=<0 0 0 1>}
 BASE_crop_top_3 {parent=spine_03, loc=<5.543 -0.3724 -12.749>, rot=<0.0387 -0.706 0.706 -0.0387>}
 crop_top_3 {parent=BASE_crop_top_3, loc=<0 0 0>, rot=<0 0 0 1>}
 BASE_crop_top_4 {parent=spine_03, loc=<6.1231 -5.6474 -10.3082>, rot=<0.0141 -0.2567 0.9649 -0.0529>}
 crop_top_4 {parent=BASE_crop_top_4, loc=<0 0 0>, rot=<0 0 0 1>}
 BASE_crop_top_5 {parent=spine_03, loc=<6.6111 -8.9415 -4.1089>, rot=<0.003 -0.055 0.997 -0.0547>}
 crop_top_5 {parent=BASE_crop_top_5, loc=<0 0 0>, rot=<0 0 0 1>}
 BASE_crop_top_6 {parent=spine_03, loc=<6.1593 -5.6287 11.1298>, rot=<-0.0249 0.4534 0.8896 -0.0488>}
 crop_top_6 {parent=BASE_crop_top_6, loc=<0 0 0>, rot=<0 0 0 1>}
 BASE_crop_top_7 {parent=spine_03, loc=<5.4153 0.8082 12.9951>, rot=<-0.0381 0.6954 0.7165 -0.0393>}
 crop_top_7 {parent=BASE_crop_top_7, loc=<0 0 0>, rot=<0 0 0 1>}
 BASE_crop_top_8 {parent=spine_03, loc=<4.5051 10.6081 10.0168>, rot=<0.0475 -0.8672 -0.4949 0.0271>}
 crop_top_8 {parent=BASE_crop_top_8, loc=<0 0 0>, rot=<0 0 0 1>}
 thigh_l {parent=pelvis, loc=<-1.0855 -0.2109 -8.7053>, rot=<0.0009 0.0355 0.0155 -0.9993>}
 calf_l {parent=thigh_l, loc=<-42.5284 -0.0019 -0.7297>, rot=<0.0006 -0.0161 -0.0391 -0.9991>}
 foot_l {parent=calf_l, loc=<-46.7401 -2.5014 -0.592>, rot=<0.0427 -0.0238 0.0353 -0.9982>}
 ball_l {parent=foot_l, loc=<-3.5658 10.8066 -0.1544>, rot=<0.0001 -0.0001 -0.7063 -0.7079>}
 calf_twist_02_l {parent=calf_l, loc=<-19.1546 -1.7943 -0.2638>, rot=<0.0237 0.0095 -0.0035 -0.9997>}
 calf_twist_01_l {parent=calf_l, loc=<-35.8483 -1.9197 -0.5794>, rot=<0.0237 0.0095 -0.0035 -0.9997>}
 deform_knee_in_base_l {parent=calf_l, loc=<0.8031 -5.0231 0.2456>, rot=<0 0 0 -1>}
 deform_knee_in_l {parent=deform_knee_in_base_l, loc=<0 0 0>, rot=<0 0 0 1>}
 deform_knee_out_base_l {parent=calf_l, loc=<0 4 0>, rot=<0 0 0 -1>}
 deform_knee_out_l {parent=deform_knee_out_base_l, loc=<0 0 0>, rot=<0 0 0 1>}
 thigh_twist_01_l {parent=thigh_l, loc=<-27.952 0 0>, rot=<0.0013 0 0 -1>}
 deform_groin_base_l {parent=thigh_l, loc=<0.4801 5.6776 -0.2486>, rot=<0 0 0 -1>}
 deform_groin_l {parent=deform_groin_base_l, loc=<0 0 0>, rot=<0 0 0 1>}
 button_l_2 {parent=thigh_l, loc=<-0.8173 6.9075 0.4503>, rot=<-0.4588 -0.5603 -0.3916 -0.5676>}
 button_l_1 {parent=thigh_l, loc=<3.6353 7.0378 0.8543>, rot=<-0.364 -0.5653 -0.392 -0.6279>}
 thigh_r {parent=pelvis, loc=<-1.0855 -0.2109 8.7053>, rot=<0.0355 -0.0009 0.9993 0.0155>}
 calf_r {parent=thigh_r, loc=<42.5284 0.0019 0.7297>, rot=<0.0006 -0.0161 -0.0391 -0.9991>}
 foot_r {parent=calf_r, loc=<46.7401 2.5014 0.592>, rot=<0.0427 -0.0238 0.0353 -0.9982>}
 ball_r {parent=foot_r, loc=<3.5658 -10.8066 0.1544>, rot=<0.0001 -0.0001 -0.7063 -0.7079>}
 calf_twist_02_r {parent=calf_r, loc=<19.1546 1.7943 0.2638>, rot=<0.0237 0.0095 -0.0035 -0.9997>}
 calf_twist_01_r {parent=calf_r, loc=<35.8483 1.9197 0.5794>, rot=<0.0237 0.0095 -0.0035 -0.9997>}
 deform_knee_in_base_r {parent=calf_r, loc=<-0.8031 5.0231 -0.2457>, rot=<0 0 0 -1>}
 deform_knee_in_r {parent=deform_knee_in_base_r, loc=<0 0 0>, rot=<0 0 0 1>}
 deform_knee_out_base_r {parent=calf_r, loc=<0 -4 0>, rot=<0 0 0 -1>}
 deform_knee_out_r {parent=deform_knee_out_base_r, loc=<0 0 0>, rot=<0 0 0 1>}
 thigh_twist_01_r {parent=thigh_r, loc=<27.952 0 0>, rot=<0.0013 0 0 -1>}
 deform_groin_base_r {parent=thigh_r, loc=<-0.4801 -5.6776 0.2486>, rot=<0 0 0 -1>}
 deform_groin_r {parent=deform_groin_base_r, loc=<0 0 0>, rot=<0 0 0 1>}
 button_r_1 {parent=thigh_r, loc=<-3.6353 -7.0378 -0.8543>, rot=<-0.3937 0.6118 0.3818 -0.57>}
 button_r_2 {parent=thigh_r, loc=<0.8173 -6.9075 -0.4503>, rot=<-0.4247 0.5864 0.428 -0.5409>}
 deform_glute_base_l {parent=pelvis, loc=<-7.3466 -9.0213 -9.1245>, rot=<0.0172 0.031 -0.4477 -0.8935>}
 deform_glute_l {parent=deform_glute_base_l, loc=<0 0 0>, rot=<0 0 0 1>}
 deform_glute_base_r {parent=pelvis, loc=<-7.3464 -9.0213 9.1245>, rot=<0.031 -0.0172 0.8935 -0.4477>}
 deform_glute_r {parent=deform_glute_base_r, loc=<0 0 0>, rot=<0 0 0 1>}
 bag_root_ctrl_null {type=Null, loc=<9.3698 -6.1149 142.0953>, rot=<0 0 0 1>}
 bag_root_ctrl {controlType=EULER_TRANSFORM, parent=bag_root_ctrl_null, type=Control, loc=<-9.3698 6.1149 -142.0953>, rot=<0 0 0 1>}
 bag_front_01_ctrl_null {parent=bag_root_ctrl, type=Null, loc=<-0.8717 5.0598 -1.8569>, rot=<-0.4739 -0.2497 0.7977 0.2769>}
 bag_front_01_ctrl {controlType=EULER_TRANSFORM, parent=bag_front_01_ctrl_null, type=Control, loc=<-4.9263 2.2476 0.6993>, rot=<0.4739 0.2497 -0.7977 0.2769>}
 bag_front_02_ctrl_null {parent=bag_front_01_ctrl, type=Null, loc=<7.8242 0 0>, rot=<-0.0125 0.0259 -0.0631 -0.9976>}
 bag_front_02_ctrl {controlType=EULER_TRANSFORM, parent=bag_front_02_ctrl_null, type=Control, loc=<-7.7514 0.9897 0.3918>, rot=<0.0125 -0.0259 0.0631 -0.9976>}
 bag_front_03_ctrl_null {parent=bag_front_02_ctrl, type=Null, loc=<7.9872 0 0>, rot=<0.061 0.0171 -0.146 -0.9873>}
 bag_front_03_ctrl {controlType=EULER_TRANSFORM, parent=bag_front_03_ctrl_null, type=Control, loc=<-7.6419 2.2865 0.4121>, rot=<-0.061 -0.0171 0.146 -0.9873>}
 bag_front_04_ctrl_null {parent=bag_front_03_ctrl, type=Null, loc=<5.7188 0 0>, rot=<-0.1047 0.0202 -0.1879 -0.9764>}
 bag_front_04_ctrl {controlType=EULER_TRANSFORM, parent=bag_front_04_ctrl_null, type=Control, loc=<-5.3102 2.1228 0>, rot=<0.1047 -0.0202 0.1879 -0.9764>}
 bag_front_05_ctrl_null {parent=bag_front_04_ctrl, type=Null, loc=<8.1103 0 0>, rot=<-0.0658 0.0061 -0.0917 -0.9936>}
 bag_front_05_ctrl {controlType=EULER_TRANSFORM, parent=bag_front_05_ctrl_null, type=Control, loc=<-7.9733 1.4842 0>, rot=<0.0658 -0.0061 0.0917 -0.9936>}
 bag_front_06_ctrl_null {parent=bag_front_05_ctrl, type=Null, loc=<8.022 0 0>, rot=<0 0 0 1>}
 bag_front_06_ctrl {controlType=EULER_TRANSFORM, parent=bag_front_06_ctrl_null, type=Control, loc=<-8.022 0 0>, rot=<0 0 0 1>}
 bag_back_01_ctrl_null {parent=bag_root_ctrl, type=Null, loc=<-1.5508 -3.7462 -1.7913>, rot=<-0.4594 0.1832 0.8598 -0.1272>}
 bag_back_01_ctrl {controlType=EULER_TRANSFORM, parent=bag_back_01_ctrl_null, type=Control, loc=<-3.6277 -2.5216 0.3593>, rot=<0.4594 -0.1832 -0.8598 -0.1272>}
 bag_back_02_ctrl_null {parent=bag_back_01_ctrl, type=Null, loc=<5.5424 0 0>, rot=<0 0 0.1808 -0.9835>}
 bag_back_02_ctrl {controlType=EULER_TRANSFORM, parent=bag_back_02_ctrl_null, type=Control, loc=<-5.18 -1.9711 0>, rot=<0 0 -0.1808 -0.9835>}
 bag_back_03_ctrl_null {parent=bag_back_02_ctrl, type=Null, loc=<6.3957 0 0>, rot=<0.0074 0.0005 0.0666 -0.9978>}
 bag_back_03_ctrl {controlType=EULER_TRANSFORM, parent=bag_back_03_ctrl_null, type=Control, loc=<-6.339 -0.8495 0>, rot=<-0.0074 -0.0005 -0.0666 -0.9978>}
 bag_back_04_ctrl_null {parent=bag_back_03_ctrl, type=Null, loc=<5.8846 0 0>, rot=<0.1516 0.0023 0.0149 -0.9883>}
 bag_back_04_ctrl {controlType=EULER_TRANSFORM, parent=bag_back_04_ctrl_null, type=Control, loc=<-5.882 -0.1769 0>, rot=<-0.1516 -0.0023 -0.0149 -0.9883>}
 bag_back_05_ctrl_null {parent=bag_back_04_ctrl, type=Null, loc=<6.3129 0 0>, rot=<-0.0263 -0.0011 0.0437 -0.9987>}
 bag_back_05_ctrl {controlType=EULER_TRANSFORM, parent=bag_back_05_ctrl_null, type=Control, loc=<-6.2888 -0.5514 0>, rot=<0.0263 0.0011 -0.0437 -0.9987>}
 bag_back_06_ctrl_null {parent=bag_back_05_ctrl, type=Null, loc=<9.4374 0 0>, rot=<0 0 0 1>}
 bag_back_06_ctrl {controlType=EULER_TRANSFORM, parent=bag_back_06_ctrl_null, type=Control, loc=<-9.4374 0 0>, rot=<0 0 0 1>}
 dyn_bag_sim_chain_03_ctrl_null {type=Null, loc=<-9.494 -1.6129 111.2473>, rot=<0.1876 0.6933 -0.6599 0.2208>}
 dyn_bag_sim_chain_03_ctrl {
  controlType=EULER_TRANSFORM
  parent=dyn_bag_sim_chain_03_ctrl_null
  type=Control
  loc=<53.6466 97.9012 2.4688>
  rot=<-0.1876 -0.6933 0.6599 0.2208>
 }
 sleeve_R {type=Curve, loc=<0 0 0>, rot=<0 0 0 1>}
 sleeve_L {type=Curve, loc=<0 0 0>, rot=<0 0 0 1>}
]
variables [
 RBAN Enabled {type=bool, default=False}
 bag_front_ctrls {type="TArray<FRigElementKey>"}
 bag_front_ctrls_RBAN {
  type="TArray<FRigElementKey>"
  default=((Type=Control,Name="bag_front_01_ctrl"),(Type=Control,Name="bag_front_02_ctrl"),(Type=Control,Name="bag_front_03_ctrl"),(Type=Control,Name="bag_front_04_ctrl"))
 }
 bag_front_joints {type="TArray<FRigElementKey>"}
 bag_front_joints_RBAN {
  type="TArray<FRigElementKey>"
  default=((Type=Bone,Name="bag_front_01"),(Type=Bone,Name="bag_front_02"),(Type=Bone,Name="bag_front_03"),(Type=Bone,Name="bag_front_04"))
 }
 bag_rear_ctrls {type="TArray<FRigElementKey>"}
 bag_rear_ctrls_RBAN {
  type="TArray<FRigElementKey>"
  default=((Type=Control,Name="bag_back_01_ctrl"),(Type=Control,Name="bag_back_02_ctrl"),(Type=Control,Name="bag_back_03_ctrl"),(Type=Control,Name="bag_back_04_ctrl"))
 }
 bag_rear_joints {type="TArray<FRigElementKey>"}
 bag_rear_joints_RBAN {
  type="TArray<FRigElementKey>"
  default=((Type=Bone,Name="bag_back_01"),(Type=Bone,Name="bag_back_02"),(Type=Bone,Name="bag_back_03"),(Type=Bone,Name="bag_back_04"))
 }
 blink {type=double, default=0}
 junping {type=double, default=0}
]
