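# Physics asset for the F_MED_CyberTentacle female player body cosmetic, set up for secondary motion of a bag and its dangling accessories.
# Capsules on pelvis, spine and right thigh form a partial body proxy for the bag pieces.
# A two-link bag chain hangs from spine_03 and carries a hedgehog sphere and zipper capsule, while a front bag piece attaches to spine_04.
# Tight, mostly asymmetric angular limits restrict the accessories' swing and twist.
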PhysicsAsset /BRCosmetics/Characters/Player/Female/Medium/Bodies/F_MED_CyberTentacle/Meshes/F_MED_CyberTentacle_Physics
bodies {
 bag_front_04 [{
   shapeType=Capsule
   length=2.8877
   radius=1.5802
   shapeName=bag_front_04_capsule
   center=<5.3674 -0.1809 -0.0001>
   rotation=<rot 0 -86.347 -88.2012>
  }]
 dyn_bag_sim_chain_01 [{
   shapeType=Capsule
   length=11.3966
   radius=4.5798
   shapeName=dyn_bag_sim_chain_01_capsule
   center=<15.3455 -4.3053 -0.2648>
   rotation=<rot 5.5982 -90 -0.72>
  }]
 dyn_bag_sim_chain_02 [{
   shapeType=Capsule
   length=11.5537
   radius=6.5671
   shapeName=dyn_bag_sim_chain_03_capsule
   center=<17.8994 -4.9136 -4.0846>
   rotation=<rot 5.0918 -89.7479 0.1974>
  }]
 hedgehog [{shapeType=Sphere, radius=6.8145, shapeName=hedgehog_sphere, center=<5.4472 -0.6176 0.2451>}]
 pelvis [
  {shapeType=Capsule, length=9.6453, radius=10.6437}
  {shapeType=Capsule, length=9.6453, radius=10.6437, shapeName=pelvis_capsule, center=<-4.8154 -3.1111 0>}
 ]
 spine_03 [{shapeType=Capsule, length=0.7317, radius=9.7204, center=<0 0.426 0>}]
 spine_04 [{shapeType=Capsule, length=14.3445, radius=11.3897, center=<0 1.0898 0>}]
 thigh_r [{shapeType=Capsule, length=23.818, radius=7.6618, center=<22.0813 0.4931 -0.1852>, rotation=<rot 0.0311 -90.4192 -88.3929>}]
 zipper [{shapeType=Capsule, length=3.941, radius=1.1214, shapeName=zipper_capsule, center=<3.8033 0.3293 0.1454>, rotation=<rot 0 -84.018 -89.8579>}]
}
constraints [
 {bone1Name=spine_03, bone2Name=pelvis, swing1LimitDegrees=45, swing2LimitDegrees=45, twistLimitDegrees=45}
 {bone1Name=spine_04, bone2Name=spine_03, swing1LimitDegrees=45, swing2LimitDegrees=45, twistLimitDegrees=45}
 {bone1Name=thigh_r, bone2Name=pelvis, swing1LimitDegrees=45, swing2LimitDegrees=45, twistLimitDegrees=45}
 {bone1Name=dyn_bag_sim_chain_02, bone2Name=dyn_bag_sim_chain_01, swing1LimitDegrees=40, swing2LimitDegrees=10, twistLimitDegrees=5}
 {bone1Name=dyn_bag_sim_chain_01, bone2Name=spine_03, swing1LimitDegrees=40, swing2LimitDegrees=2, twistLimitDegrees=2}
 {bone1Name=hedgehog, bone2Name=dyn_bag_sim_chain_02, swing1LimitDegrees=45, swing2LimitDegrees=45, twistLimitDegrees=20}
 {bone1Name=zipper, bone2Name=dyn_bag_sim_chain_02, swing1LimitDegrees=5, swing2LimitDegrees=45, twistLimitDegrees=3}
 {bone1Name=bag_front_04, bone2Name=spine_04, swing1LimitDegrees=20, swing2LimitDegrees=5, twistLimitDegrees=5}
]
