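# Post-process style AnimBlueprint for the female CyberTentacle cosmetic body, adding secondary motion to a pose copied from its parent mesh.
# Control Rig takes blink and jump values, then AnimDynamics sways sleeves, crop top and bracelet, followed by RigidBody simulation.
# Bone modifiers rotate the eyes, enlarge upper-arm twists while parachuting and reset bag-chain scale.
# EventGraph lerps eye darts through a rotator array with random pauses and derives periodic blinks from a sine-arctangent curve.
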AnimBlueprint /BRCosmetics/Characters/Player/Female/Medium/Bodies/F_MED_CyberTentacle/Meshes/F_MED_CyberTentacle_AnimBP
target_skeleton=&/F_MED_CyberTentacle_Skeleton
graphs [
 AnimGraph {nodes [
   n1 = Control Rig!(
    Source=Copy Pose From Mesh() {bUseAttachedParent=true}
    Alpha
    blink:double=Get blink() -> (blink:double)
    junping:double=Get jumping_f() -> (jumping_f:double)
   ) -> (Pose) {controlRigAssetReference {blueprintRigClass=&/F_MED_CyberTentacle_CtrlRig.F_MED_CyberTentacle_CtrlRig_C}}
   n2 = Anim Dynamics - Bone: sleeve_3_l(ComponentPose=Local To Component(LocalPose=@n1), Alpha=1, ExternalForce="0,0,0") {
    angularDampingOverride=1
    angularSpringConstant=20000
    bAngularSpring=true
    bOverrideAngularDamping=true
    bUseGravityOverride=true
    gravityScale=0.35
    simulationSpace=RootRelative
    alphaBoolBlend {blendOption=Linear}
    boundBone {boneName=sleeve_3_l}
    componentLinearAccScale=<1 1 1>
    componentLinearVelScale=<1 1 1>
    gravityOverride {z=-980}
    physicsBodyDefinitions [{
      boundBone {boneName=sleeve_3_l}
      boxExtents=<20 30 30>
      constraintSetup {angularLimitsMax=<5 30 15>, angularLimitsMin=<-5 0 -15>, angularTarget=<1 0 0>}
      localJointOffset=<4 0 0>
     }]
   }
   n3 = Anim Dynamics - Bone: sleeve_3_r(ComponentPose=@n2, Alpha=1, ExternalForce="0,0,0") {
    angularDampingOverride=1
    angularSpringConstant=20000
    bAngularSpring=true
    bOverrideAngularDamping=true
    bUseGravityOverride=true
    gravityScale=0.35
    simulationSpace=RootRelative
    alphaBoolBlend {blendOption=Linear}
    boundBone {boneName=sleeve_3_r}
    componentLinearAccScale=<1 1 1>
    componentLinearVelScale=<1 1 1>
    gravityOverride {z=-980}
    physicsBodyDefinitions [{
      boundBone {boneName=sleeve_3_r}
      boxExtents=<20 30 30>
      constraintSetup {angularLimitsMax=<5 15 15>, angularLimitsMin=<-5 -15 -15>, angularTarget=<1 0 0>}
      localJointOffset=<4 0 0>
     }]
   }
   n4 = Anim Dynamics - Bone: BASE_crop_top_2(ComponentPose=@n3, Alpha=1, ExternalForce="0,0,0") {
    angularDampingOverride=0.999
    angularSpringConstant=20000
    bAngularSpring=true
    bGravityOverrideInSimSpace=true
    bOverrideAngularDamping=true
    bUseGravityOverride=true
    gravityScale=0.35
    simulationSpace=BoneRelative
    boundBone {boneName=BASE_crop_top_2}
    componentLinearAccScale=<1 1 1>
    componentLinearVelScale=<1 1 1>
    gravityOverride {z=-980}
    physicsBodyDefinitions [{
      boundBone {boneName=BASE_crop_top_2}
      boxExtents=<20 30 30>
      constraintSetup {angularLimitsMax=<5 20 20>, angularLimitsMin=<-5 -20 -20>, angularTarget=<1 0 0>}
      localJointOffset=<1 0 0>
     }]
    relativeSpaceBone {boneName=pelvis}
   }
   n5 = Anim Dynamics - Bone: BASE_crop_top_3(ComponentPose=@n4, Alpha=1, ExternalForce="0,0,0") {
    angularDampingOverride=0.999
    angularSpringConstant=10000
    bAngularSpring=true
    bGravityOverrideInSimSpace=true
    bOverrideAngularDamping=true
    bUseGravityOverride=true
    gravityScale=0.35
    simulationSpace=BoneRelative
    boundBone {boneName=BASE_crop_top_3}
    componentLinearAccScale=<1 1 1>
    componentLinearVelScale=<1 1 1>
    gravityOverride {z=-980}
    physicsBodyDefinitions [{
      boundBone {boneName=BASE_crop_top_3}
      boxExtents=<20 30 30>
      constraintSetup {angularLimitsMax=<5 20 20>, angularLimitsMin=<-5 -20 -20>, angularTarget=<1 0 0>}
      localJointOffset=<1 0 0>
     }]
    relativeSpaceBone {boneName=pelvis}
   }
   n6 = Anim Dynamics - Bone: BASE_crop_top_4(ComponentPose=@n5, Alpha=1, ExternalForce="0,0,0") {
    angularDampingOverride=0.999
    angularSpringConstant=10000
    bAngularSpring=true
    bGravityOverrideInSimSpace=true
    bOverrideAngularDamping=true
    bUseGravityOverride=true
    gravityScale=0.35
    simulationSpace=BoneRelative
    boundBone {boneName=BASE_crop_top_4}
    componentLinearAccScale=<1 1 1>
    componentLinearVelScale=<1 1 1>
    gravityOverride {z=-980}
    physicsBodyDefinitions [{
      boundBone {boneName=BASE_crop_top_4}
      boxExtents=<20 30 30>
      constraintSetup {angularLimitsMax=<5 20 20>, angularLimitsMin=<-5 -20 -20>, angularTarget=<1 0 0>}
      localJointOffset=<1 0 0>
     }]
    relativeSpaceBone {boneName=pelvis}
   }
   n7 = Anim Dynamics - Bone: BASE_crop_top_5(ComponentPose=@n6, Alpha=1, ExternalForce="0,0,0") {
    angularDampingOverride=0.999
    angularSpringConstant=10000
    bAngularSpring=true
    bGravityOverrideInSimSpace=true
    bOverrideAngularDamping=true
    bUseGravityOverride=true
    gravityScale=0.35
    simulationSpace=BoneRelative
    boundBone {boneName=BASE_crop_top_5}
    componentLinearAccScale=<1 1 1>
    componentLinearVelScale=<1 1 1>
    gravityOverride {z=-980}
    physicsBodyDefinitions [{
      boundBone {boneName=BASE_crop_top_5}
      boxExtents=<20 30 30>
      constraintSetup {angularLimitsMax=<5 20 20>, angularLimitsMin=<-5 -20 -20>, angularTarget=<1 0 0>}
      localJointOffset=<1 0 0>
     }]
    relativeSpaceBone {boneName=pelvis}
   }
   n8 = Anim Dynamics - Bone: BASE_crop_top_6(ComponentPose=@n7, Alpha=1, ExternalForce="0,0,0") {
    angularDampingOverride=0.999
    angularSpringConstant=10000
    bAngularSpring=true
    bGravityOverrideInSimSpace=true
    bOverrideAngularDamping=true
    bUseGravityOverride=true
    gravityScale=0.35
    simulationSpace=BoneRelative
    boundBone {boneName=BASE_crop_top_6}
    componentLinearAccScale=<1 1 1>
    componentLinearVelScale=<1 1 1>
    gravityOverride {z=-980}
    physicsBodyDefinitions [{
      boundBone {boneName=BASE_crop_top_6}
      boxExtents=<20 30 30>
      constraintSetup {angularLimitsMax=<5 20 20>, angularLimitsMin=<-5 -20 -20>, angularTarget=<1 0 0>}
      localJointOffset=<1 0 0>
     }]
    relativeSpaceBone {boneName=pelvis}
   }
   n9 = Anim Dynamics - Bone: BASE_crop_top_7(ComponentPose=@n8, Alpha=1, ExternalForce="0,0,0") {
    angularDampingOverride=0.999
    angularSpringConstant=10000
    bAngularSpring=true
    bGravityOverrideInSimSpace=true
    bOverrideAngularDamping=true
    bUseGravityOverride=true
    gravityScale=0.35
    simulationSpace=BoneRelative
    boundBone {boneName=BASE_crop_top_7}
    componentLinearAccScale=<1 1 1>
    componentLinearVelScale=<1 1 1>
    gravityOverride {z=-980}
    physicsBodyDefinitions [{
      boundBone {boneName=BASE_crop_top_7}
      boxExtents=<20 30 30>
      constraintSetup {angularLimitsMax=<5 20 20>, angularLimitsMin=<-5 -20 -20>, angularTarget=<1 0 0>}
      localJointOffset=<1 0 0>
     }]
    relativeSpaceBone {boneName=pelvis}
   }
   n10 = Anim Dynamics - Bone: BASE_crop_top_8(ComponentPose=@n9, Alpha=1, ExternalForce="0,0,0") {
    angularDampingOverride=0.999
    angularSpringConstant=10000
    bAngularSpring=true
    bGravityOverrideInSimSpace=true
    bOverrideAngularDamping=true
    bUseGravityOverride=true
    gravityScale=0.35
    simulationSpace=BoneRelative
    boundBone {boneName=BASE_crop_top_8}
    componentLinearAccScale=<1 1 1>
    componentLinearVelScale=<1 1 1>
    gravityOverride {z=-980}
    physicsBodyDefinitions [{
      boundBone {boneName=BASE_crop_top_8}
      boxExtents=<20 30 30>
      constraintSetup {angularLimitsMax=<5 20 20>, angularLimitsMin=<-5 -20 -20>, angularTarget=<1 0 0>}
      localJointOffset=<1 0 0>
     }]
    relativeSpaceBone {boneName=pelvis}
   }
   n11 = Anim Dynamics - Bone: BASE_crop_top_1(ComponentPose=@n10, Alpha=1, ExternalForce="0,0,0") {
    angularDampingOverride=0.999
    angularSpringConstant=10000
    bAngularSpring=true
    bGravityOverrideInSimSpace=true
    bOverrideAngularDamping=true
    bUseGravityOverride=true
    gravityScale=0.35
    simulationSpace=BoneRelative
    boundBone {boneName=BASE_crop_top_1}
    componentLinearAccScale=<1 1 1>
    componentLinearVelScale=<1 1 1>
    gravityOverride {z=-980}
    physicsBodyDefinitions [{
      boundBone {boneName=BASE_crop_top_1}
      boxExtents=<20 30 30>
      constraintSetup {angularLimitsMax=<5 20 20>, angularLimitsMin=<-5 -20 -20>, angularTarget=<1 0 0>}
      localJointOffset=<1 0 0>
     }]
    relativeSpaceBone {boneName=pelvis}
   }
   n12 = Anim Dynamics - Bone: BASE_dyn_bracelet(ComponentPose=@n11, Alpha=1, ExternalForce="0,0,0") {
    angularDampingOverride=1
    angularSpringConstant=20000
    bAngularSpring=true
    bOverrideAngularDamping=true
    bUseGravityOverride=true
    gravityScale=0.35
    simulationSpace=RootRelative
    alphaBoolBlend {blendOption=Linear}
    boundBone {boneName=BASE_dyn_bracelet}
    componentLinearAccScale=<1 1 1>
    componentLinearVelScale=<1 1 1>
    gravityOverride {z=-980}
    physicsBodyDefinitions [{
      boundBone {boneName=BASE_dyn_bracelet}
      boxExtents=<20 30 30>
      constraintSetup {angularLimitsMax=<15 7 1>, angularLimitsMin=<-15 -7 -1>, angularTarget=<1 0 0>}
      localJointOffset=<4 0 0>
     }]
   }
   n13 = RigidBody!(
    ComponentPose=@n12
    bAlphaBoolEnabled:bool
    OverrideWorldGravity:Vector="0,0,-980"
    SimSpaceSettings:Sim Space Settings Structure=(WorldAlpha=1,VelocityScaleZ=0.3,DampingAlpha=1,MaxLinearVelocity=3000,MaxAngularVelocity=1500,MaxLinearAcceleration=1000,MaxAngularAcceleration=65,ExternalLinearDragV=(X=1,Y=1,Z=1),ExternalLinearVelocity=(X=0,Y=0,Z=0),ExternalAngularVelocity=(X=0,Y=0,Z=0))
   ) -> (Pose) {
    alphaInputType=Bool
    bClampLinearTranslationLimitToRefPose=true
    bOverrideWorldGravity=true
    overridePhysicsAsset=&/F_MED_CyberTentacle_Physics
    simulationSpace=BaseBoneSpace
    alphaBoolBlend {blendInTime=0.2, blendOption=Linear, blendOutTime=0.2}
    baseBoneRef {boneName=spine_03}
    overrideWorldGravity {z=-980}
    simSpaceSettings {
     maxAngularAcceleration=65
     maxAngularVelocity=1500
     maxLinearAcceleration=1000
     maxLinearVelocity=3000
     velocityScaleZ=0.3
     worldAlpha=1
     externalLinearDragV=<1 1 1>
    }
   }
   n14 = "Transform (Modify) Bone - Bone: eye_1"!(ComponentPose=@n13, bAlphaBoolEnabled:bool=True, Translation="0,0,0", Rotation=Get Eye_rotation() -> (Eye_rotation:Rotator), Scale="1,1,1") -> (Pose) {
    alphaInputType=Bool
    rotationMode=BMM_Additive
    rotationSpace=BCS_BoneSpace
    alphaBoolBlend {blendInTime=0.8, blendOption=Linear, blendOutTime=0.8}
    boneToModify {boneName=eye_1}
   }
   n15 = "Transform (Modify) Bone - Bone: eye_2"!(ComponentPose=@n14, bAlphaBoolEnabled:bool=True, Translation="0,0,0", Rotation=Get Eye_rotation(), Scale="1,1,1") -> (Pose) {
    alphaInputType=Bool
    rotationMode=BMM_Additive
    rotationSpace=BCS_BoneSpace
    alphaBoolBlend {blendInTime=0.8, blendOption=Linear, blendOutTime=0.8}
    boneToModify {boneName=eye_2}
   }
   n16 = Get bIsParachuteOpened()
   n17 = "Transform (Modify) Bone - Bone: upperarm_twist_02_l"!(ComponentPose=@n15, bAlphaBoolEnabled:bool=@n16, Translation="0,0,0", Rotation="0,0,0", Scale="1.7,1.7,1.7") -> (Pose) {
    alphaInputType=Bool
    rotationSpace=BCS_BoneSpace
    scaleMode=BMM_Additive
    alphaBoolBlend {blendInTime=0.5, blendOption=Linear, blendOutTime=0.5}
    boneToModify {boneName=upperarm_twist_02_l}
    scale=<1.7 1.7 1.7>
   }
   n18 = "Transform (Modify) Bone - Bone: upperarm_twist_02_r"!(ComponentPose=@n17, bAlphaBoolEnabled:bool=@n16, Translation="0,0,0", Rotation="0,0,0", Scale="1.7,1.7,1.7") -> (Pose) {
    alphaInputType=Bool
    rotationSpace=BCS_BoneSpace
    scaleMode=BMM_Additive
    alphaBoolBlend {blendInTime=0.5, blendOption=Linear, blendOutTime=0.5}
    boneToModify {boneName=upperarm_twist_02_r}
    scale=<1.7 1.7 1.7>
   }
   n19 = "Transform (Modify) Bone - Bone: dyn_bag_sim_chain_01"!(ComponentPose=@n18, bAlphaBoolEnabled:bool=True, Translation="0,0,0", Rotation="0,0,0", Scale="1,1,1") -> (Pose) {
    alphaInputType=Bool
    rotationSpace=BCS_BoneSpace
    scaleMode=BMM_Replace
    alphaBoolBlend {blendInTime=0.5, blendOption=Linear, blendOutTime=0.5}
    boneToModify {boneName=dyn_bag_sim_chain_01}
   }
   Output Pose(Result=Component To Local(ComponentPose=@n19)) {layerGroup=None}
  ]}
 EventGraph {nodes [
   n1 = Event BlueprintUpdateAnimation()
   > Sequence!(execute) -> (then_0, then_1, then_2:Exec)
   > n2 = "Gravity Override State Function (Parameter Asset) (Nativized)"(
    CharacterPartAnimInstance=Self-Reference()
    GameTimeInSeconds=GetGameTimeInSeconds()
    Parameters=&/F_MED_CyberTentacle_Parameters
   )
   > Set Gravity Override(Gravity Override=@n2)
   > Set sleeve_L(sleeve_L:double=GetCurveValue(CurveName=sleeve_L)) -> (Output_Get:double)
   > Set sleeve_R(sleeve_R:double=GetCurveValue(CurveName=sleeve_R)) -> (Output_Get:double)
   > n3 = Sequence!(execute) -> (then_0, then_1)
   n4 = Get eye_rotation_array() -> (eye_rotation_array:Array of Rotators)
   n5 = Get array_index() -> (array_index:int=0)
   > Set Eye_position_1(Eye_position_1:Rotator="Get (a copy)"(Array:Array of Rotators=@n4, Dimension 1=@n5) -> (Output:Rotator)) -> (Output_Get:Rotator)
   > Set Eye_position_2(Eye_position_2:Rotator="Get (a copy)"(Array=@n4, Dimension 1=int + int(A=@n5))) -> (Output_Get:Rotator)
   n6 = float * float!(A=Get period() -> (period:double), B=Get Rotation_speed_coefficient() -> (Rotation_speed_coefficient:double), C:double=1.5) -> (ReturnValue) {is=title K2Node_PromotableOperator}
   > n7 = Set alpha_position(alpha_position:double=float + float(A=float / float(A=@n1.DeltaTimeX, B=@n6), B=Get alpha_position() -> (alpha_position:double))) -> (Output_Get:double)
   n8 = BreakRotator(InRot=Get Eye_position_1() -> (Eye_position_1:Rotator))
   n9 = BreakRotator(InRot=Get Eye_position_2() -> (Eye_position_2:Rotator))
   n10 = float - float(A=Vector Length(A=MakeVector(X=@n8, Y=@n8.Pitch, Z=@n8.Yaw)), B=Vector Length(A=MakeVector(X=@n9, Y=@n9.Pitch, Z=@n9.Yaw)))
   > Set Rotation_speed_coefficient(Rotation_speed_coefficient:double=MapRangeClamped(Value="Absolute (Float)"(A=@n10), InRangeA=0, InRangeB=27, OutRangeA=0.5)) -> (Output_Get:double)
   > n11 = Branch(Condition="float < float"(A=@n7, B=1))
   > Set Eye_rotation(Eye_rotation:Rotator="Lerp (Rotator)"!(A:Rotator=Get Eye_position_1(), B:Rotator=Get Eye_position_2(), Alpha:float=Get alpha_position(), bShortestPath:bool=false) -> (ReturnValue:Rotator)) -> (Output_Get:Rotator)
   Delay(execute=@n11.else, Duration=RandomFloatInRange(Min=0.3, Max=0.6))
   > Set alpha_position(alpha_position=0)
   > Increment Int(Value=Get array_index() -> (array_index=0))
   n12 = Length(TargetArray:Array of Rotators=Get eye_rotation_array())
   > Branch(Condition="integer >= integer"(A=Get array_index() -> (array_index=0), B=int - int(A=@n12)))
   > Set array_index(array_index:int=0) -> (Output_Get:int=0)
   n13 = float * float(A=GetRealTimeSeconds(), B=1) {is=title K2Node_CommutativeAssociativeBinaryOperator}
   n14 = "Atan (Radians)"!(A:double=float * float(A="Sin (Radians)"(A=@n13), B=50)) -> (ReturnValue:double)
   n15 = float * float(A=@n14, B=@n14)
   Set blink(execute=@n3.then_1, blink:double=float + float(A=float * float(A=@n15, B=-0.46), B=1)) -> (Output_Get:double)
   > Set jumping_f(jumping_f:double="To Float (Boolean)"(InBool=Get bIsJumping())) -> (Output_Get:double)
  ]}
]
variables [
 alpha_position {type=number}
 array_index {type=integer}
 blink {type=number}
 Eye_position_1 {type=Rotator}
 Eye_position_2 {type=Rotator}
 Eye_rotation {type=Rotator}
 eye_rotation_array {
  type="array<Rotator>"
  default [<rot 0 0 0>, <rot 4 9 0>, <rot 5 7 0>, <rot 0 0 0>, <rot -8 -6 0>, <rot 3 9 0>, <rot 4 -8 0>, <rot 2 5 0>, <rot -6 11 0>, <rot 0 0 0>]
 }
 eye_rotation_array_2 {
  type="array<Rotator>"
  default [
   <rot 0 0 0>
   <rot 12 25 0>
   <rot 8 10 0>
   <rot 0 0 0>
   <rot -10 -20 0>
   <rot 10 -18 0>
   <rot 15 -8 0>
   <rot 7 20 0>
   <rot -15 15 0>
   <rot 0 0 0>
  ]
 }
 Gravity Override {type=Vector}
 jumping_f {type=number}
 period {type=number, default=0.1}
 Rotation_speed_coefficient {type=number}
 Sim Space Settings {
  type=SimSpaceSettings
  default {
   dampingAlpha=1
   maxAngularAcceleration=10000
   maxAngularVelocity=10000
   maxLinearAcceleration=10000
   maxLinearVelocity=10000
   velocityScaleZ=1
   worldAlpha=0
   externalLinearDragV=<0 0 0>
  }
 }
 sleeve_L {type=number}
 sleeve_R {type=number}
]
