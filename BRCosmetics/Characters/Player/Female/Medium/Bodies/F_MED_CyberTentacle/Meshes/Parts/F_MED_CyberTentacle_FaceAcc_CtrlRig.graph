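# Control Rig for the CyberTentacle character's face accessory that procedurally animates and constrains its tentacle bones.
# Front tentacles get harmonic waves passed down each chain through time-offset spring interpolation, weighted by play_f and smoothed, inverted clavicle pose readers.
# When backpack_cape_mod_bool is false, back tentacles use weighted parent constraints to body and bag bones, plus rotation offsets.
# Head pose readers relax some of those constraints as the head turns.
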ControlRigBlueprint /BRCosmetics/Characters/Player/Female/Medium/Bodies/F_MED_CyberTentacle/Meshes/Parts/F_MED_CyberTentacle_FaceAcc_CtrlRig
graphs [
 RigVMFunctionLibrary {links [], nodes []}
 RigVMFunction_Sequence_1_ContainedGraph {nodes [
   n1 = RigVMFunctionEntryNode()
   RigVMFunction_Sequence = &RigVMFunction_Sequence(ExecuteContext=@n1.ExecuteContext)
   n2 = &RigVMFunction_Sequence(ExecuteContext=@RigVMFunction_Sequence.B)
   RigVMFunctionReturnNode(A=@RigVMFunction_Sequence.A, B=@n2.A, C=@n2.B)
  ]}
 RigVMModel {nodes [
   RigUnit_BeginExecution = &RigUnit_BeginExecution()
   RigVMFunction_Sequence = RigVMAggregateNode(ExecuteContext=@RigUnit_BeginExecution.ExecutePin)
   n1 = &RigUnit_SphericalPoseReader(
    ActiveRegionScaleFactors=(PositiveWidth=0.5,NegativeWidth=0.5,PositiveHeight=0.5,NegativeHeight=0.5)
    ActiveRegionSize=0.5
    Debug=(bDrawDebug=False,bDraw2D=false,bDrawLocalAxes=false,DebugScale=30,DebugSegments=30,DebugThickness=0.2)
    DriverAxis=(X=1,Y=0,Z=0)
    DriverItem=(Type=Bone,Name="upperarm_l")
    FalloffRegionScaleFactors=(PositiveWidth=0.8,NegativeWidth=0.8,PositiveHeight=0.8,NegativeHeight=0.8)
    FalloffSize=0.5
    OptionalParentItem=(Type=Bone,Name="None")
    RotationOffset=(X=90,Y=-135,Z=0)
    ExecutePin=@RigVMFunction_Sequence.B
   )
   > n2 = &RigUnit_SphericalPoseReader(
    ActiveRegionScaleFactors=(PositiveWidth=1,NegativeWidth=0.5,PositiveHeight=1,NegativeHeight=1)
    ActiveRegionSize=0.3
    Debug=(bDrawDebug=False,bDraw2D=false,bDrawLocalAxes=false,DebugScale=30,DebugSegments=30,DebugThickness=0.2)
    DriverAxis=(X=1,Y=0,Z=0)
    DriverItem=(Type=Bone,Name="clavicle_l")
    FalloffRegionScaleFactors=(PositiveWidth=1,NegativeWidth=0.2,PositiveHeight=1,NegativeHeight=1)
    FalloffSize=0.5
    OptionalParentItem=(Type=Bone,Name="None")
    RotationOffset=(X=0,Y=40,Z=180)
   )
   AlphaInterp = &RigVMFunction_AlphaInterp(InterpSpeedDecreasing=0.5, InterpSpeedIncreasing=50, Scale=1, bClampResult=False, Value=@n2.OutputParam)
   Subtract = &RigVMFunction_MathDoubleSub(A=1, B=@AlphaInterp.Result)
   > RigVMVariableNode(Value=@Subtract.Result) {variable=sp_L}
   > n3 = &RigUnit_SphericalPoseReader(
    ActiveRegionScaleFactors=(PositiveWidth=1,NegativeWidth=0.5,PositiveHeight=1,NegativeHeight=1)
    ActiveRegionSize=0.3
    Debug=(bDrawDebug=False,bDraw2D=false,bDrawLocalAxes=false,DebugScale=30,DebugSegments=30,DebugThickness=0.2)
    DriverAxis=(X=-1,Y=0,Z=0)
    DriverItem=(Type=Bone,Name="clavicle_r")
    FalloffRegionScaleFactors=(PositiveWidth=1,NegativeWidth=0.2,PositiveHeight=1,NegativeHeight=1)
    FalloffSize=0.5
    OptionalParentItem=(Type=Bone,Name="None")
    RotationOffset=(X=0,Y=220,Z=180)
   )
   n4 = &RigVMFunction_AlphaInterp(InterpSpeedDecreasing=0.1, InterpSpeedIncreasing=50, Scale=1, bClampResult=False, Value=@n3.OutputParam)
   > RigVMVariableNode(Value=&RigVMFunction_MathDoubleSub(A=1, B=@n4.Result).Result) {variable=sp_R}
   n5 = &RigVMFunction_MathDoubleMul(A=RigVMVariableNode() {variable=play_f}.Value, B=RigVMVariableNode() {variable=sp_L}.Value)
   > /Script/ControlRig.RigUnit_ItemHarmonics(
    RotationOrder=YZX
    Targets=((Item=(Type=Bone,Name="extra_tentacle_front_l_B_02"),Ratio=1))
    WaveAmplitude=(X=0,Y=0,Z=30)
    WaveEase=Linear
    WaveFrequency=(X=1,Y=0.6,Z=0.6)
    WaveMinimum=0.5
    WaveNoise=(X=0,Y=0,Z=0)
    WaveOffset=(X=0,Y=0,Z=8)
    WaveSpeed=(X=3,Y=2,Z=2)
    WaveMaximum=@n5.Result
   ) {is=struct RigVMUnitNode}
   n6 = &RigVMFunction_MathQuaternionToEuler(
    RotationOrder=ZYX
    Value=&RigUnit_GetTransform(Item=(Type=Bone,Name="extra_tentacle_front_l_B_02"), Space=LocalSpace, bInitial=False).Transform.Rotation
   )
   > &RigUnit_ModifyTransforms(
    ItemToModify=((Item=(Type=Bone,Name="extra_tentacle_front_l_B_02"),Transform=(Rotation=(X=0,Y=0,Z=0,W=1),Translation=(X=0,Y=0,Z=0),Scale3D=(X=1,Y=1,Z=1))))
    Mode=OverrideLocal
    WeightMinimum=0
    Weight=@n5.Result
    ItemToModify.0.Transform.Rotation=&RigVMFunction_MathQuaternionFromEuler(
     Euler=(X=0,Y=0,Z=0)
     RotationOrder=XYZ
     Euler.Z=&RigVMFunction_MathDoubleRemap(SourceMaximum=29, SourceMinimum=-29, TargetMaximum=29, TargetMinimum=1, bClamp=false, Value=@n6.Result.Z).Result
    ).Result
   )
   n7 = &RigVMFunction_MathQuaternionToEuler(
    RotationOrder=XYZ
    Value=&RigUnit_GetTransform(Item=(Type=Bone,Name="extra_tentacle_front_l_B_02"), Space=LocalSpace, bInitial=False).Transform.Rotation
   )
   n8 = /Script/ControlRig.RigUnit_SpringInterpQuaternionV2(
    CriticalDamping=0.1
    Current=(X=0,Y=0,Z=0,W=1)
    Strength=1
    TargetVelocityAmount=0
    Torque=(X=0,Y=0,Z=0)
    bInitializeFromTarget=False
    bUseCurrentInput=False
    Target=/Script/RigVM.RigVMFunction_TimeOffsetTransform(
     BufferSize=16
     SecondsAgo=0.6
     TimeRange=1
     Value=(Rotation=(X=0,Y=0,Z=0,W=1),Translation=(X=0,Y=0,Z=0),Scale3D=(X=1,Y=1,Z=1))
     Value.Rotation=&RigVMFunction_MathQuaternionFromEuler(Euler=(X=0,Y=0,Z=0), RotationOrder=XYZ, Euler.Z=@n7.Result.Z).Result
    ) {is=struct RigVMUnitNode, template="TimeOffset::Execute(in Value,in SecondsAgo,in BufferSize,in TimeRange,out Result)"}.Result.Rotation
   ) {
    is=struct RigVMUnitNode
    template="SpringInterp::Execute(in Target,in Strength,in CriticalDamping,in Torque,in bUseCurrentInput,in Current,in TargetVelocityAmount,in bInitializeFromTarget,out Result,out AngularVelocity)"
   }
   > &RigUnit_ModifyTransforms(
    ItemToModify=((Item=(Type=Bone,Name="BASE_tentacle_front_l_B_03"),Transform=(Rotation=(X=0,Y=0,Z=0,W=1),Translation=(X=0,Y=0,Z=0),Scale3D=(X=1,Y=1,Z=1))))
    WeightMinimum=0
    Weight=@n5.Result
    ItemToModify.0.Transform.Rotation=@n8.Result
   )
   n9 = &RigVMFunction_MathDoubleMul(
    B=2
    A=&RigVMFunction_MathQuaternionToEuler(
     RotationOrder=XYZ
     Value=&RigUnit_GetTransform(Item=(Type=Bone,Name="BASE_tentacle_front_l_B_03"), Space=LocalSpace, bInitial=False).Transform.Rotation
    ).Result.Y
   )
   n10 = &RigUnit_SpringInterpQuaternionV2(
    CriticalDamping=1
    Current=(X=0,Y=0,Z=0,W=1)
    Strength=6
    TargetVelocityAmount=0
    Torque=(X=0,Y=0,Z=0)
    bInitializeFromTarget=False
    bUseCurrentInput=False
    Target=&RigVMFunction_TimeOffsetTransform(
     BufferSize=16
     SecondsAgo=0.1
     TimeRange=1
     Value=(Rotation=(X=0,Y=0,Z=0,W=1),Translation=(X=0,Y=0,Z=0),Scale3D=(X=1,Y=1,Z=1))
     Value.Rotation=&RigVMFunction_MathQuaternionFromEuler(Euler=(X=0,Y=0,Z=0), RotationOrder=XYZ, Euler.Y=@n9.Result).Result
    ) {template="TimeOffset::Execute(in Value,in SecondsAgo,in BufferSize,in TimeRange,out Result)"}.Result.Rotation
   ) {template="SpringInterp::Execute(in Target,in Strength,in CriticalDamping,in Torque,in bUseCurrentInput,in Current,in TargetVelocityAmount,in bInitializeFromTarget,out Result,out AngularVelocity)"}
   > &RigUnit_ModifyTransforms(
    ItemToModify=((Item=(Type=Bone,Name="BASE_tentacle_front_l_B_04"),Transform=(Rotation=(X=0,Y=0,Z=0,W=1),Translation=(X=0,Y=0,Z=0),Scale3D=(X=1,Y=1,Z=1))))
    WeightMinimum=0
    Weight=@n5.Result
    ItemToModify.0.Transform.Rotation=@n10.Result
   )
   n11 = &RigVMFunction_MathDoubleMul(
    B=1.5
    A=&RigVMFunction_MathQuaternionToEuler(
     RotationOrder=XYZ
     Value=&RigUnit_GetTransform(Item=(Type=Bone,Name="BASE_tentacle_front_l_B_04"), Space=LocalSpace, bInitial=False).Transform.Rotation
    ).Result.Y
   )
   n12 = &RigUnit_SpringInterpQuaternionV2(
    CriticalDamping=1
    Current=(X=0,Y=0,Z=0,W=1)
    Strength=6
    TargetVelocityAmount=0
    Torque=(X=0,Y=0,Z=0)
    bInitializeFromTarget=False
    bUseCurrentInput=False
    Target=&RigVMFunction_TimeOffsetTransform(
     BufferSize=16
     SecondsAgo=0.7
     TimeRange=1
     Value=(Rotation=(X=0,Y=0,Z=0,W=1),Translation=(X=0,Y=0,Z=0),Scale3D=(X=1,Y=1,Z=1))
     Value.Rotation=&RigVMFunction_MathQuaternionFromEuler(Euler=(X=0,Y=0,Z=0), RotationOrder=XYZ, Euler.Y=@n11.Result).Result
    ) {template="TimeOffset::Execute(in Value,in SecondsAgo,in BufferSize,in TimeRange,out Result)"}.Result.Rotation
   ) {template="SpringInterp::Execute(in Target,in Strength,in CriticalDamping,in Torque,in bUseCurrentInput,in Current,in TargetVelocityAmount,in bInitializeFromTarget,out Result,out AngularVelocity)"}
   > &RigUnit_ModifyTransforms(
    ItemToModify=((Item=(Type=Bone,Name="BASE_tentacle_front_l_B_05"),Transform=(Rotation=(X=0,Y=0,Z=0,W=1),Translation=(X=0,Y=0,Z=0),Scale3D=(X=1,Y=1,Z=1))))
    WeightMinimum=0
    Weight=@n5.Result
    ItemToModify.0.Transform.Rotation=@n12.Result
   )
   n13 = &RigVMFunction_MathDoubleMul(A=RigVMVariableNode() {variable=play_f}.Value, B=RigVMVariableNode() {variable=sp_R}.Value)
   > &RigUnit_ItemHarmonics(
    RotationOrder=YZX
    Targets=((Item=(Type=Bone,Name="extra_tentacle_front_r_B_02"),Ratio=1))
    WaveAmplitude=(X=0,Y=0,Z=30)
    WaveEase=Linear
    WaveFrequency=(X=1,Y=0.6,Z=0.6)
    WaveMinimum=0.5
    WaveNoise=(X=0,Y=0,Z=0)
    WaveOffset=(X=0,Y=0,Z=5)
    WaveSpeed=(X=3,Y=2,Z=2)
    WaveMaximum=@n13.Result
   )
   n14 = &RigVMFunction_MathDoubleRemap(
    SourceMaximum=29
    SourceMinimum=-29
    TargetMaximum=20
    TargetMinimum=1
    bClamp=false
    Value=&RigVMFunction_MathQuaternionToEuler(
     RotationOrder=ZYX
     Value=&RigUnit_GetTransform(Item=(Type=Bone,Name="extra_tentacle_front_r_B_02"), Space=LocalSpace, bInitial=False).Transform.Rotation
    ).Result.Z
   )
   > &RigUnit_ModifyTransforms(
    ItemToModify=((Item=(Type=Bone,Name="extra_tentacle_front_r_B_02"),Transform=(Rotation=(X=0,Y=0,Z=0,W=1),Translation=(X=0,Y=0,Z=0),Scale3D=(X=1,Y=1,Z=1))))
    Mode=OverrideLocal
    WeightMinimum=0
    Weight=@n13.Result
    ItemToModify.0.Transform.Rotation=&RigVMFunction_MathQuaternionFromEuler(Euler=(X=0,Y=0,Z=0), RotationOrder=XYZ, Euler.Y=@n14.Result, Euler.Z=@n14.Result).Result
   )
   n15 = &RigVMFunction_MathQuaternionToEuler(
    RotationOrder=XYZ
    Value=&RigUnit_GetTransform(Item=(Type=Bone,Name="extra_tentacle_front_r_B_02"), Space=LocalSpace, bInitial=False).Transform.Rotation
   )
   n16 = &RigUnit_SpringInterpQuaternionV2(
    CriticalDamping=0.1
    Current=(X=0,Y=0,Z=0,W=1)
    Strength=1
    TargetVelocityAmount=0
    Torque=(X=0,Y=0,Z=0)
    bInitializeFromTarget=False
    bUseCurrentInput=False
    Target=&RigVMFunction_TimeOffsetTransform(
     BufferSize=16
     SecondsAgo=0.6
     TimeRange=1
     Value=(Rotation=(X=0,Y=0,Z=0,W=1),Translation=(X=0,Y=0,Z=0),Scale3D=(X=1,Y=1,Z=1))
     Value.Rotation=&RigVMFunction_MathQuaternionFromEuler(Euler=(X=0,Y=0,Z=0), RotationOrder=XYZ, Euler.Y=@n15.Result.Y, Euler.Z=@n15.Result.Z).Result
    ) {template="TimeOffset::Execute(in Value,in SecondsAgo,in BufferSize,in TimeRange,out Result)"}.Result.Rotation
   ) {template="SpringInterp::Execute(in Target,in Strength,in CriticalDamping,in Torque,in bUseCurrentInput,in Current,in TargetVelocityAmount,in bInitializeFromTarget,out Result,out AngularVelocity)"}
   > &RigUnit_ModifyTransforms(
    ItemToModify=((Item=(Type=Bone,Name="BASE_tentacle_front_r_B_03"),Transform=(Rotation=(X=0,Y=0,Z=0,W=1),Translation=(X=0,Y=0,Z=0),Scale3D=(X=1,Y=1,Z=1))))
    WeightMinimum=0
    Weight=@n13.Result
    ItemToModify.0.Transform.Rotation=@n16.Result
   )
   n17 = &RigVMFunction_MathQuaternionToEuler(
    RotationOrder=XYZ
    Value=&RigUnit_GetTransform(Item=(Type=Bone,Name="BASE_tentacle_front_r_B_03"), Space=LocalSpace, bInitial=False).Transform.Rotation
   )
   n18 = &RigVMFunction_MathDoubleMul(B=1.5, A=@n17.Result.Y)
   n19 = &RigUnit_SpringInterpQuaternionV2(
    CriticalDamping=1
    Current=(X=0,Y=0,Z=0,W=1)
    Strength=6
    TargetVelocityAmount=0
    Torque=(X=0,Y=0,Z=0)
    bInitializeFromTarget=False
    bUseCurrentInput=False
    Target=&RigVMFunction_TimeOffsetTransform(
     BufferSize=16
     SecondsAgo=0.1
     TimeRange=1
     Value=(Rotation=(X=0,Y=0,Z=0,W=1),Translation=(X=0,Y=0,Z=0),Scale3D=(X=1,Y=1,Z=1))
     Value.Rotation=&RigVMFunction_MathQuaternionFromEuler(Euler=(X=0,Y=0,Z=0), RotationOrder=XYZ, Euler.Y=@n18.Result, Euler.Z=@n17.Result.Z).Result
    ) {template="TimeOffset::Execute(in Value,in SecondsAgo,in BufferSize,in TimeRange,out Result)"}.Result.Rotation
   ) {template="SpringInterp::Execute(in Target,in Strength,in CriticalDamping,in Torque,in bUseCurrentInput,in Current,in TargetVelocityAmount,in bInitializeFromTarget,out Result,out AngularVelocity)"}
   > &RigUnit_ModifyTransforms(
    ItemToModify=((Item=(Type=Bone,Name="BASE_tentacle_front_r_B_04"),Transform=(Rotation=(X=0,Y=0,Z=0,W=1),Translation=(X=0,Y=0,Z=0),Scale3D=(X=1,Y=1,Z=1))))
    WeightMinimum=0
    Weight=@n13.Result
    ItemToModify.0.Transform.Rotation=@n19.Result
   )
   n20 = &RigVMFunction_MathDoubleMul(
    B=2
    A=&RigVMFunction_MathQuaternionToEuler(
     RotationOrder=XYZ
     Value=&RigUnit_GetTransform(Item=(Type=Bone,Name="BASE_tentacle_front_r_B_04"), Space=LocalSpace, bInitial=False).Transform.Rotation
    ).Result.Y
   )
   n21 = &RigUnit_SpringInterpQuaternionV2(
    CriticalDamping=1
    Current=(X=0,Y=0,Z=0,W=1)
    Strength=6
    TargetVelocityAmount=0
    Torque=(X=0,Y=0,Z=0)
    bInitializeFromTarget=False
    bUseCurrentInput=False
    Target=&RigVMFunction_TimeOffsetTransform(
     BufferSize=16
     SecondsAgo=0.7
     TimeRange=1
     Value=(Rotation=(X=0,Y=0,Z=0,W=1),Translation=(X=0,Y=0,Z=0),Scale3D=(X=1,Y=1,Z=1))
     Value.Rotation=&RigVMFunction_MathQuaternionFromEuler(Euler=(X=0,Y=0,Z=0), RotationOrder=XYZ, Euler.Y=@n20.Result).Result
    ) {template="TimeOffset::Execute(in Value,in SecondsAgo,in BufferSize,in TimeRange,out Result)"}.Result.Rotation
   ) {template="SpringInterp::Execute(in Target,in Strength,in CriticalDamping,in Torque,in bUseCurrentInput,in Current,in TargetVelocityAmount,in bInitializeFromTarget,out Result,out AngularVelocity)"}
   > &RigUnit_ModifyTransforms(
    ItemToModify=((Item=(Type=Bone,Name="BASE_tentacle_front_r_B_05"),Transform=(Rotation=(X=0,Y=0,Z=0,W=1),Translation=(X=0,Y=0,Z=0),Scale3D=(X=1,Y=1,Z=1))))
    WeightMinimum=0
    Weight=@n13.Result
    ItemToModify.0.Transform.Rotation=@n21.Result
   )
   n22 = &RigVMFunction_MathDoubleMul(A=RigVMVariableNode() {variable=play_f}.Value, B=RigVMVariableNode() {variable=sp_R}.Value)
   > &RigUnit_ItemHarmonics(
    RotationOrder=YZX
    Targets=((Item=(Type=Bone,Name="extra_tentacle_front_r_A_02"),Ratio=1))
    WaveAmplitude=(X=0,Y=0,Z=30)
    WaveEase=Linear
    WaveFrequency=(X=1,Y=0.6,Z=0.6)
    WaveMinimum=0.5
    WaveNoise=(X=0,Y=0,Z=0)
    WaveOffset=(X=0,Y=0,Z=1)
    WaveSpeed=(X=3,Y=2,Z=2)
    WaveMaximum=@n22.Result
   )
   n23 = &RigVMFunction_MathQuaternionToEuler(
    RotationOrder=ZYX
    Value=&RigUnit_GetTransform(Item=(Type=Bone,Name="extra_tentacle_front_r_A_02"), Space=LocalSpace, bInitial=False).Transform.Rotation
   )
   > &RigUnit_ModifyTransforms(
    ItemToModify=((Item=(Type=Bone,Name="extra_tentacle_front_r_A_02"),Transform=(Rotation=(X=0,Y=0,Z=0,W=1),Translation=(X=0,Y=0,Z=0),Scale3D=(X=1,Y=1,Z=1))))
    Mode=OverrideLocal
    WeightMinimum=0
    Weight=@n22.Result
    ItemToModify.0.Transform.Rotation=&RigVMFunction_MathQuaternionFromEuler(
     Euler=(X=0,Y=0,Z=0)
     RotationOrder=XYZ
     Euler.Y=&RigVMFunction_MathDoubleRemap(SourceMaximum=29, SourceMinimum=-29, TargetMaximum=20, TargetMinimum=1, bClamp=false, Value=@n23.Result.Z).Result
    ).Result
   )
   n24 = &RigVMFunction_MathQuaternionToEuler(
    RotationOrder=XYZ
    Value=&RigUnit_GetTransform(Item=(Type=Bone,Name="extra_tentacle_front_r_A_02"), Space=LocalSpace, bInitial=False).Transform.Rotation
   )
   n25 = &RigUnit_SpringInterpQuaternionV2(
    CriticalDamping=0.1
    Current=(X=0,Y=0,Z=0,W=1)
    Strength=1
    TargetVelocityAmount=0
    Torque=(X=0,Y=0,Z=0)
    bInitializeFromTarget=False
    bUseCurrentInput=False
    Target=&RigVMFunction_TimeOffsetTransform(
     BufferSize=16
     SecondsAgo=0.6
     TimeRange=1
     Value=(Rotation=(X=0,Y=0,Z=0,W=1),Translation=(X=0,Y=0,Z=0),Scale3D=(X=1,Y=1,Z=1))
     Value.Rotation=&RigVMFunction_MathQuaternionFromEuler(Euler=(X=0,Y=0,Z=0), RotationOrder=XYZ, Euler.Y=@n24.Result.Y, Euler.Z=@n24.Result.Z).Result
    ) {template="TimeOffset::Execute(in Value,in SecondsAgo,in BufferSize,in TimeRange,out Result)"}.Result.Rotation
   ) {template="SpringInterp::Execute(in Target,in Strength,in CriticalDamping,in Torque,in bUseCurrentInput,in Current,in TargetVelocityAmount,in bInitializeFromTarget,out Result,out AngularVelocity)"}
   > &RigUnit_ModifyTransforms(
    ItemToModify=((Item=(Type=Bone,Name="BASE_tentacle_front_r_A_03"),Transform=(Rotation=(X=0,Y=0,Z=0,W=1),Translation=(X=0,Y=0,Z=0),Scale3D=(X=1,Y=1,Z=1))))
    WeightMinimum=0
    Weight=@n22.Result
    ItemToModify.0.Transform.Rotation=@n25.Result
   )
   n26 = &RigVMFunction_MathQuaternionToEuler(
    RotationOrder=XYZ
    Value=&RigUnit_GetTransform(Item=(Type=Bone,Name="BASE_tentacle_front_r_A_03"), Space=LocalSpace, bInitial=False).Transform.Rotation
   )
   n27 = &RigVMFunction_MathDoubleMul(B=1.5, A=@n26.Result.Y)
   n28 = &RigUnit_SpringInterpQuaternionV2(
    CriticalDamping=1
    Current=(X=0,Y=0,Z=0,W=1)
    Strength=6
    TargetVelocityAmount=0
    Torque=(X=0,Y=0,Z=0)
    bInitializeFromTarget=False
    bUseCurrentInput=False
    Target=&RigVMFunction_TimeOffsetTransform(
     BufferSize=16
     SecondsAgo=0.1
     TimeRange=1
     Value=(Rotation=(X=0,Y=0,Z=0,W=1),Translation=(X=0,Y=0,Z=0),Scale3D=(X=1,Y=1,Z=1))
     Value.Rotation=&RigVMFunction_MathQuaternionFromEuler(Euler=(X=0,Y=0,Z=0), RotationOrder=XYZ, Euler.Y=@n27.Result, Euler.Z=@n26.Result.Z).Result
    ) {template="TimeOffset::Execute(in Value,in SecondsAgo,in BufferSize,in TimeRange,out Result)"}.Result.Rotation
   ) {template="SpringInterp::Execute(in Target,in Strength,in CriticalDamping,in Torque,in bUseCurrentInput,in Current,in TargetVelocityAmount,in bInitializeFromTarget,out Result,out AngularVelocity)"}
   > &RigUnit_ModifyTransforms(
    ItemToModify=((Item=(Type=Bone,Name="BASE_tentacle_front_r_A_04"),Transform=(Rotation=(X=0,Y=0,Z=0,W=1),Translation=(X=0,Y=0,Z=0),Scale3D=(X=1,Y=1,Z=1))))
    WeightMinimum=0
    Weight=@n22.Result
    ItemToModify.0.Transform.Rotation=@n28.Result
   )
   n29 = &RigVMFunction_MathDoubleMul(
    B=2
    A=&RigVMFunction_MathQuaternionToEuler(
     RotationOrder=XYZ
     Value=&RigUnit_GetTransform(Item=(Type=Bone,Name="BASE_tentacle_front_r_A_04"), Space=LocalSpace, bInitial=False).Transform.Rotation
    ).Result.Y
   )
   n30 = &RigUnit_SpringInterpQuaternionV2(
    CriticalDamping=1
    Current=(X=0,Y=0,Z=0,W=1)
    Strength=6
    TargetVelocityAmount=0
    Torque=(X=0,Y=0,Z=0)
    bInitializeFromTarget=False
    bUseCurrentInput=False
    Target=&RigVMFunction_TimeOffsetTransform(
     BufferSize=16
     SecondsAgo=0.7
     TimeRange=1
     Value=(Rotation=(X=0,Y=0,Z=0,W=1),Translation=(X=0,Y=0,Z=0),Scale3D=(X=1,Y=1,Z=1))
     Value.Rotation=&RigVMFunction_MathQuaternionFromEuler(Euler=(X=0,Y=0,Z=0), RotationOrder=XYZ, Euler.Y=@n29.Result).Result
    ) {template="TimeOffset::Execute(in Value,in SecondsAgo,in BufferSize,in TimeRange,out Result)"}.Result.Rotation
   ) {template="SpringInterp::Execute(in Target,in Strength,in CriticalDamping,in Torque,in bUseCurrentInput,in Current,in TargetVelocityAmount,in bInitializeFromTarget,out Result,out AngularVelocity)"}
   > &RigUnit_ModifyTransforms(
    ItemToModify=((Item=(Type=Bone,Name="BASE_tentacle_front_r_A_05"),Transform=(Rotation=(X=0,Y=0,Z=0,W=1),Translation=(X=0,Y=0,Z=0),Scale3D=(X=1,Y=1,Z=1))))
    WeightMinimum=0
    Weight=@n22.Result
    ItemToModify.0.Transform.Rotation=@n30.Result
   )
   n31 = &RigVMFunction_MathDoubleMul(A=RigVMVariableNode() {variable=play_f}.Value, B=RigVMVariableNode() {variable=sp_L}.Value)
   > &RigUnit_ItemHarmonics(
    RotationOrder=YZX
    Targets=((Item=(Type=Bone,Name="extra_tentacle_front_l_A_02"),Ratio=1))
    WaveAmplitude=(X=0,Y=0,Z=30)
    WaveEase=Linear
    WaveFrequency=(X=1,Y=0.6,Z=0.6)
    WaveMinimum=0.5
    WaveNoise=(X=0,Y=0,Z=0)
    WaveOffset=(X=0,Y=0,Z=1)
    WaveSpeed=(X=3,Y=2,Z=2)
    WaveMaximum=@n31.Result
   )
   n32 = &RigVMFunction_MathQuaternionToEuler(
    RotationOrder=ZYX
    Value=&RigUnit_GetTransform(Item=(Type=Bone,Name="extra_tentacle_front_l_A_02"), Space=LocalSpace, bInitial=False).Transform.Rotation
   )
   n33 = &RigVMFunction_MathDoubleRemap(SourceMaximum=-29, SourceMinimum=29, TargetMaximum=-20, TargetMinimum=0, bClamp=false, Value=@n32.Result.Z)
   > &RigUnit_ModifyTransforms(
    ItemToModify=((Item=(Type=Bone,Name="extra_tentacle_front_l_A_02"),Transform=(Rotation=(X=0,Y=0,Z=0,W=1),Translation=(X=0,Y=0,Z=0),Scale3D=(X=1,Y=1,Z=1))))
    Mode=OverrideLocal
    WeightMinimum=0
    Weight=@n31.Result
    ItemToModify.0.Transform.Rotation=&RigVMFunction_MathQuaternionFromEuler(Euler=(X=0,Y=0,Z=0), RotationOrder=XYZ, Euler.Y=@n33.Result, Euler.Z=@n33.Result).Result
   )
   n34 = &RigVMFunction_MathQuaternionToEuler(
    RotationOrder=XYZ
    Value=&RigUnit_GetTransform(Item=(Type=Bone,Name="extra_tentacle_front_l_A_02"), Space=LocalSpace, bInitial=False).Transform.Rotation
   )
   n35 = &RigUnit_SpringInterpQuaternionV2(
    CriticalDamping=0.1
    Current=(X=0,Y=0,Z=0,W=1)
    Strength=1
    TargetVelocityAmount=0
    Torque=(X=0,Y=0,Z=0)
    bInitializeFromTarget=False
    bUseCurrentInput=False
    Target=&RigVMFunction_TimeOffsetTransform(
     BufferSize=16
     SecondsAgo=0.6
     TimeRange=1
     Value=(Rotation=(X=0,Y=0,Z=0,W=1),Translation=(X=0,Y=0,Z=0),Scale3D=(X=1,Y=1,Z=1))
     Value.Rotation=&RigVMFunction_MathQuaternionFromEuler(Euler=(X=0,Y=0,Z=0), RotationOrder=XYZ, Euler.Y=@n34.Result.Y).Result
    ) {template="TimeOffset::Execute(in Value,in SecondsAgo,in BufferSize,in TimeRange,out Result)"}.Result.Rotation
   ) {template="SpringInterp::Execute(in Target,in Strength,in CriticalDamping,in Torque,in bUseCurrentInput,in Current,in TargetVelocityAmount,in bInitializeFromTarget,out Result,out AngularVelocity)"}
   > &RigUnit_ModifyTransforms(
    ItemToModify=((Item=(Type=Bone,Name="BASE_tentacle_front_l_A_03"),Transform=(Rotation=(X=0,Y=0,Z=0,W=1),Translation=(X=0,Y=0,Z=0),Scale3D=(X=1,Y=1,Z=1))))
    WeightMinimum=0
    Weight=@n31.Result
    ItemToModify.0.Transform.Rotation=@n35.Result
   )
   n36 = &RigVMFunction_MathDoubleMul(
    B=1.5
    A=&RigVMFunction_MathQuaternionToEuler(
     RotationOrder=XYZ
     Value=&RigUnit_GetTransform(Item=(Type=Bone,Name="BASE_tentacle_front_l_A_03"), Space=LocalSpace, bInitial=False).Transform.Rotation
    ).Result.Y
   )
   n37 = &RigUnit_SpringInterpQuaternionV2(
    CriticalDamping=1
    Current=(X=0,Y=0,Z=0,W=1)
    Strength=6
    TargetVelocityAmount=0
    Torque=(X=0,Y=0,Z=0)
    bInitializeFromTarget=False
    bUseCurrentInput=False
    Target=&RigVMFunction_TimeOffsetTransform(
     BufferSize=16
     SecondsAgo=0.1
     TimeRange=1
     Value=(Rotation=(X=0,Y=0,Z=0,W=1),Translation=(X=0,Y=0,Z=0),Scale3D=(X=1,Y=1,Z=1))
     Value.Rotation=&RigVMFunction_MathQuaternionFromEuler(Euler=(X=0,Y=0,Z=0), RotationOrder=XYZ, Euler.Y=@n36.Result).Result
    ) {template="TimeOffset::Execute(in Value,in SecondsAgo,in BufferSize,in TimeRange,out Result)"}.Result.Rotation
   ) {template="SpringInterp::Execute(in Target,in Strength,in CriticalDamping,in Torque,in bUseCurrentInput,in Current,in TargetVelocityAmount,in bInitializeFromTarget,out Result,out AngularVelocity)"}
   > &RigUnit_ModifyTransforms(
    ItemToModify=((Item=(Type=Bone,Name="BASE_tentacle_front_l_A_04"),Transform=(Rotation=(X=0,Y=0,Z=0,W=1),Translation=(X=0,Y=0,Z=0),Scale3D=(X=1,Y=1,Z=1))))
    WeightMinimum=0
    Weight=@n31.Result
    ItemToModify.0.Transform.Rotation=@n37.Result
   )
   n38 = &RigVMFunction_MathDoubleMul(
    B=2
    A=&RigVMFunction_MathQuaternionToEuler(
     RotationOrder=XYZ
     Value=&RigUnit_GetTransform(Item=(Type=Bone,Name="BASE_tentacle_front_l_A_04"), Space=LocalSpace, bInitial=False).Transform.Rotation
    ).Result.Y
   )
   n39 = &RigUnit_SpringInterpQuaternionV2(
    CriticalDamping=1
    Current=(X=0,Y=0,Z=0,W=1)
    Strength=6
    TargetVelocityAmount=0
    Torque=(X=0,Y=0,Z=0)
    bInitializeFromTarget=False
    bUseCurrentInput=False
    Target=&RigVMFunction_TimeOffsetTransform(
     BufferSize=16
     SecondsAgo=0.7
     TimeRange=1
     Value=(Rotation=(X=0,Y=0,Z=0,W=1),Translation=(X=0,Y=0,Z=0),Scale3D=(X=1,Y=1,Z=1))
     Value.Rotation=&RigVMFunction_MathQuaternionFromEuler(Euler=(X=0,Y=0,Z=0), RotationOrder=XYZ, Euler.Y=@n38.Result).Result
    ) {template="TimeOffset::Execute(in Value,in SecondsAgo,in BufferSize,in TimeRange,out Result)"}.Result.Rotation
   ) {template="SpringInterp::Execute(in Target,in Strength,in CriticalDamping,in Torque,in bUseCurrentInput,in Current,in TargetVelocityAmount,in bInitializeFromTarget,out Result,out AngularVelocity)"}
   > &RigUnit_ModifyTransforms(
    ItemToModify=((Item=(Type=Bone,Name="BASE_tentacle_front_l_A_05"),Transform=(Rotation=(X=0,Y=0,Z=0,W=1),Translation=(X=0,Y=0,Z=0),Scale3D=(X=1,Y=1,Z=1))))
    WeightMinimum=0
    Weight=@n31.Result
    ItemToModify.0.Transform.Rotation=@n39.Result
   )
   RigVMFunction_ControlFlowBranch = &RigVMFunction_ControlFlowBranch(ExecuteContext=@RigVMFunction_Sequence.C, Condition=RigVMVariableNode() {variable=backpack_cape_mod_bool}.Value)
   n40 = &RigUnit_SphericalPoseReader(
    ActiveRegionScaleFactors=(PositiveWidth=0.5,NegativeWidth=0.5,PositiveHeight=0.5,NegativeHeight=0.5)
    ActiveRegionSize=0.5
    Debug=(bDrawDebug=True,bDraw2D=false,bDrawLocalAxes=false,DebugScale=30,DebugSegments=30,DebugThickness=0.2)
    DriverAxis=(X=0,Y=1,Z=0)
    DriverItem=(Type=Bone,Name="head")
    FalloffRegionScaleFactors=(PositiveWidth=0.8,NegativeWidth=0.8,PositiveHeight=0.8,NegativeHeight=0.8)
    FalloffSize=0.5
    OptionalParentItem=(Type=Bone,Name="spine_05")
    RotationOffset=(X=0,Y=0,Z=0)
    ExecutePin=@RigVMFunction_ControlFlowBranch.False
   )
   > RigVMVariableNode(Value=@n40.OutputParam) {variable=scanning_r}
   > n41 = &RigUnit_SphericalPoseReader(
    ActiveRegionScaleFactors=(PositiveWidth=0.5,NegativeWidth=0.5,PositiveHeight=0.5,NegativeHeight=0.5)
    ActiveRegionSize=0.5
    Debug=(bDrawDebug=True,bDraw2D=false,bDrawLocalAxes=false,DebugScale=30,DebugSegments=30,DebugThickness=0.2)
    DriverAxis=(X=0,Y=1,Z=0)
    DriverItem=(Type=Bone,Name="head")
    FalloffRegionScaleFactors=(PositiveWidth=0.8,NegativeWidth=0.8,PositiveHeight=0.8,NegativeHeight=0.8)
    FalloffSize=0.5
    OptionalParentItem=(Type=Bone,Name="spine_05")
    RotationOffset=(X=180,Y=0,Z=0)
   )
   > RigVMVariableNode(Value=@n41.OutputParam) {variable=scanning_l}
   > RigVMVariableNode(Value=1) {variable=l_constraint_float}
   > RigVMVariableNode(Value=0.4) {variable=r_constraint_float}
   n42 = RigVMVariableNode() {variable=scanning_r}
   n43 = RigVMVariableNode() {variable=scanning_l}
   n44 = &RigVMFunction_MathDoubleMax(A=@n42.Value, B=@n43.Value)
   > &RigUnit_ParentConstraint(
    AdvancedSettings=(RotationOrderForFilter=XZY)
    Child=(Type=Bone,Name="tentacle_back_l_A_02")
    Parents="((Item=(Type=Bone,Name=\"head\"),Weight=0.0025),(Item=(Type=Bone,Name=\"spine_03\"),Weight=0.0283),(Item=(Type=Bone,Name=\"neck_02\"),Weight=0.0353),(Item=(Type=Bone,Name=\"neck_01\"),Weight=0.0639),(Item=(Type=Bone,Name=\"clavicle_l\"),Weight=0.2427),(Item=(Type=Bo ...(truncated)"
    Weight=&RigVMFunction_MathDoubleRemap(SourceMaximum=1, SourceMinimum=0, TargetMaximum=0.2, TargetMinimum=1, bClamp=False, Value=@n44.Result).Result
   )
   Greater = /Script/RigVM.RigVMFunction_MathDoubleGreater(A=@n42.Value, B=@n43.Value) {is=struct RigVMUnitNode, template="Greater::Execute(in A,in B,out Result)"}
   n45 = &RigVMFunction_MathDoubleRemap(
    SourceMaximum=1
    SourceMinimum=0
    TargetMinimum=1
    bClamp=False
    TargetMaximum=RigVMDispatchNode(
     Condition=@Greater.Result
     True=RigVMVariableNode() {variable=l_constraint_float}.Value
     False=RigVMVariableNode() {variable=r_constraint_float}.Value
    ) {template="DISPATCH_RigVMDispatch_If(in Condition,in True,in False,out Result)"}.Result
    Value=@n44.Result
   )
   > &RigUnit_ParentConstraint(
    AdvancedSettings=(RotationOrderForFilter=XZY)
    Child=(Type=Bone,Name="tentacle_back_l_A_03")
    Parents="((Item=(Type=Bone,Name=\"bag_back_03\"),Weight=0.0001),(Item=(Type=Bone,Name=\"bag_front_02\"),Weight=0.0002),(Item=(Type=Bone,Name=\"bag_back_02\"),Weight=0.0023),(Item=(Type=Bone,Name=\"bag_front_01\"),Weight=0.0063),(Item=(Type=Bone,Name=\"bag_back_01\"),Weight=0 ...(truncated)"
    Weight=@n45.Result
   )
   > &RigUnit_ParentConstraint(
    AdvancedSettings=(RotationOrderForFilter=XZY)
    Child=(Type=Bone,Name="tentacle_back_l_A_04")
    Parents="((Item=(Type=Bone,Name=\"bag_front_02\"),Weight=0.0001),(Item=(Type=Bone,Name=\"bag_back_03\"),Weight=0.0002),(Item=(Type=Bone,Name=\"bag_front_01\"),Weight=0.0029),(Item=(Type=Bone,Name=\"bag_back_02\"),Weight=0.0073),(Item=(Type=Bone,Name=\"bag_back_01\"),Weight=0 ...(truncated)"
    Weight=@n45.Result
   )
   > &RigUnit_ParentConstraint(
    AdvancedSettings=(RotationOrderForFilter=XZY)
    Child=(Type=Bone,Name="extra_tentacle_back_l_A_05")
    Parents="((Item=(Type=Bone,Name=\"bag_back_04\"),Weight=0.0004),(Item=(Type=Bone,Name=\"bag_front_01\"),Weight=0.0004),(Item=(Type=Bone,Name=\"bag_back_03\"),Weight=0.0092),(Item=(Type=Bone,Name=\"bag_back_02\"),Weight=0.1812),(Item=(Type=Bone,Name=\"bag_root\"),Weight=0.193 ...(truncated)"
    Weight=1
   )
   > &RigUnit_ParentConstraint(
    AdvancedSettings=(RotationOrderForFilter=XZY)
    Child=(Type=Bone,Name="extra_tentacle_back_l_A_06")
    Parents="((Item=(Type=Bone,Name=\"bag_front_01\"),Weight=0.0002),(Item=(Type=Bone,Name=\"bag_back_04\"),Weight=0.0023),(Item=(Type=Bone,Name=\"bag_back_03\"),Weight=0.0496),(Item=(Type=Bone,Name=\"bag_root\"),Weight=0.0587),(Item=(Type=Bone,Name=\"bag_back_02\"),Weight=0.437 ...(truncated)"
    Weight=1
   )
   > &RigUnit_ModifyTransforms(
    ItemToModify="((Item=(Type=Bone,Name=\"tentacle_back_l_A_04\"),Transform=(Rotation=(X=0,Y=0,Z=0,W=1),Translation=(X=0,Y=0,Z=0),Scale3D=(X=1,Y=1,Z=1))),(Item=(Type=Bone,Name=\"extra_tentacle_back_l_A_05\"),Transform=(Rotation=(X=0,Y=0,Z=0,W=1),Translation=(X=0,Y=0,Z=0),Scale ...(truncated)"
    Weight=1
    WeightMinimum=0
    ItemToModify.0.Transform.Rotation=&RigVMFunction_MathQuaternionFromEuler(Euler=(X=0,Y=15,Z=0)).Result
    ItemToModify.1.Transform.Rotation=&RigVMFunction_MathQuaternionFromEuler(Euler=(X=0,Y=10,Z=0)).Result
    ItemToModify.2.Transform.Rotation=&RigVMFunction_MathQuaternionFromEuler(Euler=(X=0,Y=0,Z=0)).Result
   )
   n46 = &RigVMFunction_MathDoubleMax(A=RigVMVariableNode() {variable=scanning_r}.Value, B=RigVMVariableNode() {variable=scanning_l}.Value)
   > &RigUnit_ParentConstraint(
    AdvancedSettings=(RotationOrderForFilter=XZY)
    Child=(Type=Bone,Name="tentacle_back_l_B_02")
    Parents=((Item=(Type=Bone,Name="spine_03"),Weight=0.0393),(Item=(Type=Bone,Name="neck_02"),Weight=0.0689),(Item=(Type=Bone,Name="clavicle_l"),Weight=0.0933),(Item=(Type=Bone,Name="neck_01"),Weight=0.1019),(Item=(Type=Bone,Name="spine_05"),Weight=0.6966))
    Weight=&RigVMFunction_MathDoubleRemap(SourceMaximum=1, SourceMinimum=0, TargetMaximum=0.2, TargetMinimum=1, bClamp=False, Value=@n46.Result).Result
   )
   n47 = &RigVMFunction_MathDoubleRemap(SourceMaximum=1, SourceMinimum=0, TargetMaximum=0.4, TargetMinimum=1, bClamp=False, Value=@n46.Result)
   > &RigUnit_ParentConstraint(
    AdvancedSettings=(RotationOrderForFilter=XZY)
    Child=(Type=Bone,Name="tentacle_back_l_B_03")
    Parents=((Item=(Type=Bone,Name="spine_03"),Weight=0.0462),(Item=(Type=Bone,Name="neck_02"),Weight=0.0657),(Item=(Type=Bone,Name="clavicle_l"),Weight=0.0706),(Item=(Type=Bone,Name="neck_01"),Weight=0.1022),(Item=(Type=Bone,Name="spine_05"),Weight=0.7154))
    Weight=@n47.Result
   )
   > &RigUnit_ParentConstraint(
    AdvancedSettings=(RotationOrderForFilter=XZY)
    Child=(Type=Bone,Name="tentacle_back_l_B_04")
    Parents=((Item=(Type=Bone,Name="neck_01"),Weight=0.0127),(Item=(Type=Bone,Name="spine_03"),Weight=0.0139),(Item=(Type=Bone,Name="clavicle_l"),Weight=0.3225),(Item=(Type=Bone,Name="spine_05"),Weight=0.6509))
    Weight=@n47.Result
   )
   > &RigUnit_ParentConstraint(
    AdvancedSettings=(RotationOrderForFilter=XZY)
    Child=(Type=Bone,Name="tentacle_back_l_B_05")
    Parents="((Item=(Type=Bone,Name=\"bag_front_01\"),Weight=0.0003),(Item=(Type=Bone,Name=\"bag_back_04\"),Weight=0.0005),(Item=(Type=Bone,Name=\"bag_back_03\"),Weight=0.0131),(Item=(Type=Bone,Name=\"bag_root\"),Weight=0.1284),(Item=(Type=Bone,Name=\"bag_back_02\"),Weight=0.249 ...(truncated)"
    Weight=1
   )
   > &RigUnit_ParentConstraint(
    AdvancedSettings=(RotationOrderForFilter=XZY)
    Child=(Type=Bone,Name="BASE_tentacle_back_l_B_06")
    Parents="((Item=(Type=Bone,Name=\"bag_front_01\"),Weight=0.0003),(Item=(Type=Bone,Name=\"bag_back_04\"),Weight=0.0005),(Item=(Type=Bone,Name=\"bag_back_03\"),Weight=0.0131),(Item=(Type=Bone,Name=\"bag_root\"),Weight=0.1284),(Item=(Type=Bone,Name=\"bag_back_02\"),Weight=0.249 ...(truncated)"
    Weight=1
   )
   > &RigUnit_ParentConstraint(
    AdvancedSettings=(RotationOrderForFilter=XZY)
    Child=(Type=Bone,Name="BASE_tentacle_back_l_B_07")
    Parents="((Item=(Type=Bone,Name=\"bag_front_01\"),Weight=0.0002),(Item=(Type=Bone,Name=\"bag_back_04\"),Weight=0.0023),(Item=(Type=Bone,Name=\"bag_back_03\"),Weight=0.0496),(Item=(Type=Bone,Name=\"bag_root\"),Weight=0.0587),(Item=(Type=Bone,Name=\"bag_back_02\"),Weight=0.437 ...(truncated)"
    Weight=1
   )
   > &RigUnit_ParentConstraint(
    AdvancedSettings=(RotationOrderForFilter=XZY)
    Child=(Type=Bone,Name="BASE_tentacle_back_l_B_08")
    Parents="((Item=(Type=Bone,Name=\"bag_front_01\"),Weight=0.0002),(Item=(Type=Bone,Name=\"bag_back_04\"),Weight=0.0023),(Item=(Type=Bone,Name=\"bag_back_03\"),Weight=0.0496),(Item=(Type=Bone,Name=\"bag_root\"),Weight=0.0587),(Item=(Type=Bone,Name=\"bag_back_02\"),Weight=0.437 ...(truncated)"
    Weight=1
   )
   > &RigUnit_ModifyTransforms(
    ItemToModify="((Item=(Type=Bone,Name=\"tentacle_back_l_B_05\"),Transform=(Rotation=(X=0,Y=0,Z=0,W=1),Translation=(X=0,Y=0,Z=0),Scale3D=(X=1,Y=1,Z=1))),(Item=(Type=Bone,Name=\"BASE_tentacle_back_l_B_06\"),Transform=(Rotation=(X=0,Y=0,Z=0,W=1),Translation=(X=0,Y=0,Z=0),Scale3 ...(truncated)"
    Weight=1
    WeightMinimum=0
    ItemToModify.0.Transform.Rotation=&RigVMFunction_MathQuaternionFromEuler(Euler=(X=0,Y=50,Z=0)).Result
    ItemToModify.1.Transform.Rotation=&RigVMFunction_MathQuaternionFromEuler(Euler=(X=0,Y=-20,Z=0)).Result
    ItemToModify.2.Transform.Rotation=&RigVMFunction_MathQuaternionFromEuler(Euler=(X=0,Y=-40,Z=0)).Result
    ItemToModify.3.Transform.Rotation=&RigVMFunction_MathQuaternionFromEuler(Euler=(X=0,Y=30,Z=0)).Result
   )
   n48 = &RigVMFunction_MathDoubleMax(A=RigVMVariableNode() {variable=scanning_r}.Value, B=RigVMVariableNode() {variable=scanning_l}.Value)
   > &RigUnit_ParentConstraint(
    AdvancedSettings=(RotationOrderForFilter=XZY)
    Child=(Type=Bone,Name="tentacle_back_r_B_02")
    Parents=((Item=(Type=Bone,Name="spine_03"),Weight=0.0462),(Item=(Type=Bone,Name="neck_02"),Weight=0.0658),(Item=(Type=Bone,Name="clavicle_r"),Weight=0.0708),(Item=(Type=Bone,Name="neck_01"),Weight=0.1021),(Item=(Type=Bone,Name="spine_05"),Weight=0.7152))
    Weight=&RigVMFunction_MathDoubleRemap(SourceMaximum=1, SourceMinimum=0, TargetMaximum=0.2, TargetMinimum=1, bClamp=False, Value=@n48.Result).Result
   )
   n49 = &RigVMFunction_MathDoubleRemap(SourceMaximum=1, SourceMinimum=0, TargetMaximum=0.4, TargetMinimum=1, bClamp=False, Value=@n48.Result)
   > &RigUnit_ParentConstraint(
    AdvancedSettings=(RotationOrderForFilter=XZY)
    Child=(Type=Bone,Name="tentacle_back_r_B_03")
    Parents="((Item=(Type=Bone,Name=\"head\"),Weight=0.0013),(Item=(Type=Bone,Name=\"clavicle_r\"),Weight=0.0414),(Item=(Type=Bone,Name=\"spine_03\"),Weight=0.0552),(Item=(Type=Bone,Name=\"neck_02\"),Weight=0.0632),(Item=(Type=Bone,Name=\"neck_01\"),Weight=0.1009),(Item=(Type=Bo ...(truncated)"
    Weight=@n49.Result
   )
   > &RigUnit_ParentConstraint(
    AdvancedSettings=(RotationOrderForFilter=XZY)
    Child=(Type=Bone,Name="tentacle_back_r_B_04")
    Parents=((Item=(Type=Bone,Name="neck_01"),Weight=0.0108),(Item=(Type=Bone,Name="spine_03"),Weight=0.0145),(Item=(Type=Bone,Name="clavicle_l"),Weight=0.0214),(Item=(Type=Bone,Name="clavicle_r"),Weight=0.1602),(Item=(Type=Bone,Name="spine_05"),Weight=0.7932))
    Weight=@n49.Result
   )
   > &RigUnit_ParentConstraint(
    AdvancedSettings=(RotationOrderForFilter=XZY)
    Child=(Type=Bone,Name="BASE_tentacle_back_r_B_05")
    Parents=((Item=(Type=Bone,Name="clavicle_l"),Weight=0.0637),(Item=(Type=Bone,Name="clavicle_r"),Weight=0.2463),(Item=(Type=Bone,Name="spine_05"),Weight=0.6899))
    Weight=1
   )
   > &RigUnit_ParentConstraint(
    AdvancedSettings=(RotationOrderForFilter=XZY)
    Child=(Type=Bone,Name="BASE_tentacle_back_r_B_06")
    Parents="((Item=(Type=Bone,Name=\"bag_root\"),Weight=0.0021),(Item=(Type=Bone,Name=\"dyn_bag_sim_chain_01\"),Weight=0.0029),(Item=(Type=Bone,Name=\"bag_back_01\"),Weight=0.0568),(Item=(Type=Bone,Name=\"bag_back_04\"),Weight=0.0639),(Item=(Type=Bone,Name=\"bag_back_02\"),Weig ...(truncated)"
    Weight=1
   )
   n50 = &RigVMFunction_MathDoubleMax(A=RigVMVariableNode() {variable=scanning_r}.Value, B=RigVMVariableNode() {variable=scanning_l}.Value)
   > &RigUnit_ParentConstraint(
    AdvancedSettings=(RotationOrderForFilter=XZY)
    Child=(Type=Bone,Name="tentacle_back_r_A_02")
    Parents="((Item=(Type=Bone,Name=\"head\"),Weight=0.0025),(Item=(Type=Bone,Name=\"spine_03\"),Weight=0.0283),(Item=(Type=Bone,Name=\"neck_02\"),Weight=0.0352),(Item=(Type=Bone,Name=\"neck_01\"),Weight=0.064),(Item=(Type=Bone,Name=\"clavicle_r\"),Weight=0.2427),(Item=(Type=Bon ...(truncated)"
    Weight=&RigVMFunction_MathDoubleRemap(SourceMaximum=1, SourceMinimum=0, TargetMaximum=0.2, TargetMinimum=1, bClamp=False, Value=@n50.Result).Result
   )
   n51 = RigVMVariableNode() {variable=scanning_r}
   n52 = RigVMVariableNode() {variable=scanning_l}
   n53 = &RigVMFunction_MathDoubleRemap(
    SourceMaximum=1
    SourceMinimum=0
    TargetMinimum=1
    bClamp=False
    TargetMaximum=RigVMDispatchNode(
     Condition=&RigVMFunction_MathDoubleGreater(B=@n51.Value, A=@n52.Value) {template="Greater::Execute(in A,in B,out Result)"}.Result
     True=RigVMVariableNode() {variable=l_constraint_float}.Value
     False=RigVMVariableNode() {variable=r_constraint_float}.Value
    ) {template="DISPATCH_RigVMDispatch_If(in Condition,in True,in False,out Result)"}.Result
    Value=@n50.Result
   )
   > &RigUnit_ParentConstraint(
    AdvancedSettings=(RotationOrderForFilter=XZY)
    Child=(Type=Bone,Name="tentacle_back_r_A_03")
    Parents=((Item=(Type=Bone,Name="neck_02"),Weight=0.0075),(Item=(Type=Bone,Name="spine_03"),Weight=0.0148),(Item=(Type=Bone,Name="neck_01"),Weight=0.0233),(Item=(Type=Bone,Name="clavicle_r"),Weight=0.4517),(Item=(Type=Bone,Name="spine_05"),Weight=0.5027))
    Weight=@n53.Result
   )
   > &RigUnit_ParentConstraint(
    AdvancedSettings=(RotationOrderForFilter=XZY)
    Child=(Type=Bone,Name="BASE_tentacle_back_r_A_04")
    Parents="((Item=(Type=Bone,Name=\"deform_upperarm_bk_r\"),Weight=0.0195),(Item=(Type=Bone,Name=\"upperarm_twist_01_r\"),Weight=0.0302),(Item=(Type=Bone,Name=\"deform_upperarm_up_r\"),Weight=0.0325),(Item=(Type=Bone,Name=\"spine_05\"),Weight=0.1998),(Item=(Type=Bone,Name=\"c ...(truncated)"
    Weight=@n53.Result
   )
   > &RigUnit_ParentConstraint(
    AdvancedSettings=(RotationOrderForFilter=XZY)
    Child=(Type=Bone,Name="BASE_tentacle_back_r_A_05")
    Parents="((Item=(Type=Bone,Name=\"spine_04\"),Weight=0.0124),(Item=(Type=Bone,Name=\"deform_upperarm_up_r\"),Weight=0.026),(Item=(Type=Bone,Name=\"deform_upperarm_bk_r\"),Weight=0.0369),(Item=(Type=Bone,Name=\"upperarm_twist_01_r\"),Weight=0.0394),(Item=(Type=Bone,Name=\"sp ...(truncated)"
    Weight=@n53.Result
   )
   > &RigUnit_ParentConstraint(
    AdvancedSettings=(RotationOrderForFilter=XZY)
    Child=(Type=Bone,Name="BASE_tentacle_back_r_A_06")
    Parents="((Item=(Type=Bone,Name=\"clavicle_l\"),Weight=0.0105),(Item=(Type=Bone,Name=\"deform_upperarm_bk_r\"),Weight=0.0135),(Item=(Type=Bone,Name=\"upperarm_twist_01_r\"),Weight=0.0241),(Item=(Type=Bone,Name=\"spine_04\"),Weight=0.0406),(Item=(Type=Bone,Name=\"spine_05\"), ...(truncated)"
    Weight=1
   )
   > &RigUnit_ModifyTransforms(
    ItemToModify="((Item=(Type=Bone,Name=\"tentacle_back_r_A_03\"),Transform=(Rotation=(X=0,Y=0,Z=0,W=1),Translation=(X=0,Y=0,Z=0),Scale3D=(X=1,Y=1,Z=1))),(Item=(Type=Bone,Name=\"BASE_tentacle_back_r_A_04\"),Transform=(Rotation=(X=0,Y=0,Z=0,W=1),Translation=(X=0,Y=0,Z=0),Scale3 ...(truncated)"
    Weight=1
    WeightMinimum=0
    ItemToModify.0.Transform.Rotation=&RigVMFunction_MathQuaternionFromEuler(Euler=(X=0,Y=0,Z=0)).Result
    ItemToModify.1.Transform.Rotation=&RigVMFunction_MathQuaternionFromEuler(Euler=(X=0,Y=30,Z=0)).Result
    ItemToModify.2.Transform.Rotation=&RigVMFunction_MathQuaternionFromEuler(Euler=(X=0,Y=0,Z=0)).Result
    ItemToModify.3.Transform.Rotation=&RigVMFunction_MathQuaternionFromEuler(Euler=(X=0,Y=0,Z=0)).Result
   )
   n54 = RigVMVariableNode() {variable=backpack_cape_mod}
   &RigUnit_ItemHarmonics(
    RotationOrder=YZX
    Targets=((Item=(Type=Bone,Name="extra_tentacle_back_l_A_03"),Ratio=1))
    WaveAmplitude=(X=0,Y=0,Z=30)
    WaveEase=Linear
    WaveFrequency=(X=1,Y=0.6,Z=0.6)
    WaveMinimum=0.5
    WaveNoise=(X=0,Y=0,Z=0)
    WaveOffset=(X=0,Y=0,Z=8)
    WaveSpeed=(X=3,Y=2,Z=2)
    WaveMaximum=@n54.Value
    ExecutePin=@RigVMFunction_Sequence.A
   )
   Remap = &RigVMFunction_MathDoubleRemap(
    SourceMaximum=29
    SourceMinimum=-29
    TargetMaximum=29
    TargetMinimum=0
    Value=&RigVMFunction_MathQuaternionToEuler(
     RotationOrder=ZYX
     Value=&RigUnit_GetTransform(Item=(Type=Bone,Name="extra_tentacle_back_l_A_03"), Space=LocalSpace, bInitial=False).Transform.Rotation
    ).Result.Z
   )
   RigVMFunction_MathQuaternionFromEuler = &RigVMFunction_MathQuaternionFromEuler(
    Euler=(X=0,Y=0,Z=0)
    RotationOrder=XYZ
    Euler.Y=&RigVMFunction_MathDoubleMul(B=-1, A=@Remap.Result).Result
    Euler.Z=@Remap.Result
   )
   > &RigUnit_ModifyTransforms(
    ItemToModify=((Item=(Type=Bone,Name="extra_tentacle_back_l_A_03"),Transform=(Rotation=(X=0,Y=0,Z=0,W=1),Translation=(X=0,Y=0,Z=0),Scale3D=(X=1,Y=1,Z=1))))
    Mode=OverrideLocal
    WeightMinimum=0
    ItemToModify.0.Transform.Rotation=@RigVMFunction_MathQuaternionFromEuler.Result
    Weight=@n54.Value
   )
   RigVMFunction_MathQuaternionToEuler = &RigVMFunction_MathQuaternionToEuler(
    RotationOrder=XYZ
    Value=&RigUnit_GetTransform(Item=(Type=Bone,Name="extra_tentacle_back_l_A_03"), Space=LocalSpace, bInitial=False).Transform.Rotation
   )
   TimeOffset = &RigVMFunction_TimeOffsetTransform(
    BufferSize=16
    SecondsAgo=0.6
    TimeRange=1
    Value=(Rotation=(X=0,Y=0,Z=0,W=1),Translation=(X=0,Y=0,Z=0),Scale3D=(X=1,Y=1,Z=1))
    Value.Rotation=&RigVMFunction_MathQuaternionFromEuler(
     Euler=(X=0,Y=0,Z=0)
     RotationOrder=XYZ
     Euler.Y=@RigVMFunction_MathQuaternionToEuler.Result.Y
     Euler.Z=@RigVMFunction_MathQuaternionToEuler.Result.Z
    ).Result
   ) {template="TimeOffset::Execute(in Value,in SecondsAgo,in BufferSize,in TimeRange,out Result)"}
   n55 = &RigUnit_SpringInterpQuaternionV2(
    CriticalDamping=0.1
    Current=(X=0,Y=0,Z=0,W=1)
    Strength=1
    TargetVelocityAmount=0
    Torque=(X=0,Y=0,Z=0)
    bInitializeFromTarget=False
    bUseCurrentInput=False
    Target=@TimeOffset.Result.Rotation
   ) {template="SpringInterp::Execute(in Target,in Strength,in CriticalDamping,in Torque,in bUseCurrentInput,in Current,in TargetVelocityAmount,in bInitializeFromTarget,out Result,out AngularVelocity)"}
   > &RigUnit_ModifyTransforms(
    ItemToModify=((Item=(Type=Bone,Name="extra_tentacle_back_l_A_04"),Transform=(Rotation=(X=0,Y=0,Z=0,W=1),Translation=(X=0,Y=0,Z=0),Scale3D=(X=1,Y=1,Z=1))))
    WeightMinimum=0
    ItemToModify.0.Transform.Rotation=@n55.Result
    Weight=@n54.Value
   )
   n56 = &RigVMFunction_MathQuaternionToEuler(
    RotationOrder=XYZ
    Value=&RigUnit_GetTransform(Item=(Type=Bone,Name="extra_tentacle_back_l_A_04"), Space=LocalSpace, bInitial=False).Transform.Rotation
   )
   n57 = &RigVMFunction_MathDoubleMul(B=2, A=@n56.Result.Z)
   n58 = &RigUnit_SpringInterpQuaternionV2(
    CriticalDamping=1
    Current=(X=0,Y=0,Z=0,W=1)
    Strength=6
    TargetVelocityAmount=0
    Torque=(X=0,Y=0,Z=0)
    bInitializeFromTarget=False
    bUseCurrentInput=False
    Target=&RigVMFunction_TimeOffsetTransform(
     BufferSize=16
     SecondsAgo=0.1
     TimeRange=1
     Value=(Rotation=(X=0,Y=0,Z=0,W=1),Translation=(X=0,Y=0,Z=0),Scale3D=(X=1,Y=1,Z=1))
     Value.Rotation=&RigVMFunction_MathQuaternionFromEuler(Euler=(X=0,Y=0,Z=0), RotationOrder=XYZ, Euler.Z=@n57.Result, Euler.Y=@n56.Result.Y).Result
    ) {template="TimeOffset::Execute(in Value,in SecondsAgo,in BufferSize,in TimeRange,out Result)"}.Result.Rotation
   ) {template="SpringInterp::Execute(in Target,in Strength,in CriticalDamping,in Torque,in bUseCurrentInput,in Current,in TargetVelocityAmount,in bInitializeFromTarget,out Result,out AngularVelocity)"}
   > &RigUnit_ModifyTransforms(
    ItemToModify=((Item=(Type=Bone,Name="extra_tentacle_back_l_A_05"),Transform=(Rotation=(X=0,Y=0,Z=0,W=1),Translation=(X=0,Y=0,Z=0),Scale3D=(X=1,Y=1,Z=1))))
    WeightMinimum=0
    ItemToModify.0.Transform.Rotation=@n58.Result
    Weight=@n54.Value
   )
   n59 = &RigVMFunction_MathQuaternionToEuler(
    RotationOrder=XYZ
    Value=&RigUnit_GetTransform(Item=(Type=Bone,Name="extra_tentacle_back_l_A_05"), Space=LocalSpace, bInitial=False).Transform.Rotation
   )
   n60 = &RigUnit_SpringInterpQuaternionV2(
    CriticalDamping=1
    Current=(X=0,Y=0,Z=0,W=1)
    Strength=6
    TargetVelocityAmount=0
    Torque=(X=0,Y=0,Z=0)
    bInitializeFromTarget=False
    bUseCurrentInput=False
    Target=&RigVMFunction_TimeOffsetTransform(
     BufferSize=16
     SecondsAgo=0.7
     TimeRange=1
     Value=(Rotation=(X=0,Y=0,Z=0,W=1),Translation=(X=0,Y=0,Z=0),Scale3D=(X=1,Y=1,Z=1))
     Value.Rotation=&RigVMFunction_MathQuaternionFromEuler(Euler=(X=0,Y=0,Z=0), RotationOrder=XYZ, Euler.Y=@n59.Result.Y, Euler.Z=@n59.Result.Z).Result
    ) {template="TimeOffset::Execute(in Value,in SecondsAgo,in BufferSize,in TimeRange,out Result)"}.Result.Rotation
   ) {template="SpringInterp::Execute(in Target,in Strength,in CriticalDamping,in Torque,in bUseCurrentInput,in Current,in TargetVelocityAmount,in bInitializeFromTarget,out Result,out AngularVelocity)"}
   > &RigUnit_ModifyTransforms(
    ItemToModify=((Item=(Type=Bone,Name="extra_tentacle_back_l_A_06"),Transform=(Rotation=(X=0,Y=0,Z=0,W=1),Translation=(X=0,Y=0,Z=0),Scale3D=(X=1,Y=1,Z=1))))
    WeightMinimum=0
    ItemToModify.0.Transform.Rotation=@n60.Result
    Weight=@n54.Value
   )
   n61 = RigVMVariableNode() {variable=backpack_cape_mod}
   > &RigUnit_ItemHarmonics(
    RotationOrder=YZX
    Targets=((Item=(Type=Bone,Name="extra_tentacle_back_l_B_05"),Ratio=1))
    WaveAmplitude=(X=0,Y=30,Z=0)
    WaveEase=Linear
    WaveFrequency=(X=1,Y=1,Z=0.6)
    WaveMinimum=0.5
    WaveNoise=(X=0,Y=0,Z=0)
    WaveOffset=(X=0,Y=0,Z=0)
    WaveSpeed=(X=3,Y=2,Z=2)
    WaveMaximum=@n61.Value
   )
   n62 = &RigVMFunction_MathQuaternionToEuler(
    RotationOrder=ZYX
    Value=&RigUnit_GetTransform(Item=(Type=Bone,Name="extra_tentacle_back_l_B_05"), Space=LocalSpace, bInitial=False).Transform.Rotation
   )
   > &RigUnit_ModifyTransforms(
    ItemToModify=((Item=(Type=Bone,Name="extra_tentacle_back_l_B_05"),Transform=(Rotation=(X=0,Y=0,Z=0,W=1),Translation=(X=0,Y=0,Z=0),Scale3D=(X=1,Y=1,Z=1))))
    Mode=OverrideLocal
    WeightMinimum=0
    ItemToModify.0.Transform.Rotation=&RigVMFunction_MathQuaternionFromEuler(Euler=(X=0,Y=0,Z=0), RotationOrder=XYZ, Euler.Y=@n62.Result.Y).Result
    Weight=@n61.Value
   )
   n63 = &RigVMFunction_MathQuaternionToEuler(
    RotationOrder=XYZ
    Value=&RigUnit_GetTransform(Item=(Type=Bone,Name="extra_tentacle_back_l_B_05"), Space=LocalSpace, bInitial=False).Transform.Rotation
   )
   n64 = &RigUnit_SpringInterpQuaternionV2(
    CriticalDamping=0.1
    Current=(X=0,Y=0,Z=0,W=1)
    Strength=1
    TargetVelocityAmount=0
    Torque=(X=0,Y=0,Z=0)
    bInitializeFromTarget=False
    bUseCurrentInput=False
    Target=&RigVMFunction_TimeOffsetTransform(
     BufferSize=16
     SecondsAgo=0.6
     TimeRange=1
     Value=(Rotation=(X=0,Y=0,Z=0,W=1),Translation=(X=0,Y=0,Z=0),Scale3D=(X=1,Y=1,Z=1))
     Value.Rotation=&RigVMFunction_MathQuaternionFromEuler(Euler=(X=0,Y=0,Z=0), RotationOrder=XYZ, Euler.Y=@n63.Result.Y).Result
    ) {template="TimeOffset::Execute(in Value,in SecondsAgo,in BufferSize,in TimeRange,out Result)"}.Result.Rotation
   ) {template="SpringInterp::Execute(in Target,in Strength,in CriticalDamping,in Torque,in bUseCurrentInput,in Current,in TargetVelocityAmount,in bInitializeFromTarget,out Result,out AngularVelocity)"}
   > &RigUnit_ModifyTransforms(
    ItemToModify=((Item=(Type=Bone,Name="BASE_tentacle_back_l_B_06"),Transform=(Rotation=(X=0,Y=0,Z=0,W=1),Translation=(X=0,Y=0,Z=0),Scale3D=(X=1,Y=1,Z=1))))
    WeightMinimum=0
    ItemToModify.0.Transform.Rotation=@n64.Result
    Weight=@n61.Value
   )
   n65 = &RigVMFunction_MathQuaternionToEuler(
    RotationOrder=XYZ
    Value=&RigUnit_GetTransform(Item=(Type=Bone,Name="BASE_tentacle_back_l_B_06"), Space=LocalSpace, bInitial=False).Transform.Rotation
   )
   n66 = &RigUnit_SpringInterpQuaternionV2(
    CriticalDamping=1
    Current=(X=0,Y=0,Z=0,W=1)
    Strength=6
    TargetVelocityAmount=0
    Torque=(X=0,Y=0,Z=0)
    bInitializeFromTarget=False
    bUseCurrentInput=False
    Target=&RigVMFunction_TimeOffsetTransform(
     BufferSize=16
     SecondsAgo=0.1
     TimeRange=1
     Value=(Rotation=(X=0,Y=0,Z=0,W=1),Translation=(X=0,Y=0,Z=0),Scale3D=(X=1,Y=1,Z=1))
     Value.Rotation=&RigVMFunction_MathQuaternionFromEuler(Euler=(X=0,Y=0,Z=0), RotationOrder=XYZ, Euler.Y=@n65.Result.Y).Result
    ) {template="TimeOffset::Execute(in Value,in SecondsAgo,in BufferSize,in TimeRange,out Result)"}.Result.Rotation
   ) {template="SpringInterp::Execute(in Target,in Strength,in CriticalDamping,in Torque,in bUseCurrentInput,in Current,in TargetVelocityAmount,in bInitializeFromTarget,out Result,out AngularVelocity)"}
   > &RigUnit_ModifyTransforms(
    ItemToModify=((Item=(Type=Bone,Name="BASE_tentacle_back_l_B_07"),Transform=(Rotation=(X=0,Y=0,Z=0,W=1),Translation=(X=0,Y=0,Z=0),Scale3D=(X=1,Y=1,Z=1))))
    WeightMinimum=0
    ItemToModify.0.Transform.Rotation=@n66.Result
    Weight=@n61.Value
   )
   Multiply = &RigVMFunction_MathDoubleMul(
    B=1
    A=&RigVMFunction_MathQuaternionToEuler(
     RotationOrder=XYZ
     Value=&RigUnit_GetTransform(Item=(Type=Bone,Name="BASE_tentacle_back_l_B_07"), Space=LocalSpace, bInitial=False).Transform.Rotation
    ).Result.Y
   )
   n67 = &RigUnit_SpringInterpQuaternionV2(
    CriticalDamping=1
    Current=(X=0,Y=0,Z=0,W=1)
    Strength=6
    TargetVelocityAmount=0
    Torque=(X=0,Y=0,Z=0)
    bInitializeFromTarget=False
    bUseCurrentInput=False
    Target=&RigVMFunction_TimeOffsetTransform(
     BufferSize=16
     SecondsAgo=0.7
     TimeRange=1
     Value=(Rotation=(X=0,Y=0,Z=0,W=1),Translation=(X=0,Y=0,Z=0),Scale3D=(X=1,Y=1,Z=1))
     Value.Rotation=&RigVMFunction_MathQuaternionFromEuler(Euler=(X=0,Y=0,Z=0), RotationOrder=XYZ, Euler.Y=@Multiply.Result).Result
    ) {template="TimeOffset::Execute(in Value,in SecondsAgo,in BufferSize,in TimeRange,out Result)"}.Result.Rotation
   ) {template="SpringInterp::Execute(in Target,in Strength,in CriticalDamping,in Torque,in bUseCurrentInput,in Current,in TargetVelocityAmount,in bInitializeFromTarget,out Result,out AngularVelocity)"}
   > &RigUnit_ModifyTransforms(
    ItemToModify=((Item=(Type=Bone,Name="BASE_tentacle_back_l_B_08"),Transform=(Rotation=(X=0,Y=0,Z=0,W=1),Translation=(X=0,Y=0,Z=0),Scale3D=(X=1,Y=1,Z=1))))
    WeightMinimum=0
    ItemToModify.0.Transform.Rotation=@n67.Result
    Weight=@n61.Value
   )
   n68 = RigVMVariableNode() {variable=backpack_cape_mod}
   > &RigUnit_ItemHarmonics(
    RotationOrder=YZX
    Targets=((Item=(Type=Bone,Name="extra_tentacle_back_r_B_04"),Ratio=1))
    WaveAmplitude=(X=0,Y=0,Z=30)
    WaveEase=Linear
    WaveFrequency=(X=1,Y=0.6,Z=0.6)
    WaveMinimum=0.5
    WaveNoise=(X=0,Y=0,Z=0)
    WaveOffset=(X=0,Y=0,Z=0)
    WaveSpeed=(X=3,Y=2,Z=2)
    WaveMaximum=@n68.Value
   )
   n69 = &RigVMFunction_MathDoubleRemap(
    SourceMaximum=29
    SourceMinimum=-29
    TargetMaximum=-30
    TargetMinimum=0
    Value=&RigVMFunction_MathQuaternionToEuler(
     RotationOrder=ZYX
     Value=&RigUnit_GetTransform(Item=(Type=Bone,Name="extra_tentacle_back_r_B_04"), Space=LocalSpace, bInitial=False).Transform.Rotation
    ).Result.Z
   )
   > &RigUnit_ModifyTransforms(
    ItemToModify=((Item=(Type=Bone,Name="extra_tentacle_back_r_B_04"),Transform=(Rotation=(X=0,Y=0,Z=0,W=1),Translation=(X=0,Y=0,Z=0),Scale3D=(X=1,Y=1,Z=1))))
    Mode=OverrideLocal
    WeightMinimum=0
    ItemToModify.0.Transform.Rotation=&RigVMFunction_MathQuaternionFromEuler(Euler=(X=0,Y=0,Z=0), RotationOrder=XYZ, Euler.X=@n69.Result, Euler.Y=@n69.Result).Result
    Weight=@n68.Value
   )
   n70 = &RigVMFunction_MathQuaternionToEuler(
    RotationOrder=XYZ
    Value=&RigUnit_GetTransform(Item=(Type=Bone,Name="extra_tentacle_back_r_B_04"), Space=LocalSpace, bInitial=False).Transform.Rotation
   )
   n71 = &RigUnit_SpringInterpQuaternionV2(
    CriticalDamping=0.1
    Current=(X=0,Y=0,Z=0,W=1)
    Strength=1
    TargetVelocityAmount=0
    Torque=(X=0,Y=0,Z=0)
    bInitializeFromTarget=False
    bUseCurrentInput=False
    Target=&RigVMFunction_TimeOffsetTransform(
     BufferSize=16
     SecondsAgo=0.6
     TimeRange=1
     Value=(Rotation=(X=0,Y=0,Z=0,W=1),Translation=(X=0,Y=0,Z=0),Scale3D=(X=1,Y=1,Z=1))
     Value.Rotation=&RigVMFunction_MathQuaternionFromEuler(Euler=(X=0,Y=0,Z=0), RotationOrder=XYZ, Euler.Y=@n70.Result.Y).Result
    ) {template="TimeOffset::Execute(in Value,in SecondsAgo,in BufferSize,in TimeRange,out Result)"}.Result.Rotation
   ) {template="SpringInterp::Execute(in Target,in Strength,in CriticalDamping,in Torque,in bUseCurrentInput,in Current,in TargetVelocityAmount,in bInitializeFromTarget,out Result,out AngularVelocity)"}
   > &RigUnit_ModifyTransforms(
    ItemToModify=((Item=(Type=Bone,Name="BASE_tentacle_back_r_B_05"),Transform=(Rotation=(X=0,Y=0,Z=0,W=1),Translation=(X=0,Y=0,Z=0),Scale3D=(X=1,Y=1,Z=1))))
    WeightMinimum=0
    ItemToModify.0.Transform.Rotation=@n71.Result
    Weight=@n68.Value
   )
   n72 = &RigVMFunction_MathQuaternionToEuler(
    RotationOrder=XYZ
    Value=&RigUnit_GetTransform(Item=(Type=Bone,Name="BASE_tentacle_back_r_B_05"), Space=LocalSpace, bInitial=False).Transform.Rotation
   )
   n73 = &RigVMFunction_MathDoubleMul(B=0.5, A=@n72.Result.Y)
   n74 = &RigUnit_SpringInterpQuaternionV2(
    CriticalDamping=1
    Current=(X=0,Y=0,Z=0,W=1)
    Strength=6
    TargetVelocityAmount=0
    Torque=(X=0,Y=0,Z=0)
    bInitializeFromTarget=False
    bUseCurrentInput=False
    Target=&RigVMFunction_TimeOffsetTransform(
     BufferSize=16
     SecondsAgo=0.1
     TimeRange=1
     Value=(Rotation=(X=0,Y=0,Z=0,W=1),Translation=(X=0,Y=0,Z=0),Scale3D=(X=1,Y=1,Z=1))
     Value.Rotation=&RigVMFunction_MathQuaternionFromEuler(Euler=(X=0,Y=0,Z=0), RotationOrder=XYZ, Euler.Y=@n73.Result, Euler.Z=@n72.Result.Z).Result
    ) {template="TimeOffset::Execute(in Value,in SecondsAgo,in BufferSize,in TimeRange,out Result)"}.Result.Rotation
   ) {template="SpringInterp::Execute(in Target,in Strength,in CriticalDamping,in Torque,in bUseCurrentInput,in Current,in TargetVelocityAmount,in bInitializeFromTarget,out Result,out AngularVelocity)"}
   > &RigUnit_ModifyTransforms(
    ItemToModify=((Item=(Type=Bone,Name="BASE_tentacle_back_r_B_06"),Transform=(Rotation=(X=0,Y=0,Z=0,W=1),Translation=(X=0,Y=0,Z=0),Scale3D=(X=1,Y=1,Z=1))))
    WeightMinimum=0
    ItemToModify.0.Transform.Rotation=@n74.Result
    Weight=@n68.Value
   )
   n75 = &RigVMFunction_AlphaInterp(
    InterpSpeedDecreasing=50
    InterpSpeedIncreasing=1
    Scale=1
    bClampResult=False
    Value=&RigVMFunction_MathDoubleMul(A=RigVMVariableNode() {variable=backpack_cape_mod}.Value, B=RigVMVariableNode() {variable=jump_sprint}.Value).Result
   )
   > &RigUnit_ItemHarmonics(
    RotationOrder=YZX
    Targets=((Item=(Type=Bone,Name="extra_tentacle_back_r_A_03"),Ratio=1))
    WaveAmplitude=(X=0,Y=0,Z=30)
    WaveEase=Linear
    WaveFrequency=(X=1,Y=0.6,Z=0.6)
    WaveMinimum=0.5
    WaveNoise=(X=0,Y=0,Z=0)
    WaveOffset=(X=0,Y=0,Z=0)
    WaveSpeed=(X=3,Y=2,Z=2)
    WaveMaximum=@n75.Result
   )
   n76 = &RigVMFunction_MathDoubleRemap(
    SourceMaximum=29
    SourceMinimum=-29
    TargetMaximum=-29
    TargetMinimum=0
    Value=&RigVMFunction_MathQuaternionToEuler(
     RotationOrder=ZYX
     Value=&RigUnit_GetTransform(Item=(Type=Bone,Name="extra_tentacle_back_r_A_03"), Space=LocalSpace, bInitial=False).Transform.Rotation
    ).Result.Z
   )
   > &RigUnit_ModifyTransforms(
    ItemToModify=((Item=(Type=Bone,Name="extra_tentacle_back_r_A_03"),Transform=(Rotation=(X=0,Y=0,Z=0,W=1),Translation=(X=0,Y=0,Z=0),Scale3D=(X=1,Y=1,Z=1))))
    Mode=OverrideLocal
    WeightMinimum=0
    Weight=@n75.Result
    ItemToModify.0.Transform.Rotation=&RigVMFunction_MathQuaternionFromEuler(Euler=(X=0,Y=0,Z=0), RotationOrder=XYZ, Euler.X=@n76.Result, Euler.Y=@n76.Result, Euler.Z=@n76.Result).Result
   )
   n77 = &RigUnit_GetTransform(Item=(Type=Bone,Name="extra_tentacle_back_r_A_03"), Space=LocalSpace, bInitial=False)
   n78 = &RigUnit_SpringInterpQuaternionV2(
    CriticalDamping=0.1
    Current=(X=0,Y=0,Z=0,W=1)
    Strength=1
    TargetVelocityAmount=0
    Torque=(X=0,Y=0,Z=0)
    bInitializeFromTarget=False
    bUseCurrentInput=False
    Target=&RigVMFunction_TimeOffsetTransform(
     BufferSize=16
     SecondsAgo=0.6
     TimeRange=1
     Value=(Rotation=(X=0,Y=0,Z=0,W=1),Translation=(X=0,Y=0,Z=0),Scale3D=(X=1,Y=1,Z=1))
     Value.Rotation=@n77.Transform.Rotation
    ) {template="TimeOffset::Execute(in Value,in SecondsAgo,in BufferSize,in TimeRange,out Result)"}.Result.Rotation
   ) {template="SpringInterp::Execute(in Target,in Strength,in CriticalDamping,in Torque,in bUseCurrentInput,in Current,in TargetVelocityAmount,in bInitializeFromTarget,out Result,out AngularVelocity)"}
   > &RigUnit_ModifyTransforms(
    ItemToModify=((Item=(Type=Bone,Name="BASE_tentacle_back_r_A_04"),Transform=(Rotation=(X=0,Y=0,Z=0,W=1),Translation=(X=0,Y=0,Z=0),Scale3D=(X=1,Y=1,Z=1))))
    WeightMinimum=0
    Weight=@n75.Result
    ItemToModify.0.Transform.Rotation=@n78.Result
   )
   n79 = &RigUnit_GetTransform(Item=(Type=Bone,Name="BASE_tentacle_back_r_A_04"), Space=LocalSpace, bInitial=False)
   n80 = &RigUnit_SpringInterpQuaternionV2(
    CriticalDamping=1
    Current=(X=0,Y=0,Z=0,W=1)
    Strength=6
    TargetVelocityAmount=0
    Torque=(X=0,Y=0,Z=0)
    bInitializeFromTarget=False
    bUseCurrentInput=False
    Target=&RigVMFunction_TimeOffsetTransform(
     BufferSize=16
     SecondsAgo=0.1
     TimeRange=1
     Value=(Rotation=(X=0,Y=0,Z=0,W=1),Translation=(X=0,Y=0,Z=0),Scale3D=(X=1,Y=1,Z=1))
     Value.Rotation=@n79.Transform.Rotation
    ) {template="TimeOffset::Execute(in Value,in SecondsAgo,in BufferSize,in TimeRange,out Result)"}.Result.Rotation
   ) {template="SpringInterp::Execute(in Target,in Strength,in CriticalDamping,in Torque,in bUseCurrentInput,in Current,in TargetVelocityAmount,in bInitializeFromTarget,out Result,out AngularVelocity)"}
   > &RigUnit_ModifyTransforms(
    ItemToModify=((Item=(Type=Bone,Name="BASE_tentacle_back_r_A_05"),Transform=(Rotation=(X=0,Y=0,Z=0,W=1),Translation=(X=0,Y=0,Z=0),Scale3D=(X=1,Y=1,Z=1))))
    WeightMinimum=0
    Weight=@n75.Result
    ItemToModify.0.Transform.Rotation=@n80.Result
   )
   n81 = &RigUnit_GetTransform(Item=(Type=Bone,Name="BASE_tentacle_back_r_A_05"), Space=LocalSpace, bInitial=False)
   n82 = &RigUnit_SpringInterpQuaternionV2(
    CriticalDamping=1
    Current=(X=0,Y=0,Z=0,W=1)
    Strength=6
    TargetVelocityAmount=0
    Torque=(X=0,Y=0,Z=0)
    bInitializeFromTarget=False
    bUseCurrentInput=False
    Target=&RigVMFunction_TimeOffsetTransform(
     BufferSize=16
     SecondsAgo=0.7
     TimeRange=1
     Value=(Rotation=(X=0,Y=0,Z=0,W=1),Translation=(X=0,Y=0,Z=0),Scale3D=(X=1,Y=1,Z=1))
     Value.Rotation=@n81.Transform.Rotation
    ) {template="TimeOffset::Execute(in Value,in SecondsAgo,in BufferSize,in TimeRange,out Result)"}.Result.Rotation
   ) {template="SpringInterp::Execute(in Target,in Strength,in CriticalDamping,in Torque,in bUseCurrentInput,in Current,in TargetVelocityAmount,in bInitializeFromTarget,out Result,out AngularVelocity)"}
   > &RigUnit_ModifyTransforms(
    ItemToModify=((Item=(Type=Bone,Name="BASE_tentacle_back_r_A_06"),Transform=(Rotation=(X=0,Y=0,Z=0,W=1),Translation=(X=0,Y=0,Z=0),Scale3D=(X=1,Y=1,Z=1))))
    WeightMinimum=0
    Weight=@n75.Result
    ItemToModify.0.Transform.Rotation=@n82.Result
   )
   Maximum = &RigVMFunction_MathDoubleMax(B=@n1.OutputParam, A=@n2.OutputParam)
   &RigVMFunction_MathDoubleRemap(SourceMaximum=-30, SourceMinimum=30, TargetMaximum=-29, TargetMinimum=0, bClamp=false, Value=@n62.Result.Y)
   &RigVMFunction_MathDoubleRemap(SourceMaximum=-29, SourceMinimum=29, TargetMaximum=20, TargetMinimum=1, bClamp=false, Value=@n32.Result.Z)
   &RigVMFunction_MathQuaternionFromEuler(
    Euler=(X=0,Y=0,Z=0)
    RotationOrder=XYZ
    Euler.Z=&RigVMFunction_MathQuaternionToEuler(RotationOrder=XYZ, Value=@n77.Transform.Rotation).Result.Z
   )
   &RigVMFunction_MathQuaternionFromEuler(
    Euler=(X=0,Y=0,Z=0)
    RotationOrder=XYZ
    Euler.Z=&RigVMFunction_MathQuaternionToEuler(RotationOrder=XYZ, Value=@n79.Transform.Rotation).Result.Z
   )
   &RigVMFunction_MathQuaternionFromEuler(
    Euler=(X=0,Y=0,Z=0)
    RotationOrder=XYZ
    Euler.Z=&RigVMFunction_MathQuaternionToEuler(RotationOrder=XYZ, Value=@n81.Transform.Rotation).Result.Z
   )
  ]}
]
hierarchy [
 root {loc=<0 0 0>, rot=<0 0 0 1>}
 pelvis {parent=root, loc=<0 -0.5301 98.0571>, rot=<0.0127 0.707 -0.0127 -0.707>}
 spine_01 {parent=pelvis, loc=<6.1208 -0.9328 0>, rot=<0 0 0.034 -0.9994>}
 spine_02 {parent=spine_01, loc=<6.5997 -0.2095 0>, rot=<0 0 0.0238 -0.9997>}
 spine_03 {parent=spine_02, loc=<6.4914 -0.1237 0>, rot=<0 0 0.0149 -0.9999>}
 spine_04 {parent=spine_03, loc=<6.9598 0.0794 0>, rot=<0 0 -0.0132 -0.9999>}
 spine_05 {parent=spine_04, loc=<6.9987 0 0>, rot=<0 0 -0.0015 -1>}
 neck_01 {parent=spine_05, loc=<11.3138 0.1529 0>, rot=<0 0 -0.2003 -0.9797>}
 neck_02 {parent=neck_01, loc=<4.5627 0 0>, rot=<0 0 0.0504 -0.9987>}
 head {parent=neck_02, loc=<5.3914 0 0>, rot=<0 0 0.115 -0.9934>}
 base_hat {parent=head, loc=<5.3677 0.059 0>, rot=<0.4731 -0.881 0 0>}
 BASE_dyn_hat {parent=base_hat, loc=<0 0 0>, rot=<0 0 0 1>}
 dyn_hat {parent=BASE_dyn_hat, loc=<0 0 0>, rot=<0 0 0 1>}
 base_Beannie {parent=dyn_hat, loc=<2.0038 -5.9678 0>, rot=<0 0 -0.0824 -0.9966>}
 dyn_Beannie {parent=base_Beannie, loc=<0 0 0>, rot=<0 0 0 1>}
 BeannieEnd {parent=dyn_Beannie, loc=<12.9883 0 0>, rot=<0 0 0 1>}
 BASE_tentacle_back_l_A_01 {parent=head, loc=<1.9998 -5.507 -6.5638>, rot=<-0.0033 -0.7635 0.6455 -0.0203>}
 tentacle_back_l_A_01 {parent=BASE_tentacle_back_l_A_01, loc=<0 0 0>, rot=<0 0 0 1>}
 tentacle_back_l_A_02 {parent=tentacle_back_l_A_01, loc=<3.839 0 0>, rot=<0.0949 0.154 -0.1208 -0.9761>}
 tentacle_back_l_A_03 {parent=tentacle_back_l_A_02, loc=<3.7887 0 0>, rot=<-0.0108 0.1528 0.0891 -0.9842>}
 extra_tentacle_back_l_A_03 {parent=tentacle_back_l_A_03, loc=<0 0 0>, rot=<0 0 0 1>}
 extra_tentacle_back_l_A_04 {parent=extra_tentacle_back_l_A_03, loc=<4.6748 0 0>, rot=<-0.0466 -0.0024 0.1201 -0.9917>}
 tentacle_back_l_A_04 {parent=extra_tentacle_back_l_A_04, loc=<0 0 0>, rot=<0 0 0 1>}
 extra_tentacle_back_l_A_05 {parent=tentacle_back_l_A_04, loc=<4.8885 0 0>, rot=<-0.0697 -0.0634 0.1277 -0.9873>}
 tentacle_back_l_A_05 {parent=extra_tentacle_back_l_A_05, loc=<0 0 0>, rot=<0 0 0 1>}
 extra_tentacle_back_l_A_06 {parent=tentacle_back_l_A_05, loc=<5.2483 0 0>, rot=<0.1011 -0.0183 -0.1772 -0.9788>}
 tentacle_back_l_A_06 {parent=extra_tentacle_back_l_A_06, loc=<0 0 0>, rot=<0 0 0 1>}
 BASE_tentacle_back_r_A_01 {parent=head, loc=<1.6393 -6.4682 6.0603>, rot=<0.0544 -0.5147 0.8549 0.037>}
 tentacle_back_r_A_01 {parent=BASE_tentacle_back_r_A_01, loc=<0 0 0>, rot=<-0.0609 -0.087 -0.0053 -0.9943>}
 tentacle_back_r_A_02 {parent=tentacle_back_r_A_01, loc=<5.0348 0.0399 0.3952>, rot=<0.0481 0.2881 -0.0145 -0.9563>}
 tentacle_back_r_A_03 {parent=tentacle_back_r_A_02, loc=<3.7523 0 0>, rot=<0.0354 0.1498 -0.0194 -0.9879>}
 extra_tentacle_back_r_A_03 {parent=tentacle_back_r_A_03, loc=<0 0 0>, rot=<0 0 0 1>}
 BASE_tentacle_back_r_A_04 {parent=extra_tentacle_back_r_A_03, loc=<4.5945 0 0>, rot=<0.0057 0.0249 -0.1942 -0.9806>}
 tentacle_back_r_A_04 {parent=BASE_tentacle_back_r_A_04, loc=<0 0 0>, rot=<0 0 0 1>}
 BASE_tentacle_back_r_A_05 {parent=tentacle_back_r_A_04, loc=<4.3915 0 0>, rot=<-0.0652 -0.1153 -0.293 -0.9469>}
 tentacle_back_r_A_05 {parent=BASE_tentacle_back_r_A_05, loc=<0 0 0>, rot=<0 0 0 1>}
 BASE_tentacle_back_r_A_06 {parent=tentacle_back_r_A_05, loc=<4.602 0 0>, rot=<-0.0107 -0.2033 -0.0516 -0.9777>}
 tentacle_back_r_A_06 {parent=BASE_tentacle_back_r_A_06, loc=<0 0 0>, rot=<0 0 0 1>}
 extra_tentacle_back_r_A_02 {parent=tentacle_back_r_A_02, loc=<0 0 0>, rot=<0 0 0 1>}
 extra_tentacle_back_r_A_01 {parent=tentacle_back_r_A_01, loc=<0 0 0>, rot=<0 0 0 1>}
 extra_tentacle_back_r_A_01_end {parent=extra_tentacle_back_r_A_01, loc=<2.2767 0 0>, rot=<0 0 0 1>}
 BASE_tentacle_back_r_B_01 {parent=head, loc=<1.3481 -7.8211 2.0209>, rot=<0.0352 -0.7043 0.7087 0.0205>}
 tentacle_back_r_B_01 {parent=BASE_tentacle_back_r_B_01, loc=<0 0 0>, rot=<-0.0609 -0.087 -0.0053 -0.9943>}
 tentacle_back_r_B_02 {parent=tentacle_back_r_B_01, loc=<5.0348 0.0399 0.3952>, rot=<0.0488 0.239 -0.012 -0.9697>}
 tentacle_back_r_B_03 {parent=tentacle_back_r_B_02, loc=<3.7523 0 0>, rot=<0.0354 0.1498 -0.0194 -0.9879>}
 tentacle_back_r_B_04 {parent=tentacle_back_r_B_03, loc=<4.5945 0 0>, rot=<0.021 -0.0349 -0.1976 -0.9794>}
 extra_tentacle_back_r_B_04 {parent=tentacle_back_r_B_04, loc=<0 0 0>, rot=<0 0 0 1>}
 BASE_tentacle_back_r_B_05 {parent=extra_tentacle_back_r_B_04, loc=<4.3915 0 0>, rot=<0.0468 -0.2288 0.011 -0.9723>}
 tentacle_back_r_B_05 {parent=BASE_tentacle_back_r_B_05, loc=<0 0 0>, rot=<0 0 0 1>}
 BASE_tentacle_back_r_B_06 {parent=tentacle_back_r_B_05, loc=<4.602 0 0>, rot=<-0.0415 -0.1492 0.8557 -0.4937>}
 tentacle_back_r_B_06 {parent=BASE_tentacle_back_r_B_06, loc=<0 0 0>, rot=<0.1594 0.1511 -0.4387 -0.8714>}
 BASE_tentacle_back_l_B_01 {parent=head, loc=<1.863 -6.2387 -2.2394>, rot=<-0.0033 -0.7635 0.6455 -0.0203>}
 tentacle_back_l_B_01 {parent=BASE_tentacle_back_l_B_01, loc=<0 0 0>, rot=<-0.0612 0 0 -0.9981>}
 tentacle_back_l_B_02 {parent=tentacle_back_l_B_01, loc=<3.839 0 0>, rot=<0.0496 0.1595 -0.008 -0.9859>}
 tentacle_back_l_B_03 {parent=tentacle_back_l_B_02, loc=<3.7523 0 0>, rot=<0.0354 0.1498 -0.0194 -0.9879>}
 tentacle_back_l_B_04 {parent=tentacle_back_l_B_03, loc=<4.5945 0 0>, rot=<-0.0673 0.186 0.0128 -0.9802>}
 tentacle_back_l_B_05 {parent=tentacle_back_l_B_04, loc=<4.3915 0 0>, rot=<0.0463 0.2642 -0.0127 -0.9633>}
 extra_tentacle_back_l_B_05 {parent=tentacle_back_l_B_05, loc=<0 0 0>, rot=<0 0 0 1>}
 BASE_tentacle_back_l_B_06 {parent=extra_tentacle_back_l_B_05, loc=<4.602 0 0>, rot=<0 -0.2036 0 -0.9791>}
 tentacle_back_l_B_06 {parent=BASE_tentacle_back_l_B_06, loc=<0 0 0>, rot=<0 0 0 1>}
 BASE_tentacle_back_l_B_07 {parent=tentacle_back_l_B_06, loc=<3.4145 0 0>, rot=<0 0 0 1>}
 tentacle_back_l_B_07 {parent=BASE_tentacle_back_l_B_07, loc=<0 0 0>, rot=<0 0 0 1>}
 BASE_tentacle_back_l_B_08 {parent=tentacle_back_l_B_07, loc=<1.0419 0 -2.1533>, rot=<0 0 0 1>}
 tentacle_back_l_B_08 {parent=BASE_tentacle_back_l_B_08, loc=<0 0 0>, rot=<0 0 0 1>}
 BASE_tentacle_front_r_A_01 {parent=head, loc=<-3.1509 -1.8827 7.5106>, rot=<-0.0883 -0.9322 0.3228 0.1375>}
 tentacle_front_r_A_01 {parent=BASE_tentacle_front_r_A_01, loc=<0 0 0>, rot=<0 0 0 1>}
 tentacle_front_r_A_02 {parent=tentacle_front_r_A_01, loc=<3.7919 0 0>, rot=<0 -0.1517 0 -0.9884>}
 extra_tentacle_front_r_A_02 {parent=tentacle_front_r_A_02, loc=<0 0 0>, rot=<0 0 0 1>}
 BASE_tentacle_front_r_A_03 {parent=extra_tentacle_front_r_A_02, loc=<4.2634 0 0>, rot=<0.1117 -0.0411 0.0357 -0.9922>}
 tentacle_front_r_A_03 {parent=BASE_tentacle_front_r_A_03, loc=<0 0 0>, rot=<0 0 0 1>}
 extra_tentacle_front_r_A_03 {parent=tentacle_front_r_A_03, loc=<0 0 0>, rot=<0 0 0 1>}
 BASE_tentacle_front_r_A_04 {parent=tentacle_front_r_A_03, loc=<3.6184 0 0>, rot=<0.1329 0.0968 -0.0131 -0.9863>}
 tentacle_front_r_A_04 {parent=BASE_tentacle_front_r_A_04, loc=<0 0 0>, rot=<0 0 0 1>}
 extra_tentacle_front_r_A_04 {parent=tentacle_front_r_A_04, loc=<0 0 0>, rot=<0 0 0 1>}
 BASE_tentacle_front_r_A_05 {parent=tentacle_front_r_A_04, loc=<3.5936 0 0>, rot=<0 0 0 1>}
 tentacle_front_r_A_05 {parent=BASE_tentacle_front_r_A_05, loc=<0 0 0>, rot=<0 0 0 1>}
 extra_tentacle_front_r_A_05 {parent=tentacle_front_r_A_05, loc=<0 0 0>, rot=<0 0 0 1>}
 extra_tentacle_front_r_A_01 {parent=tentacle_front_r_A_01, loc=<0 0 0>, rot=<0 0 0 1>}
 BASE_tentacle_front_r_B_01 {parent=head, loc=<-2.0465 3.2033 7.2709>, rot=<-0.0573 -0.9416 0.329 -0.0428>}
 tentacle_front_r_B_01 {parent=BASE_tentacle_front_r_B_01, loc=<0 0 0>, rot=<0 0 0 1>}
 tentacle_front_r_B_02 {parent=tentacle_front_r_B_01, loc=<5.5529 0 0>, rot=<0 -0.2558 0 -0.9667>}
 extra_tentacle_front_r_B_02 {parent=tentacle_front_r_B_02, loc=<0 0 0>, rot=<0 0 0 1>}
 BASE_tentacle_front_r_B_03 {parent=extra_tentacle_front_r_B_02, loc=<5.5326 0 0>, rot=<0.0302 -0.0565 0.0554 -0.9964>}
 tentacle_front_r_B_03 {parent=BASE_tentacle_front_r_B_03, loc=<0 0 0>, rot=<0 0 0 1>}
 extra_tentacle_front_r_B_03 {parent=tentacle_front_r_B_03, loc=<0 0 0>, rot=<0 0 0 1>}
 BASE_tentacle_front_r_B_04 {parent=tentacle_front_r_B_03, loc=<4.2789 0 0>, rot=<-0.0293 0.117 0.0035 -0.9927>}
 tentacle_front_r_B_04 {parent=BASE_tentacle_front_r_B_04, loc=<0 0 0>, rot=<0 0 0 1>}
 extra_tentacle_front_r_B_04 {parent=tentacle_front_r_B_04, loc=<0 0 0>, rot=<0 0 0 1>}
 BASE_tentacle_front_r_B_05 {parent=tentacle_front_r_B_04, loc=<3.7261 0 0>, rot=<0 0 0 1>}
 tentacle_front_r_B_05 {parent=BASE_tentacle_front_r_B_05, loc=<0 0 0>, rot=<0 0 0 1>}
 extra_tentacle_front_r_B_05 {parent=tentacle_front_r_B_05, loc=<0 0 0>, rot=<0 0 0 1>}
 extra_tentacle_front_r_B_01 {parent=tentacle_front_r_B_01, loc=<0 0 0>, rot=<0 0 0 1>}
 BASE_tentacle_front_l_B_01 {parent=head, loc=<-1.4879 4.2361 -7.5252>, rot=<-0.1177 -0.9597 0.2505 0.0494>}
 tentacle_front_l_B_01 {parent=BASE_tentacle_front_l_B_01, loc=<0 0 0>, rot=<0 0 0 1>}
 tentacle_front_l_B_02 {parent=tentacle_front_l_B_01, loc=<6.2575 0 0>, rot=<0 0.1455 0 -0.9894>}
 extra_tentacle_front_l_B_02 {parent=tentacle_front_l_B_02, loc=<0 0 0>, rot=<0 0 0 1>}
 BASE_tentacle_front_l_B_03 {parent=extra_tentacle_front_l_B_02, loc=<4.1508 0 0>, rot=<0.3298 0.1744 -0.0621 -0.9257>}
 tentacle_front_l_B_03 {parent=BASE_tentacle_front_l_B_03, loc=<0 0 0>, rot=<0 0 0 1>}
 extra_tentacle_front_l_B_03 {parent=tentacle_front_l_B_03, loc=<0 0 0>, rot=<0 0 0 1>}
 BASE_tentacle_front_l_B_04 {parent=tentacle_front_l_B_03, loc=<4.7436 0 0>, rot=<0.0171 -0.1391 0.0024 -0.9901>}
 tentacle_front_l_B_04 {parent=BASE_tentacle_front_l_B_04, loc=<0 0 0>, rot=<0 0 0 1>}
 extra_tentacle_front_l_B_04 {parent=tentacle_front_l_B_04, loc=<0 0 0>, rot=<0 0 0 1>}
 BASE_tentacle_front_l_B_05 {parent=tentacle_front_l_B_04, loc=<3.8267 0 0>, rot=<0 -0.1349 0 -0.9909>}
 tentacle_front_l_B_05 {parent=BASE_tentacle_front_l_B_05, loc=<0 0 0>, rot=<0 0 0 1>}
 extra_tentacle_front_l_B_05 {parent=tentacle_front_l_B_05, loc=<0 0 0>, rot=<0 0 0 1>}
 extra_tentacle_front_l_B_01 {parent=tentacle_front_l_B_01, loc=<0 0 0>, rot=<0 0 0 1>}
 BASE_tentacle_front_l_A_01 {parent=head, loc=<-3.682 -0.311 -7.5493>, rot=<-0.1226 -0.9765 -0.1653 -0.0631>}
 tentacle_front_l_A_01 {parent=BASE_tentacle_front_l_A_01, loc=<0 0 0>, rot=<0 0 0 1>}
 tentacle_front_l_A_02 {parent=tentacle_front_l_A_01, loc=<3.5258 0 0>, rot=<0 0.2107 0 -0.9776>}
 extra_tentacle_front_l_A_02 {parent=tentacle_front_l_A_02, loc=<0 0 0>, rot=<0 0 0 1>}
 BASE_tentacle_front_l_A_03 {parent=extra_tentacle_front_l_A_02, loc=<3.7671 0 0>, rot=<-0.2776 0.119 0.0347 -0.9527>}
 tentacle_front_l_A_03 {parent=BASE_tentacle_front_l_A_03, loc=<0 0 0>, rot=<0 0 0 1>}
 extra_tentacle_front_l_A_03 {parent=tentacle_front_l_A_03, loc=<0 0 0>, rot=<0 0 0 1>}
 BASE_tentacle_front_l_A_04 {parent=tentacle_front_l_A_03, loc=<4.4844 0 0>, rot=<0.0923 -0.1977 0.0187 -0.9757>}
 tentacle_front_l_A_04 {parent=BASE_tentacle_front_l_A_04, loc=<0 0 0>, rot=<0 0 0 1>}
 extra_tentacle_front_l_A_04 {parent=tentacle_front_l_A_04, loc=<0 0 0>, rot=<0 0 0 1>}
 BASE_tentacle_front_l_A_05 {parent=tentacle_front_l_A_04, loc=<3.2938 0 0>, rot=<0 -0.1158 0 -0.9933>}
 tentacle_front_l_A_05 {parent=BASE_tentacle_front_l_A_05, loc=<0 0 0>, rot=<0 0 0 1>}
 extra_tentacle_front_l_A_05 {parent=tentacle_front_l_A_05, loc=<0 0 0>, rot=<0 0 0 1>}
 extra_tentacle_front_l_A_01 {parent=tentacle_front_l_A_01, loc=<0 0 0>, rot=<0 0 0 1>}
 FACIAL_C_FacialRoot {parent=head, loc=<2.1524 -0.746 -0.0002>, rot=<0.4879 -0.5118 -0.4879 -0.5118>}
 FACIAL_C_Forehead {parent=FACIAL_C_FacialRoot, loc=<-0.0014 -3.3461 8.7089>, rot=<-0.0524 0 0 -0.9986>}
 FACIAL_L_ForeheadIn {parent=FACIAL_C_FacialRoot, loc=<2.0694 -3.3307 8.5684>, rot=<-0.0697 -0.0696 0.0036 -0.9951>}
 FACIAL_R_ForeheadIn {parent=FACIAL_C_FacialRoot, loc=<-2.0723 -3.3303 8.5648>, rot=<-0.0697 0.0696 -0.0036 -0.9951>}
 FACIAL_L_ForeheadMid {parent=FACIAL_C_FacialRoot, loc=<3.8266 -3.7005 7.9466>, rot=<-0.0798 -0.1755 -0.0222 -0.981>}
 FACIAL_R_ForeheadMid {parent=FACIAL_C_FacialRoot, loc=<-3.8272 -3.6992 7.9419>, rot=<-0.0798 0.1755 0.0222 -0.981>}
 FACIAL_L_ForeheadOut {parent=FACIAL_C_FacialRoot, loc=<5.3962 -3.671 6.9286>, rot=<-0.13 -0.2549 -0.0814 -0.9547>}
 FACIAL_R_ForeheadOut {parent=FACIAL_C_FacialRoot, loc=<-5.396 -3.6708 6.9253>, rot=<-0.13 0.2549 0.0814 -0.9547>}
 FACIAL_L_EyesackUpper {parent=FACIAL_C_FacialRoot, loc=<3.9165 -2.6545 8.5218>, rot=<0.0399 -0.0836 -0.0653 -0.9936>}
 FACIAL_R_EyesackUpper {parent=FACIAL_C_FacialRoot, loc=<-3.9195 -2.6554 8.522>, rot=<0.0399 0.0837 0.0653 -0.9936>}
 FACIAL_L_EyelidUpperFurrow {parent=FACIAL_C_FacialRoot, loc=<3.7839 -2.3743 8.4671>, rot=<-0.0389 -0.1501 -0.081 -0.9846>}
 FACIAL_R_EyelidUpperFurrow {parent=FACIAL_C_FacialRoot, loc=<-3.7871 -2.3761 8.4677>, rot=<-0.0389 0.1501 0.081 -0.9846>}
 FACIAL_L_EyelidUpperA {parent=FACIAL_C_FacialRoot, loc=<3.3646 -0.9032 6.3856>, rot=<-0.1567 0 0 -0.9876>}
 FACIAL_R_EyelidUpperA {parent=FACIAL_C_FacialRoot, loc=<-3.3625 -0.9047 6.387>, rot=<-0.1567 0 0 -0.9876>}
 FACIAL_L_Eye {parent=FACIAL_C_FacialRoot, loc=<3.3553 -0.9133 6.3694>, rot=<-0.009 0 0 -1>}
 FACIAL_R_Eye {parent=FACIAL_C_FacialRoot, loc=<-3.3574 -0.9116 6.37>, rot=<-0.009 0 0 -1>}
 FACIAL_L_EyelidLowerA {parent=FACIAL_C_FacialRoot, loc=<3.3646 -0.9032 6.3856>, rot=<0.2754 0 0 -0.9613>}
 FACIAL_R_EyelidLowerA {parent=FACIAL_C_FacialRoot, loc=<-3.3625 -0.9047 6.387>, rot=<0.2754 0 0 -0.9613>}
 FACIAL_L_EyeCornerInner {parent=FACIAL_C_FacialRoot, loc=<1.9207 -0.2453 7.8056>, rot=<-0.009 0 0 -1>}
 FACIAL_R_EyeCornerInner {parent=FACIAL_C_FacialRoot, loc=<-1.9124 -0.2499 7.8067>, rot=<-0.009 0 0 -1>}
 FACIAL_L_EyeCornerOuter {parent=FACIAL_C_FacialRoot, loc=<5.428 -1.1819 6.9959>, rot=<-0.0083 -0.3826 -0.0034 -0.9239>}
 FACIAL_R_EyeCornerOuter {parent=FACIAL_C_FacialRoot, loc=<-5.4307 -1.1771 6.9953>, rot=<-0.0083 0.3826 0.0034 -0.9239>}
 FACIAL_L_EyesackLower {parent=FACIAL_C_FacialRoot, loc=<4.1701 1.054 6.9559>, rot=<-0.0312 -0.1395 -0.0235 -0.9895>}
 FACIAL_R_EyesackLower {parent=FACIAL_C_FacialRoot, loc=<-4.1653 1.0572 6.9593>, rot=<-0.0312 0.1395 0.0235 -0.9894>}
 FACIAL_L_CheekInner {parent=FACIAL_C_FacialRoot, loc=<2.2978 1.6962 7.6699>, rot=<-0.1319 -0.0862 -0.1979 -0.9675>}
 FACIAL_R_CheekInner {parent=FACIAL_C_FacialRoot, loc=<-2.2933 1.6957 7.6711>, rot=<-0.1319 0.0862 0.1979 -0.9675>}
 FACIAL_L_CheekOuter {parent=FACIAL_C_FacialRoot, loc=<4.6045 1.5395 5.9971>, rot=<0.0469 -0.3778 0.097 -0.9196>}
 FACIAL_R_CheekOuter {parent=FACIAL_C_FacialRoot, loc=<-4.6034 1.5321 5.9977>, rot=<0.0469 0.3778 -0.097 -0.9196>}
 FACIAL_L_NasolabialBulge {parent=FACIAL_C_FacialRoot, loc=<3.3479 4.8304 6.1358>, rot=<0.0496 -0.3775 0.0204 -0.9244>}
 FACIAL_R_NasolabialBulge {parent=FACIAL_C_FacialRoot, loc=<-3.3479 4.8301 6.1361>, rot=<0.0496 0.3775 -0.0204 -0.9244>}
 FACIAL_L_NasolabialFurrow {parent=FACIAL_C_FacialRoot, loc=<3.1766 3.8263 8.5693>, rot=<-0.0416 -0.1225 -0.3647 -0.9221>}
 FACIAL_R_NasolabialFurrow {parent=FACIAL_C_FacialRoot, loc=<-3.1792 3.8219 8.5691>, rot=<-0.0416 0.1225 0.3647 -0.9221>}
 FACIAL_L_CheekLower {parent=FACIAL_C_FacialRoot, loc=<3.7061 4.6319 5.4605>, rot=<0.0265 -0.4274 0.0539 -0.9021>}
 FACIAL_R_CheekLower {parent=FACIAL_C_FacialRoot, loc=<-3.7058 4.631 5.4613>, rot=<0.0265 0.4274 -0.0539 -0.9021>}
 FACIAL_C_Nose {parent=FACIAL_C_FacialRoot, loc=<0.0004 0.7471 8.8981>, rot=<-0.1393 0 0 -0.9903>}
 FACIAL_C_NoseLower {parent=FACIAL_C_Nose, loc=<-0.0006 3.101 -0.1813>, rot=<0.1305 0 0 -0.9914>}
 FACIAL_L_Nostril {parent=FACIAL_C_Nose, loc=<0.6535 2.3758 -0.0484>, rot=<0.0173 0.1305 0.0023 -0.9913>}
 FACIAL_R_Nostril {parent=FACIAL_C_Nose, loc=<-0.6514 2.373 -0.0519>, rot=<0.0173 -0.1305 -0.0023 -0.9913>}
 FACIAL_C_NoseTip {parent=FACIAL_C_Nose, loc=<-0.0008 2.4446 0.4603>, rot=<0.0872 0 0 -0.9962>}
 FACIAL_C_MouthUpper {parent=FACIAL_C_FacialRoot, loc=<0.0004 4.9677 7.2142>, rot=<-0.009 0 0 -1>}
 FACIAL_C_LipUpper {parent=FACIAL_C_MouthUpper, loc=<-0.0002 0.2029 2.0922>, rot=<-0.0872 0 0 -0.9962>}
 FACIAL_L_LipUpper {parent=FACIAL_C_MouthUpper, loc=<0.7891 0.2516 2.0203>, rot=<-0.0778 -0.1601 -0.0379 -0.9833>}
 FACIAL_R_LipUpper {parent=FACIAL_C_MouthUpper, loc=<-0.7906 0.2515 2.0195>, rot=<-0.0778 0.1601 0.0379 -0.9833>}
 FACIAL_L_LipUpperOuter {parent=FACIAL_C_MouthUpper, loc=<1.502 0.3994 1.5472>, rot=<-0.0984 -0.2293 -0.0983 -0.9634>}
 FACIAL_R_LipUpperOuter {parent=FACIAL_C_MouthUpper, loc=<-1.5022 0.4008 1.5462>, rot=<-0.0984 0.2293 0.0983 -0.9634>}
 FACIAL_L_LipCorner {parent=FACIAL_C_MouthUpper, loc=<2.0534 0.69 0.7577>, rot=<0 -0.2078 0 -0.9782>}
 FACIAL_R_LipCorner {parent=FACIAL_C_MouthUpper, loc=<-2.0536 0.6911 0.7547>, rot=<0 0.2078 0 -0.9782>}
 FACIAL_L_Masseter {parent=FACIAL_C_FacialRoot, loc=<6.7474 1.7801 3.7222>, rot=<0.0604 0.1128 0.0089 -0.9917>}
 FACIAL_R_Masseter {parent=FACIAL_C_FacialRoot, loc=<-6.7459 1.7821 3.7214>, rot=<0.0604 -0.1128 -0.0089 -0.9917>}
 FACIAL_C_UnderChin {parent=FACIAL_C_FacialRoot, loc=<0.0004 8.7617 4.3354>, rot=<0.492 0 0 -0.8706>}
 FACIAL_L_UnderChin {parent=FACIAL_C_FacialRoot, loc=<2.8692 7.4209 3.0053>, rot=<0.395 -0.3556 0.0791 -0.8434>}
 FACIAL_R_UnderChin {parent=FACIAL_C_FacialRoot, loc=<-2.8674 7.4197 3.0057>, rot=<0.395 0.3556 -0.0791 -0.8434>}
 FACIAL_C_TeethUpper {parent=FACIAL_C_FacialRoot, loc=<0.0003 5.6622 8.1656>, rot=<-0.009 0 0 -1>}
 FACIAL_C_LowerLipRotation {parent=FACIAL_C_FacialRoot, loc=<0.0007 2.3263 1.7413>, rot=<0.3336 0 0 -0.9427>}
 FACIAL_C_MouthLower {parent=FACIAL_C_LowerLipRotation, loc=<-0.0002 -1.3886 5.9162>, rot=<-0.342 0 0 -0.9397>}
 FACIAL_C_LipLower {parent=FACIAL_C_MouthLower, loc=<-0.0001 1.4533 1.6081>, rot=<0.0872 0 0 -0.9962>}
 FACIAL_L_LipLower {parent=FACIAL_C_MouthLower, loc=<0.6771 1.3644 1.4814>, rot=<0.0927 -0.1294 0.0272 -0.9869>}
 FACIAL_R_LipLower {parent=FACIAL_C_MouthLower, loc=<-0.6777 1.3653 1.4816>, rot=<0.0927 0.1294 -0.0272 -0.9869>}
 FACIAL_L_LipLowerOuter {parent=FACIAL_C_MouthLower, loc=<1.3097 1.1011 1.1738>, rot=<0.0913 -0.2321 0.0318 -0.9679>}
 FACIAL_R_LipLowerOuter {parent=FACIAL_C_MouthLower, loc=<-1.3114 1.1005 1.1725>, rot=<0.0913 0.2321 -0.0318 -0.9679>}
 FACIAL_C_TeethLower {parent=FACIAL_C_LowerLipRotation, loc=<-0.0003 -1.405 7.0296>, rot=<-0.342 0 0 -0.9397>}
 FACIAL_C_Tongue1 {parent=FACIAL_C_TeethLower, loc=<-0.0005 -0.0143 -4.754>, rot=<-0.2588 0 0 -0.9659>}
 FACIAL_C_Tongue2 {parent=FACIAL_C_Tongue1, loc=<0.0003 0.4252 1.3924>, rot=<0.2588 0 0 -0.9659>}
 FACIAL_C_Tongue3 {parent=FACIAL_C_Tongue2, loc=<0.0002 0.2433 1.2423>, rot=<0.1305 0 0 -0.9914>}
 FACIAL_C_Tongue4 {parent=FACIAL_C_Tongue3, loc=<0.0006 -0.1237 1.0632>, rot=<0.0436 0 0 -0.999>}
 FACIAL_C_TongueUpper3 {parent=FACIAL_C_Tongue3, loc=<0 -0.339 -0.2014>, rot=<0 0 0 1>}
 FACIAL_C_TongueLower3 {parent=FACIAL_C_Tongue3, loc=<0 0.8521 0.009>, rot=<0 0 0 1>}
 FACIAL_L_TongueSide3 {parent=FACIAL_C_Tongue3, loc=<1.556 -0.009 0.0793>, rot=<0 0 0 1>}
 FACIAL_R_TongueSide3 {parent=FACIAL_C_Tongue3, loc=<-1.5559 -0.0092 0.0795>, rot=<0 0 0 1>}
 FACIAL_C_TongueUpper2 {parent=FACIAL_C_Tongue2, loc=<-0.0001 -0.6115 -0.2769>, rot=<0 0 0 1>}
 FACIAL_L_TongueSide2 {parent=FACIAL_C_Tongue2, loc=<1.8719 -0.0351 0.0743>, rot=<0 0 0 1>}
 FACIAL_R_TongueSide2 {parent=FACIAL_C_Tongue2, loc=<-1.8718 -0.036 0.0748>, rot=<0 0 0 1>}
 FACIAL_C_Jaw {parent=FACIAL_C_FacialRoot, loc=<0.0007 2.3263 1.7413>, rot=<0.3336 0 0 -0.9427>}
 FACIAL_C_Jawline {parent=FACIAL_C_Jaw, loc=<-0.0008 2.1902 7.613>, rot=<0.2586 0 0 -0.966>}
 FACIAL_L_ChinSide {parent=FACIAL_C_Jaw, loc=<2.4133 0.2993 7.3799>, rot=<-0.2208 -0.3738 0.0499 -0.8995>}
 FACIAL_R_ChinSide {parent=FACIAL_C_Jaw, loc=<-2.4133 0.2982 7.3803>, rot=<-0.2208 0.3738 -0.0499 -0.8995>}
 FACIAL_C_Chin {parent=FACIAL_C_Jaw, loc=<-0.0006 0.7865 7.7913>, rot=<-0.2591 0 0 -0.9659>}
 lip_ring {parent=FACIAL_C_Jaw, loc=<0.3736 -1.2348 7.9016>, rot=<0.89 0 0 0.4559>}
 FACIAL_L_Ear {parent=FACIAL_C_FacialRoot, loc=<6.9788 -0.5351 0.167>, rot=<-0.0353 -0.0004 -0.0262 -0.999>}
 FACIAL_R_Ear {parent=FACIAL_C_FacialRoot, loc=<-6.9748 -0.5304 0.1692>, rot=<-0.0353 0.0005 0.0262 -0.999>}
 FACIAL_L_Jawline {parent=FACIAL_C_FacialRoot, loc=<4.3929 5.6561 4.3875>, rot=<0.2006 0.1582 -0.0532 -0.9653>}
 FACIAL_R_Jawline {parent=FACIAL_C_FacialRoot, loc=<-4.3933 5.6552 4.393>, rot=<0.2006 -0.1582 0.0532 -0.9653>}
 FACIAL_C_NoseBridge {parent=FACIAL_C_FacialRoot, loc=<0.0001 -1.3575 9.3544>, rot=<0.0348 0 0 -0.9994>}
 FACIAL_L_EyelidUpperB {parent=FACIAL_C_FacialRoot, loc=<3.3646 -0.9032 6.3856>, rot=<-0.2337 0 0 -0.9723>}
 FACIAL_L_EyelidLowerB {parent=FACIAL_C_FacialRoot, loc=<3.3646 -0.9032 6.3856>, rot=<0.3088 0 0 -0.9511>}
 FACIAL_R_EyelidLowerB {parent=FACIAL_C_FacialRoot, loc=<-3.3625 -0.9047 6.387>, rot=<0.317 0 0 -0.9484>}
 FACIAL_R_EyelidUpperB {parent=FACIAL_C_FacialRoot, loc=<-3.3625 -0.9047 6.387>, rot=<-0.2337 0 0 -0.9723>}
 clavicle_l {parent=spine_05, loc=<5.4831 -0.1053 -3.2946>, rot=<0.0465 -0.7289 -0.009 -0.6829>}
 upperarm_l {parent=clavicle_l, loc=<9.8658 -1.2136 -0.4105>, rot=<0.0164 -0.3245 -0.0269 -0.9454>}
 upperarm_twist_01_l {parent=upperarm_l, loc=<7.2135 -0.1134 0>, rot=<0 0 0 1>}
 upperarm_twist_02_l {parent=upperarm_l, loc=<19.7998 -0.1134 0>, rot=<0 0 0 1>}
 clavicle_r {parent=spine_05, loc=<5.4831 -0.1053 3.2946>, rot=<0.7289 0.0465 -0.6829 0.009>}
 upperarm_r {parent=clavicle_r, loc=<-9.8658 1.2136 0.4105>, rot=<0.0164 -0.3245 -0.0269 -0.9454>}
 upperarm_twist_01_r {parent=upperarm_r, loc=<-7.2135 0.1134 0>, rot=<0 0 0 -1>}
 sleeve_2_r {parent=upperarm_twist_01_r, loc=<0.4269 5.3259 2.2865>, rot=<0 -1 0 0>}
 sleeve_2_r1 {parent=sleeve_2_r, loc=<5.6414 0 0>, rot=<0 0 0 1>}
 deform_upperarm_bk_base_r {parent=upperarm_twist_01_r, loc=<6.4053 4.3746 -0.314>, rot=<0 0 0 -1>}
 deform_upperarm_bk_r {parent=deform_upperarm_bk_base_r, loc=<0 0 0>, rot=<0 0 0 1>}
 extra_spine_05 {parent=spine_05, loc=<0 0 0>, rot=<0 0 0 1>}
 bag_root {parent=spine_05, loc=<11.1051 -1.6608 -9.3698>, rot=<0.0284 -0.7065 -0.0284 -0.7065>}
 bag_front_01 {parent=bag_root, loc=<-0.8717 5.0598 -1.8569>, rot=<-0.4739 -0.2497 0.7977 0.2769>}
 bag_front_02 {parent=bag_front_01, loc=<7.8242 0 0>, rot=<-0.0125 0.0259 -0.0631 -0.9976>}
 bag_front_03 {parent=bag_front_02, loc=<7.9872 0 0>, rot=<0.061 0.0171 -0.146 -0.9873>}
 bag_front_04 {parent=bag_front_03, loc=<5.7188 0 0>, rot=<-0.1047 0.0202 -0.1879 -0.9764>}
 bag_front_05 {parent=bag_front_04, loc=<8.1103 0 0>, rot=<-0.0658 0.0061 -0.0917 -0.9936>}
 bag_front_06 {parent=bag_front_05, loc=<8.022 0 0>, rot=<0 0 0 1>}
 bag_back_01 {parent=bag_root, loc=<-1.5508 -3.7462 -1.7913>, rot=<-0.4594 0.1832 0.8598 -0.1272>}
 bag_back_02 {parent=bag_back_01, loc=<5.5424 0 0>, rot=<0 0 0.1808 -0.9835>}
 bag_back_03 {parent=bag_back_02, loc=<6.3957 0 0>, rot=<0.0074 0.0005 0.0666 -0.9978>}
 bag_back_04 {parent=bag_back_03, loc=<5.8846 0 0>, rot=<0.1516 0.0023 0.0149 -0.9883>}
 bag_back_05 {parent=bag_back_04, loc=<6.3129 0 0>, rot=<-0.0263 -0.0011 0.0437 -0.9987>}
 bag_back_06 {parent=bag_back_05, loc=<9.4374 0 0>, rot=<0 0 0 1>}
 deform_pec_aim_base_l {parent=spine_04, loc=<3.2456 9.3309 -1.2237>, rot=<0.0603 -0.4033 0.3318 -0.8507>}
 deform_pec_aim_l {parent=deform_pec_aim_base_l, loc=<0 0 0>, rot=<0 0 0 1>}
 deform_pec_base_l {parent=deform_pec_aim_l, loc=<9.3665 0 0>, rot=<0 0 0 1>}
 deform_pec_l {parent=deform_pec_base_l, loc=<0 0 0>, rot=<0 0 0 1>}
 deform_pec_aim_base_r {parent=spine_04, loc=<3.2452 9.3309 1.2237>, rot=<-0.4033 -0.0603 0.8507 0.3318>}
 deform_pec_aim_r {parent=deform_pec_aim_base_r, loc=<0 0 0>, rot=<0 0 0 1>}
 deform_pec_base_r {parent=deform_pec_aim_r, loc=<-9.3667 -0.0003 -0.0004>, rot=<0 0 0 1>}
 deform_pec_r {parent=deform_pec_base_r, loc=<0 0 0>, rot=<0 0 0 1>}
 dyn_bag_sim_chain_01_MSTR {parent=spine_04, loc=<-3.0683 1.178 3.6634>, rot=<0.2209 -0.6271 0.7355 0.1304>}
 dyn_bag_sim_chain_01 {parent=dyn_bag_sim_chain_01_MSTR, loc=<0 0 0>, rot=<0 0 0 1>}
 dyn_bag_sim_chain_02 {parent=dyn_bag_sim_chain_01, loc=<11.1795 0 0>, rot=<0 0 0 1>}
 dyn_bag_sim_chain_03 {parent=dyn_bag_sim_chain_02, loc=<0.3745 -0.2247 -0.0633>, rot=<0.1018 0.0185 0.2338 -0.9668>}
 hedgehog {parent=dyn_bag_sim_chain_03, loc=<15.1963 5.04 9.9258>, rot=<0.082 0.4136 -0.3511 -0.836>}
 eye_1 {parent=hedgehog, loc=<6.0428 -0.9119 0.398>, rot=<0.7705 -0.4222 0.3168 -0.3574>}
 eye_1_end {parent=eye_1, loc=<4.3233 0 0>, rot=<0 0 0 1>}
 eye_2 {parent=hedgehog, loc=<6.0428 -0.9119 0.398>, rot=<0.884 0.1025 0.4435 -0.1065>}
 eye_2_end {parent=eye_2, loc=<4.2253 0 0>, rot=<0 0 0 1>}
 eyelid_upper_2 {parent=hedgehog, loc=<6.0428 -0.9119 0.398>, rot=<-0.3065 0.4541 -0.1679 -0.8196>}
 joint2 {parent=eyelid_upper_2, loc=<4.3674 0 0>, rot=<0 0 0 1>}
 eyelid_upper_1 {parent=hedgehog, loc=<6.0428 -0.9119 0.398>, rot=<-0.1832 0.2913 0.4882 -0.802>}
 joint4 {parent=eyelid_upper_1, loc=<4.471 0 0>, rot=<0 0 0 1>}
 eyelid_lower_2 {parent=hedgehog, loc=<6.0428 -0.9119 0.398>, rot=<-0.2585 0.2304 -0.2219 -0.9115>}
 joint6 {parent=eyelid_lower_2, loc=<4.2473 0 0>, rot=<0 0 0 1>}
 eyelid_lower_1 {parent=hedgehog, loc=<6.0428 -0.9119 0.398>, rot=<-0.3301 0.0912 0.4257 -0.8375>}
 joint8 {parent=eyelid_lower_1, loc=<4.3502 0 0>, rot=<0 0 0 1>}
 zipper {parent=dyn_bag_sim_chain_03, loc=<8.8014 2.572 -10.3533>, rot=<0.6318 0.2623 -0.2672 -0.6787>}
 BASE_crop_top_1 {parent=spine_03, loc=<4.0646 13.0723 -4.9185>, rot=<0.0547 -0.9985 0 0>}
 crop_top_1 {parent=BASE_crop_top_1, loc=<0 0 0>, rot=<0 0 0 1>}
 crop_top_1_end {parent=crop_top_1, loc=<3 0 0>, rot=<0 0 0 1>}
 BASE_crop_top_2 {parent=spine_03, loc=<4.6487 7.7601 -11.1887>, rot=<0.0468 -0.8545 0.5165 -0.0283>}
 crop_top_2 {parent=BASE_crop_top_2, loc=<0 0 0>, rot=<0 0 0 1>}
 crop_top_2_end {parent=crop_top_2, loc=<3 0 0>, rot=<0 0 0 1>}
 BASE_crop_top_3 {parent=spine_03, loc=<5.543 -0.3724 -12.749>, rot=<0.0387 -0.706 0.706 -0.0387>}
 crop_top_3 {parent=BASE_crop_top_3, loc=<0 0 0>, rot=<0 0 0 1>}
 crop_top_3_end {parent=crop_top_3, loc=<3 0 0>, rot=<0 0 0 1>}
 BASE_crop_top_4 {parent=spine_03, loc=<6.1231 -5.6474 -10.3082>, rot=<0.0141 -0.2567 0.9649 -0.0529>}
 crop_top_4 {parent=BASE_crop_top_4, loc=<0 0 0>, rot=<0 0 0 1>}
 crop_top_4_end {parent=crop_top_4, loc=<3 0 0>, rot=<0 0 0 1>}
 BASE_crop_top_5 {parent=spine_03, loc=<6.6111 -8.9415 -4.1089>, rot=<0.003 -0.055 0.997 -0.0547>}
 crop_top_5 {parent=BASE_crop_top_5, loc=<0 0 0>, rot=<0 0 0 1>}
 crop_top_5_end {parent=crop_top_5, loc=<3 0 0>, rot=<0 0 0 1>}
 BASE_crop_top_6 {parent=spine_03, loc=<6.1593 -5.6287 11.1298>, rot=<-0.0249 0.4534 0.8896 -0.0488>}
 crop_top_6 {parent=BASE_crop_top_6, loc=<0 0 0>, rot=<0 0 0 1>}
 crop_top_6_end {parent=crop_top_6, loc=<3 0 0>, rot=<0 0 0 1>}
 BASE_crop_top_7 {parent=spine_03, loc=<5.4153 0.8082 12.9951>, rot=<-0.0381 0.6954 0.7165 -0.0393>}
 crop_top_7 {parent=BASE_crop_top_7, loc=<0 0 0>, rot=<0 0 0 1>}
 crop_top_7_end {parent=crop_top_7, loc=<3 0 0>, rot=<0 0 0 1>}
 BASE_crop_top_8 {parent=spine_03, loc=<4.5051 10.6081 10.0168>, rot=<0.0475 -0.8672 -0.4949 0.0271>}
 crop_top_8 {parent=BASE_crop_top_8, loc=<0 0 0>, rot=<0 0 0 1>}
 crop_top_8_end {parent=crop_top_8, loc=<3 0 0>, rot=<0 0 0 1>}
 sleeve_l {type=Curve, loc=<0 0 0>, rot=<0 0 0 1>}
 sleeve_r {type=Curve, loc=<0 0 0>, rot=<0 0 0 1>}
 tentacle_back_l_A_02_ctrl_null {type=Null, loc=<6.6989 -7.2601 150.1521>, rot=<-0.3917 -0.417 0.3953 -0.7186>}
 tentacle_back_l_A_02_ctrl {
  controlType=EULER_TRANSFORM
  parent=tentacle_back_l_A_02_ctrl_null
  type=Control
  loc=<132.4565 -38.2651 -60.2853>
  rot=<0.3917 0.417 -0.3953 -0.7186>
 }
 tentacle_back_l_A_03_ctrl_null {parent=tentacle_back_l_A_02_ctrl, type=Null, loc=<3.7887 0 0>, rot=<-0.0108 0.1528 0.0891 -0.9842>}
 tentacle_back_l_A_03_ctrl {
  controlType=EULER_TRANSFORM
  parent=tentacle_back_l_A_03_ctrl_null
  type=Control
  loc=<-3.5516 -0.6518 1.147>
  rot=<0.0108 -0.1528 -0.0891 -0.9842>
 }
 extra_tentacle_back_l_A_04_ctrl_null {parent=tentacle_back_l_A_03_ctrl, type=Null, loc=<4.6748 0 0>, rot=<-0.0466 -0.0024 0.1201 -0.9917>}
 extra_tentacle_back_l_A_04_ctrl {
  controlType=EULER_TRANSFORM
  parent=extra_tentacle_back_l_A_04_ctrl_null
  type=Control
  loc=<-4.54 -1.1144 0.0298>
  rot=<0.0466 0.0024 -0.1201 -0.9917>
 }
 extra_tentacle_back_l_A_05_ctrl_null {parent=extra_tentacle_back_l_A_04_ctrl, type=Null, loc=<4.8885 0 0>, rot=<-0.0697 -0.0634 0.1277 -0.9873>}
 extra_tentacle_back_l_A_05_ctrl {
  controlType=EULER_TRANSFORM
  parent=extra_tentacle_back_l_A_05_ctrl_null
  type=Control
  loc=<-4.6898 -1.2755 -0.5252>
  rot=<0.0697 0.0634 -0.1277 -0.9873>
 }
 extra_tentacle_back_l_A_06_ctrl_null {parent=extra_tentacle_back_l_A_05_ctrl, type=Null, loc=<5.2483 0 0>, rot=<0.1011 -0.0183 -0.1772 -0.9788>}
 extra_tentacle_back_l_A_06_ctrl {
  controlType=EULER_TRANSFORM
  parent=extra_tentacle_back_l_A_06_ctrl_null
  type=Control
  loc=<-4.9152 1.8399 0>
  rot=<-0.1011 0.0183 0.1772 -0.9788>
 }
]
variables [
 backpack_cape_mod {type=double, default=0}
 backpack_cape_mod_bool {type=bool, default=False}
 jump_sprint {type=double, default=0}
 l_constraint_float {type=double, default=0}
 play_f {type=double, default=0}
 r_constraint_float {type=double, default=0}
 scanning_l {type=double, default=0}
 scanning_r {type=double, default=0}
 sp_L {type=double, default=0}
 sp_R {type=double, default=0}
]
