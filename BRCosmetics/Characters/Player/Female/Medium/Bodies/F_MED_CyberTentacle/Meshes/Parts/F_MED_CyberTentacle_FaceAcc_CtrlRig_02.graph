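# Control Rig for the CyberTentacle female character's face accessories, correcting head tentacle base bones and the lip ring.
# One sequence branch rotates the front-left tentacle by `play_f`, then parent-constrains `lip_ring` to weighted facial bones.
# The other offsets back tentacles through spherical pose readers: right clavicle scaled by `crouching`, then head using `test_01`/`test_02` offsets.
# A smoothed clavicle-driven front-right tentacle rotation exists but is disconnected from execution.
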ControlRigBlueprint /BRCosmetics/Characters/Player/Female/Medium/Bodies/F_MED_CyberTentacle/Meshes/Parts/F_MED_CyberTentacle_FaceAcc_CtrlRig_02
graphs [
 RigVMFunctionLibrary {links [], nodes []}
 RigVMModel {nodes [
   RigUnit_BeginExecution = &RigUnit_BeginExecution()
   RigVMFunction_Sequence = &RigVMFunction_Sequence(ExecuteContext=@RigUnit_BeginExecution.ExecutePin)
   n1 = &RigUnit_SphericalPoseReader(
    ActiveRegionScaleFactors=(PositiveWidth=1,NegativeWidth=0.1,PositiveHeight=1,NegativeHeight=1)
    ActiveRegionSize=0.3
    Debug=(bDrawDebug=True,bDraw2D=false,bDrawLocalAxes=false,DebugScale=30,DebugSegments=30,DebugThickness=0.2)
    DriverAxis=(X=-1,Y=0,Z=0)
    DriverItem=(Type=Bone,Name="clavicle_r")
    FalloffRegionScaleFactors=(PositiveWidth=1,NegativeWidth=0.4,PositiveHeight=1,NegativeHeight=1)
    FalloffSize=0.5
    OptionalParentItem=(Type=None,Name="pelvis")
    RotationOffset=(X=0,Y=170,Z=0)
    ExecutePin=@RigVMFunction_Sequence.B
   )
   Multiply = &RigVMFunction_MathDoubleMul(A=@n1.OutputParam, B=RigVMVariableNode() {variable=crouching}.Value)
   > &RigUnit_ModifyTransforms(
    ItemToModify="((Item=(Type=Bone,Name=\"BASE_tentacle_back_r_A_01\"),Transform=(Rotation=(X=0,Y=0,Z=0,W=1),Translation=(X=-2,Y=0,Z=2),Scale3D=(X=1,Y=1,Z=1))),(Item=(Type=Bone,Name=\"BASE_tentacle_back_r_B_01\"),Transform=(Rotation=(X=0,Y=0,Z=0,W=1),Translation=(X=-2,Y=0,Z=2) ...(truncated)"
    WeightMinimum=0
    Weight=@Multiply.Result
   )
   > n2 = &RigUnit_SphericalPoseReader(
    ActiveRegionScaleFactors=(PositiveWidth=1,NegativeWidth=0.1,PositiveHeight=1,NegativeHeight=1)
    ActiveRegionSize=0.3
    Debug=(bDrawDebug=False,bDraw2D=false,bDrawLocalAxes=false,DebugScale=30,DebugSegments=30,DebugThickness=0.2)
    DriverAxis=(X=1,Y=0,Z=0)
    DriverItem=(Type=Bone,Name="head")
    FalloffRegionScaleFactors=(PositiveWidth=1,NegativeWidth=0.4,PositiveHeight=1,NegativeHeight=1)
    FalloffSize=0.5
    OptionalParentItem=(Type=None,Name="pelvis")
    RotationOffset=(X=-90,Y=180,Z=40)
   )
   n3 = RigVMVariableNode() {variable=test_01}
   n4 = RigVMVariableNode() {variable=test_02}
   > &RigUnit_ModifyTransforms(
    ItemToModify="((Item=(Type=Bone,Name=\"BASE_tentacle_back_r_A_01\"),Transform=(Rotation=(X=0,Y=0,Z=0,W=1),Translation=(X=-4,Y=0,Z=1),Scale3D=(X=1,Y=1,Z=1))),(Item=(Type=Bone,Name=\"BASE_tentacle_back_r_B_01\"),Transform=(Rotation=(X=0,Y=0,Z=0,W=1),Translation=(X=-4,Y=0,Z=1) ...(truncated)"
    WeightMinimum=0
    Weight=@n2.OutputParam
    ItemToModify.0.Transform.Translation.X=@n3.Value
    ItemToModify.1.Transform.Translation.X=@n3.Value
    ItemToModify.2.Transform.Translation.X=@n3.Value
    ItemToModify.3.Transform.Translation.X=@n3.Value
    ItemToModify.0.Transform.Translation.Z=@n4.Value
    ItemToModify.1.Transform.Translation.Z=@n4.Value
    ItemToModify.2.Transform.Translation.Z=@n4.Value
    ItemToModify.3.Transform.Translation.Z=@n4.Value
   )
   &RigUnit_ModifyTransforms(
    ItemToModify=((Item=(Type=Bone,Name="BASE_tentacle_front_l_A_01"),Transform=(Rotation=(X=0,Y=0,Z=0,W=1),Translation=(X=0,Y=0,Z=0),Scale3D=(X=1,Y=1,Z=1))))
    WeightMinimum=0
    ItemToModify.0.Transform.Rotation=&RigVMFunction_MathQuaternionFromEuler(Euler=(X=0,Y=-50,Z=0)).Result
    Weight=RigVMVariableNode() {variable=play_f}.Value
    ExecutePin=@RigVMFunction_Sequence.A
   )
   > &RigUnit_ParentConstraint(
    AdvancedSettings=(RotationOrderForFilter=XZY)
    Child=(Type=Bone,Name="lip_ring")
    Parents="((Item=(Type=Bone,Name=\"FACIAL_C_Chin\"),Weight=0.0037),(Item=(Type=Bone,Name=\"FACIAL_L_LipCorner\"),Weight=0.0069),(Item=(Type=Bone,Name=\"FACIAL_R_LipLower\"),Weight=0.01),(Item=(Type=Bone,Name=\"FACIAL_L_LipLowerOuter\"),Weight=0.0178),(Item=(Type=Bone,Name=\" ...(truncated)"
    Weight=1
   )
   n5 = &RigUnit_SphericalPoseReader(
    ActiveRegionScaleFactors=(PositiveWidth=1,NegativeWidth=0.5,PositiveHeight=1,NegativeHeight=1)
    ActiveRegionSize=0.3
    Debug=(bDrawDebug=False,bDraw2D=false,bDrawLocalAxes=false,DebugScale=30,DebugSegments=30,DebugThickness=0.2)
    DriverAxis=(X=-1,Y=0,Z=0)
    DriverItem=(Type=Bone,Name="clavicle_r")
    ExecutePin=()
    FalloffRegionScaleFactors=(PositiveWidth=1,NegativeWidth=0.2,PositiveHeight=1,NegativeHeight=1)
    FalloffSize=0.5
    OptionalParentItem=(Type=Bone,Name="None")
    RotationOffset=(X=0,Y=220,Z=180)
   )
   n6 = &RigVMFunction_AlphaInterp(InterpSpeedDecreasing=0.1, InterpSpeedIncreasing=50, Scale=1, bClampResult=False, Value=@n5.OutputParam)
   > &RigUnit_ModifyTransforms(
    ItemToModify=((Item=(Type=Bone,Name="BASE_tentacle_front_r_A_01"),Transform=(Rotation=(X=0,Y=0,Z=0,W=1),Translation=(X=0,Y=0,Z=0),Scale3D=(X=1,Y=1,Z=1))))
    WeightMinimum=0
    Weight=@n6.Result
    ItemToModify.0.Transform.Rotation=&RigVMFunction_MathQuaternionFromEuler(Euler=(X=0,Y=50,Z=0)).Result
   )
  ]}
]
hierarchy [
 root {loc=<0 0 0>, rot=<0 0 0 1>}
 pelvis {parent=root, loc=<0 -0.5301 98.0571>, rot=<0.0127 0.707 -0.0127 -0.707>}
 spine_01 {parent=pelvis, loc=<6.1208 -0.9328 0>, rot=<0 0 0.034 -0.9994>}
 spine_02 {parent=spine_01, loc=<6.5997 -0.2095 0>, rot=<0 0 0.0238 -0.9997>}
 spine_03 {parent=spine_02, loc=<6.4914 -0.1237 0>, rot=<0 0 0.0149 -0.9999>}
 spine_04 {parent=spine_03, loc=<6.9598 0.0794 0>, rot=<0 0 -0.0132 -0.9999>}
 spine_05 {parent=spine_04, loc=<6.9987 0 0>, rot=<0 0 -0.0015 -1>}
 neck_01 {parent=spine_05, loc=<11.3138 0.1529 0>, rot=<0 0 -0.2003 -0.9797>}
 neck_02 {parent=neck_01, loc=<4.5627 0 0>, rot=<0 0 0.0504 -0.9987>}
 head {parent=neck_02, loc=<5.3914 0 0>, rot=<0 0 0.115 -0.9934>}
 base_hat {parent=head, loc=<5.3677 0.059 0>, rot=<0.4731 -0.881 0 0>}
 BASE_dyn_hat {parent=base_hat, loc=<0 0 0>, rot=<0 0 0 1>}
 dyn_hat {parent=BASE_dyn_hat, loc=<0 0 0>, rot=<0 0 0 1>}
 base_Beannie {parent=dyn_hat, loc=<2.0038 -5.9678 0>, rot=<0 0 -0.0824 -0.9966>}
 dyn_Beannie {parent=base_Beannie, loc=<0 0 0>, rot=<0 0 0 1>}
 BeannieEnd {parent=dyn_Beannie, loc=<12.9883 0 0>, rot=<0 0 0 1>}
 BASE_tentacle_back_l_A_01 {parent=head, loc=<1.9998 -5.507 -6.5638>, rot=<-0.0033 -0.7635 0.6455 -0.0203>}
 tentacle_back_l_A_01 {parent=BASE_tentacle_back_l_A_01, loc=<0 0 0>, rot=<0 0 0 1>}
 tentacle_back_l_A_02 {parent=tentacle_back_l_A_01, loc=<3.839 0 0>, rot=<0.0949 0.154 -0.1208 -0.9761>}
 tentacle_back_l_A_03 {parent=tentacle_back_l_A_02, loc=<3.7887 0 0>, rot=<-0.0108 0.1528 0.0891 -0.9842>}
 extra_tentacle_back_l_A_03 {parent=tentacle_back_l_A_03, loc=<0 0 0>, rot=<0 0 0 1>}
 extra_tentacle_back_l_A_04 {parent=extra_tentacle_back_l_A_03, loc=<4.6748 0 0>, rot=<-0.0466 -0.0024 0.1201 -0.9917>}
 tentacle_back_l_A_04 {parent=extra_tentacle_back_l_A_04, loc=<0 0 0>, rot=<0 0 0 1>}
 extra_tentacle_back_l_A_05 {parent=tentacle_back_l_A_04, loc=<4.8885 0 0>, rot=<-0.0697 -0.0634 0.1277 -0.9873>}
 tentacle_back_l_A_05 {parent=extra_tentacle_back_l_A_05, loc=<0 0 0>, rot=<0 0 0 1>}
 extra_tentacle_back_l_A_06 {parent=tentacle_back_l_A_05, loc=<5.2483 0 0>, rot=<0.1011 -0.0183 -0.1772 -0.9788>}
 tentacle_back_l_A_06 {parent=extra_tentacle_back_l_A_06, loc=<0 0 0>, rot=<0 0 0 1>}
 BASE_tentacle_back_r_A_01 {parent=head, loc=<1.6393 -6.4682 6.0603>, rot=<0.0544 -0.5147 0.8549 0.037>}
 tentacle_back_r_A_01 {parent=BASE_tentacle_back_r_A_01, loc=<0 0 0>, rot=<-0.0609 -0.087 -0.0053 -0.9943>}
 tentacle_back_r_A_02 {parent=tentacle_back_r_A_01, loc=<5.0348 0.0399 0.3952>, rot=<0.0481 0.2881 -0.0145 -0.9563>}
 tentacle_back_r_A_03 {parent=tentacle_back_r_A_02, loc=<3.7523 0 0>, rot=<0.0354 0.1498 -0.0194 -0.9879>}
 extra_tentacle_back_r_A_03 {parent=tentacle_back_r_A_03, loc=<0 0 0>, rot=<0 0 0 1>}
 BASE_tentacle_back_r_A_04 {parent=extra_tentacle_back_r_A_03, loc=<4.5945 0 0>, rot=<0.0057 0.0249 -0.1942 -0.9806>}
 tentacle_back_r_A_04 {parent=BASE_tentacle_back_r_A_04, loc=<0 0 0>, rot=<0 0 0 1>}
 BASE_tentacle_back_r_A_05 {parent=tentacle_back_r_A_04, loc=<4.3915 0 0>, rot=<-0.0652 -0.1153 -0.293 -0.9469>}
 tentacle_back_r_A_05 {parent=BASE_tentacle_back_r_A_05, loc=<0 0 0>, rot=<0 0 0 1>}
 BASE_tentacle_back_r_A_06 {parent=tentacle_back_r_A_05, loc=<4.602 0 0>, rot=<-0.0107 -0.2033 -0.0516 -0.9777>}
 tentacle_back_r_A_06 {parent=BASE_tentacle_back_r_A_06, loc=<0 0 0>, rot=<0 0 0 1>}
 extra_tentacle_back_r_A_02 {parent=tentacle_back_r_A_02, loc=<0 0 0>, rot=<0 0 0 1>}
 extra_tentacle_back_r_A_01 {parent=tentacle_back_r_A_01, loc=<0 0 0>, rot=<0 0 0 1>}
 extra_tentacle_back_r_A_01_end {parent=extra_tentacle_back_r_A_01, loc=<2.2767 0 0>, rot=<0 0 0 1>}
 BASE_tentacle_back_r_B_01 {parent=head, loc=<1.3481 -7.8211 2.0209>, rot=<0.0352 -0.7043 0.7087 0.0205>}
 tentacle_back_r_B_01 {parent=BASE_tentacle_back_r_B_01, loc=<0 0 0>, rot=<-0.0609 -0.087 -0.0053 -0.9943>}
 tentacle_back_r_B_02 {parent=tentacle_back_r_B_01, loc=<5.0348 0.0399 0.3952>, rot=<0.0488 0.239 -0.012 -0.9697>}
 tentacle_back_r_B_03 {parent=tentacle_back_r_B_02, loc=<3.7523 0 0>, rot=<0.0354 0.1498 -0.0194 -0.9879>}
 tentacle_back_r_B_04 {parent=tentacle_back_r_B_03, loc=<4.5945 0 0>, rot=<0.021 -0.0349 -0.1976 -0.9794>}
 extra_tentacle_back_r_B_04 {parent=tentacle_back_r_B_04, loc=<0 0 0>, rot=<0 0 0 1>}
 BASE_tentacle_back_r_B_05 {parent=extra_tentacle_back_r_B_04, loc=<4.3915 0 0>, rot=<0.0468 -0.2288 0.011 -0.9723>}
 tentacle_back_r_B_05 {parent=BASE_tentacle_back_r_B_05, loc=<0 0 0>, rot=<0 0 0 1>}
 BASE_tentacle_back_r_B_06 {parent=tentacle_back_r_B_05, loc=<4.602 0 0>, rot=<-0.0415 -0.1492 0.8557 -0.4937>}
 tentacle_back_r_B_06 {parent=BASE_tentacle_back_r_B_06, loc=<0 0 0>, rot=<0.1594 0.1511 -0.4387 -0.8714>}
 BASE_tentacle_back_l_B_01 {parent=head, loc=<1.863 -6.2387 -2.2394>, rot=<-0.0033 -0.7635 0.6455 -0.0203>}
 tentacle_back_l_B_01 {parent=BASE_tentacle_back_l_B_01, loc=<0 0 0>, rot=<-0.0612 0 0 -0.9981>}
 tentacle_back_l_B_02 {parent=tentacle_back_l_B_01, loc=<3.839 0 0>, rot=<0.0496 0.1595 -0.008 -0.9859>}
 tentacle_back_l_B_03 {parent=tentacle_back_l_B_02, loc=<3.7523 0 0>, rot=<0.0354 0.1498 -0.0194 -0.9879>}
 tentacle_back_l_B_04 {parent=tentacle_back_l_B_03, loc=<4.5945 0 0>, rot=<-0.0673 0.186 0.0128 -0.9802>}
 tentacle_back_l_B_05 {parent=tentacle_back_l_B_04, loc=<4.3915 0 0>, rot=<0.0463 0.2642 -0.0127 -0.9633>}
 extra_tentacle_back_l_B_05 {parent=tentacle_back_l_B_05, loc=<0 0 0>, rot=<0 0 0 1>}
 BASE_tentacle_back_l_B_06 {parent=extra_tentacle_back_l_B_05, loc=<4.602 0 0>, rot=<0 -0.2036 0 -0.9791>}
 tentacle_back_l_B_06 {parent=BASE_tentacle_back_l_B_06, loc=<0 0 0>, rot=<0 0 0 1>}
 BASE_tentacle_back_l_B_07 {parent=tentacle_back_l_B_06, loc=<3.4145 0 0>, rot=<0 0 0 1>}
 tentacle_back_l_B_07 {parent=BASE_tentacle_back_l_B_07, loc=<0 0 0>, rot=<0 0 0 1>}
 BASE_tentacle_back_l_B_08 {parent=tentacle_back_l_B_07, loc=<1.0419 0 -2.1533>, rot=<0 0 0 1>}
 tentacle_back_l_B_08 {parent=BASE_tentacle_back_l_B_08, loc=<0 0 0>, rot=<0 0 0 1>}
 BASE_tentacle_front_r_A_01 {parent=head, loc=<-3.1509 -1.8827 7.5106>, rot=<-0.0883 -0.9322 0.3228 0.1375>}
 tentacle_front_r_A_01 {parent=BASE_tentacle_front_r_A_01, loc=<0 0 0>, rot=<0 0 0 1>}
 tentacle_front_r_A_02 {parent=tentacle_front_r_A_01, loc=<3.7919 0 0>, rot=<0 -0.1517 0 -0.9884>}
 extra_tentacle_front_r_A_02 {parent=tentacle_front_r_A_02, loc=<0 0 0>, rot=<0 0 0 1>}
 BASE_tentacle_front_r_A_03 {parent=extra_tentacle_front_r_A_02, loc=<4.2634 0 0>, rot=<0.1117 -0.0411 0.0357 -0.9922>}
 tentacle_front_r_A_03 {parent=BASE_tentacle_front_r_A_03, loc=<0 0 0>, rot=<0 0 0 1>}
 extra_tentacle_front_r_A_03 {parent=tentacle_front_r_A_03, loc=<0 0 0>, rot=<0 0 0 1>}
 BASE_tentacle_front_r_A_04 {parent=tentacle_front_r_A_03, loc=<3.6184 0 0>, rot=<0.1329 0.0968 -0.0131 -0.9863>}
 tentacle_front_r_A_04 {parent=BASE_tentacle_front_r_A_04, loc=<0 0 0>, rot=<0 0 0 1>}
 extra_tentacle_front_r_A_04 {parent=tentacle_front_r_A_04, loc=<0 0 0>, rot=<0 0 0 1>}
 BASE_tentacle_front_r_A_05 {parent=tentacle_front_r_A_04, loc=<3.5936 0 0>, rot=<0 0 0 1>}
 tentacle_front_r_A_05 {parent=BASE_tentacle_front_r_A_05, loc=<0 0 0>, rot=<0 0 0 1>}
 extra_tentacle_front_r_A_05 {parent=tentacle_front_r_A_05, loc=<0 0 0>, rot=<0 0 0 1>}
 extra_tentacle_front_r_A_01 {parent=tentacle_front_r_A_01, loc=<0 0 0>, rot=<0 0 0 1>}
 BASE_tentacle_front_r_B_01 {parent=head, loc=<-2.0465 3.2033 7.2709>, rot=<-0.0573 -0.9416 0.329 -0.0428>}
 tentacle_front_r_B_01 {parent=BASE_tentacle_front_r_B_01, loc=<0 0 0>, rot=<0 0 0 1>}
 tentacle_front_r_B_02 {parent=tentacle_front_r_B_01, loc=<5.5529 0 0>, rot=<0 -0.2558 0 -0.9667>}
 extra_tentacle_front_r_B_02 {parent=tentacle_front_r_B_02, loc=<0 0 0>, rot=<0 0 0 1>}
 BASE_tentacle_front_r_B_03 {parent=extra_tentacle_front_r_B_02, loc=<5.5326 0 0>, rot=<0.0302 -0.0565 0.0554 -0.9964>}
 tentacle_front_r_B_03 {parent=BASE_tentacle_front_r_B_03, loc=<0 0 0>, rot=<0 0 0 1>}
 extra_tentacle_front_r_B_03 {parent=tentacle_front_r_B_03, loc=<0 0 0>, rot=<0 0 0 1>}
 BASE_tentacle_front_r_B_04 {parent=tentacle_front_r_B_03, loc=<4.2789 0 0>, rot=<-0.0293 0.117 0.0035 -0.9927>}
 tentacle_front_r_B_04 {parent=BASE_tentacle_front_r_B_04, loc=<0 0 0>, rot=<0 0 0 1>}
 extra_tentacle_front_r_B_04 {parent=tentacle_front_r_B_04, loc=<0 0 0>, rot=<0 0 0 1>}
 BASE_tentacle_front_r_B_05 {parent=tentacle_front_r_B_04, loc=<3.7261 0 0>, rot=<0 0 0 1>}
 tentacle_front_r_B_05 {parent=BASE_tentacle_front_r_B_05, loc=<0 0 0>, rot=<0 0 0 1>}
 extra_tentacle_front_r_B_05 {parent=tentacle_front_r_B_05, loc=<0 0 0>, rot=<0 0 0 1>}
 extra_tentacle_front_r_B_01 {parent=tentacle_front_r_B_01, loc=<0 0 0>, rot=<0 0 0 1>}
 BASE_tentacle_front_l_B_01 {parent=head, loc=<-1.4879 4.2361 -7.5252>, rot=<-0.1177 -0.9597 0.2505 0.0494>}
 tentacle_front_l_B_01 {parent=BASE_tentacle_front_l_B_01, loc=<0 0 0>, rot=<0 0 0 1>}
 tentacle_front_l_B_02 {parent=tentacle_front_l_B_01, loc=<6.2575 0 0>, rot=<0 0.1455 0 -0.9894>}
 extra_tentacle_front_l_B_02 {parent=tentacle_front_l_B_02, loc=<0 0 0>, rot=<0 0 0 1>}
 BASE_tentacle_front_l_B_03 {parent=extra_tentacle_front_l_B_02, loc=<4.1508 0 0>, rot=<0.3298 0.1744 -0.0621 -0.9257>}
 tentacle_front_l_B_03 {parent=BASE_tentacle_front_l_B_03, loc=<0 0 0>, rot=<0 0 0 1>}
 extra_tentacle_front_l_B_03 {parent=tentacle_front_l_B_03, loc=<0 0 0>, rot=<0 0 0 1>}
 BASE_tentacle_front_l_B_04 {parent=tentacle_front_l_B_03, loc=<4.7436 0 0>, rot=<0.0171 -0.1391 0.0024 -0.9901>}
 tentacle_front_l_B_04 {parent=BASE_tentacle_front_l_B_04, loc=<0 0 0>, rot=<0 0 0 1>}
 extra_tentacle_front_l_B_04 {parent=tentacle_front_l_B_04, loc=<0 0 0>, rot=<0 0 0 1>}
 BASE_tentacle_front_l_B_05 {parent=tentacle_front_l_B_04, loc=<3.8267 0 0>, rot=<0 -0.1349 0 -0.9909>}
 tentacle_front_l_B_05 {parent=BASE_tentacle_front_l_B_05, loc=<0 0 0>, rot=<0 0 0 1>}
 extra_tentacle_front_l_B_05 {parent=tentacle_front_l_B_05, loc=<0 0 0>, rot=<0 0 0 1>}
 extra_tentacle_front_l_B_01 {parent=tentacle_front_l_B_01, loc=<0 0 0>, rot=<0 0 0 1>}
 BASE_tentacle_front_l_A_01 {parent=head, loc=<-3.682 -0.311 -7.5493>, rot=<-0.1226 -0.9765 -0.1653 -0.0631>}
 tentacle_front_l_A_01 {parent=BASE_tentacle_front_l_A_01, loc=<0 0 0>, rot=<0 0 0 1>}
 tentacle_front_l_A_02 {parent=tentacle_front_l_A_01, loc=<3.5258 0 0>, rot=<0 0.2107 0 -0.9776>}
 extra_tentacle_front_l_A_02 {parent=tentacle_front_l_A_02, loc=<0 0 0>, rot=<0 0 0 1>}
 BASE_tentacle_front_l_A_03 {parent=extra_tentacle_front_l_A_02, loc=<3.7671 0 0>, rot=<-0.2776 0.119 0.0347 -0.9527>}
 tentacle_front_l_A_03 {parent=BASE_tentacle_front_l_A_03, loc=<0 0 0>, rot=<0 0 0 1>}
 extra_tentacle_front_l_A_03 {parent=tentacle_front_l_A_03, loc=<0 0 0>, rot=<0 0 0 1>}
 BASE_tentacle_front_l_A_04 {parent=tentacle_front_l_A_03, loc=<4.4844 0 0>, rot=<0.0923 -0.1977 0.0187 -0.9757>}
 tentacle_front_l_A_04 {parent=BASE_tentacle_front_l_A_04, loc=<0 0 0>, rot=<0 0 0 1>}
 extra_tentacle_front_l_A_04 {parent=tentacle_front_l_A_04, loc=<0 0 0>, rot=<0 0 0 1>}
 BASE_tentacle_front_l_A_05 {parent=tentacle_front_l_A_04, loc=<3.2938 0 0>, rot=<0 -0.1158 0 -0.9933>}
 tentacle_front_l_A_05 {parent=BASE_tentacle_front_l_A_05, loc=<0 0 0>, rot=<0 0 0 1>}
 extra_tentacle_front_l_A_05 {parent=tentacle_front_l_A_05, loc=<0 0 0>, rot=<0 0 0 1>}
 extra_tentacle_front_l_A_01 {parent=tentacle_front_l_A_01, loc=<0 0 0>, rot=<0 0 0 1>}
 FACIAL_C_FacialRoot {parent=head, loc=<2.1524 -0.746 -0.0002>, rot=<0.4879 -0.5118 -0.4879 -0.5118>}
 FACIAL_C_Forehead {parent=FACIAL_C_FacialRoot, loc=<-0.0014 -3.3461 8.7089>, rot=<-0.0524 0 0 -0.9986>}
 FACIAL_L_ForeheadIn {parent=FACIAL_C_FacialRoot, loc=<2.0694 -3.3307 8.5684>, rot=<-0.0697 -0.0696 0.0036 -0.9951>}
 FACIAL_R_ForeheadIn {parent=FACIAL_C_FacialRoot, loc=<-2.0723 -3.3303 8.5648>, rot=<-0.0697 0.0696 -0.0036 -0.9951>}
 FACIAL_L_ForeheadMid {parent=FACIAL_C_FacialRoot, loc=<3.8266 -3.7005 7.9466>, rot=<-0.0798 -0.1755 -0.0222 -0.981>}
 FACIAL_R_ForeheadMid {parent=FACIAL_C_FacialRoot, loc=<-3.8272 -3.6992 7.9419>, rot=<-0.0798 0.1755 0.0222 -0.981>}
 FACIAL_L_ForeheadOut {parent=FACIAL_C_FacialRoot, loc=<5.3962 -3.671 6.9286>, rot=<-0.13 -0.2549 -0.0814 -0.9547>}
 FACIAL_R_ForeheadOut {parent=FACIAL_C_FacialRoot, loc=<-5.396 -3.6708 6.9253>, rot=<-0.13 0.2549 0.0814 -0.9547>}
 FACIAL_L_EyesackUpper {parent=FACIAL_C_FacialRoot, loc=<3.9165 -2.6545 8.5218>, rot=<0.0399 -0.0836 -0.0653 -0.9936>}
 FACIAL_R_EyesackUpper {parent=FACIAL_C_FacialRoot, loc=<-3.9195 -2.6554 8.522>, rot=<0.0399 0.0837 0.0653 -0.9936>}
 FACIAL_L_EyelidUpperFurrow {parent=FACIAL_C_FacialRoot, loc=<3.7839 -2.3743 8.4671>, rot=<-0.0389 -0.1501 -0.081 -0.9846>}
 FACIAL_R_EyelidUpperFurrow {parent=FACIAL_C_FacialRoot, loc=<-3.7871 -2.3761 8.4677>, rot=<-0.0389 0.1501 0.081 -0.9846>}
 FACIAL_L_EyelidUpperA {parent=FACIAL_C_FacialRoot, loc=<3.3646 -0.9032 6.3856>, rot=<-0.1567 0 0 -0.9876>}
 FACIAL_R_EyelidUpperA {parent=FACIAL_C_FacialRoot, loc=<-3.3625 -0.9047 6.387>, rot=<-0.1567 0 0 -0.9876>}
 FACIAL_L_Eye {parent=FACIAL_C_FacialRoot, loc=<3.3553 -0.9133 6.3694>, rot=<-0.009 0 0 -1>}
 FACIAL_R_Eye {parent=FACIAL_C_FacialRoot, loc=<-3.3574 -0.9116 6.37>, rot=<-0.009 0 0 -1>}
 FACIAL_L_EyelidLowerA {parent=FACIAL_C_FacialRoot, loc=<3.3646 -0.9032 6.3856>, rot=<0.2754 0 0 -0.9613>}
 FACIAL_R_EyelidLowerA {parent=FACIAL_C_FacialRoot, loc=<-3.3625 -0.9047 6.387>, rot=<0.2754 0 0 -0.9613>}
 FACIAL_L_EyeCornerInner {parent=FACIAL_C_FacialRoot, loc=<1.9207 -0.2453 7.8056>, rot=<-0.009 0 0 -1>}
 FACIAL_R_EyeCornerInner {parent=FACIAL_C_FacialRoot, loc=<-1.9124 -0.2499 7.8067>, rot=<-0.009 0 0 -1>}
 FACIAL_L_EyeCornerOuter {parent=FACIAL_C_FacialRoot, loc=<5.428 -1.1819 6.9959>, rot=<-0.0083 -0.3826 -0.0034 -0.9239>}
 FACIAL_R_EyeCornerOuter {parent=FACIAL_C_FacialRoot, loc=<-5.4307 -1.1771 6.9953>, rot=<-0.0083 0.3826 0.0034 -0.9239>}
 FACIAL_L_EyesackLower {parent=FACIAL_C_FacialRoot, loc=<4.1701 1.054 6.9559>, rot=<-0.0312 -0.1395 -0.0235 -0.9895>}
 FACIAL_R_EyesackLower {parent=FACIAL_C_FacialRoot, loc=<-4.1653 1.0572 6.9593>, rot=<-0.0312 0.1395 0.0235 -0.9894>}
 FACIAL_L_CheekInner {parent=FACIAL_C_FacialRoot, loc=<2.2978 1.6962 7.6699>, rot=<-0.1319 -0.0862 -0.1979 -0.9675>}
 FACIAL_R_CheekInner {parent=FACIAL_C_FacialRoot, loc=<-2.2933 1.6957 7.6711>, rot=<-0.1319 0.0862 0.1979 -0.9675>}
 FACIAL_L_CheekOuter {parent=FACIAL_C_FacialRoot, loc=<4.6045 1.5395 5.9971>, rot=<0.0469 -0.3778 0.097 -0.9196>}
 FACIAL_R_CheekOuter {parent=FACIAL_C_FacialRoot, loc=<-4.6034 1.5321 5.9977>, rot=<0.0469 0.3778 -0.097 -0.9196>}
 FACIAL_L_NasolabialBulge {parent=FACIAL_C_FacialRoot, loc=<3.3479 4.8304 6.1358>, rot=<0.0496 -0.3775 0.0204 -0.9244>}
 FACIAL_R_NasolabialBulge {parent=FACIAL_C_FacialRoot, loc=<-3.3479 4.8301 6.1361>, rot=<0.0496 0.3775 -0.0204 -0.9244>}
 FACIAL_L_NasolabialFurrow {parent=FACIAL_C_FacialRoot, loc=<3.1766 3.8263 8.5693>, rot=<-0.0416 -0.1225 -0.3647 -0.9221>}
 FACIAL_R_NasolabialFurrow {parent=FACIAL_C_FacialRoot, loc=<-3.1792 3.8219 8.5691>, rot=<-0.0416 0.1225 0.3647 -0.9221>}
 FACIAL_L_CheekLower {parent=FACIAL_C_FacialRoot, loc=<3.7061 4.6319 5.4605>, rot=<0.0265 -0.4274 0.0539 -0.9021>}
 FACIAL_R_CheekLower {parent=FACIAL_C_FacialRoot, loc=<-3.7058 4.631 5.4613>, rot=<0.0265 0.4274 -0.0539 -0.9021>}
 FACIAL_C_Nose {parent=FACIAL_C_FacialRoot, loc=<0.0004 0.7471 8.8981>, rot=<-0.1393 0 0 -0.9903>}
 FACIAL_C_NoseLower {parent=FACIAL_C_Nose, loc=<-0.0006 3.101 -0.1813>, rot=<0.1305 0 0 -0.9914>}
 FACIAL_L_Nostril {parent=FACIAL_C_Nose, loc=<0.6535 2.3758 -0.0484>, rot=<0.0173 0.1305 0.0023 -0.9913>}
 FACIAL_R_Nostril {parent=FACIAL_C_Nose, loc=<-0.6514 2.373 -0.0519>, rot=<0.0173 -0.1305 -0.0023 -0.9913>}
 FACIAL_C_NoseTip {parent=FACIAL_C_Nose, loc=<-0.0008 2.4446 0.4603>, rot=<0.0872 0 0 -0.9962>}
 FACIAL_C_MouthUpper {parent=FACIAL_C_FacialRoot, loc=<0.0004 4.9677 7.2142>, rot=<-0.009 0 0 -1>}
 FACIAL_C_LipUpper {parent=FACIAL_C_MouthUpper, loc=<-0.0002 0.2029 2.0922>, rot=<-0.0872 0 0 -0.9962>}
 FACIAL_L_LipUpper {parent=FACIAL_C_MouthUpper, loc=<0.7891 0.2516 2.0203>, rot=<-0.0778 -0.1601 -0.0379 -0.9833>}
 FACIAL_R_LipUpper {parent=FACIAL_C_MouthUpper, loc=<-0.7906 0.2515 2.0195>, rot=<-0.0778 0.1601 0.0379 -0.9833>}
 FACIAL_L_LipUpperOuter {parent=FACIAL_C_MouthUpper, loc=<1.502 0.3994 1.5472>, rot=<-0.0984 -0.2293 -0.0983 -0.9634>}
 FACIAL_R_LipUpperOuter {parent=FACIAL_C_MouthUpper, loc=<-1.5022 0.4008 1.5462>, rot=<-0.0984 0.2293 0.0983 -0.9634>}
 FACIAL_L_LipCorner {parent=FACIAL_C_MouthUpper, loc=<2.0534 0.69 0.7577>, rot=<0 -0.2078 0 -0.9782>}
 FACIAL_R_LipCorner {parent=FACIAL_C_MouthUpper, loc=<-2.0536 0.6911 0.7547>, rot=<0 0.2078 0 -0.9782>}
 FACIAL_L_Masseter {parent=FACIAL_C_FacialRoot, loc=<6.7474 1.7801 3.7222>, rot=<0.0604 0.1128 0.0089 -0.9917>}
 FACIAL_R_Masseter {parent=FACIAL_C_FacialRoot, loc=<-6.7459 1.7821 3.7214>, rot=<0.0604 -0.1128 -0.0089 -0.9917>}
 FACIAL_C_UnderChin {parent=FACIAL_C_FacialRoot, loc=<0.0004 8.7617 4.3354>, rot=<0.492 0 0 -0.8706>}
 FACIAL_L_UnderChin {parent=FACIAL_C_FacialRoot, loc=<2.8692 7.4209 3.0053>, rot=<0.395 -0.3556 0.0791 -0.8434>}
 FACIAL_R_UnderChin {parent=FACIAL_C_FacialRoot, loc=<-2.8674 7.4197 3.0057>, rot=<0.395 0.3556 -0.0791 -0.8434>}
 FACIAL_C_TeethUpper {parent=FACIAL_C_FacialRoot, loc=<0.0003 5.6622 8.1656>, rot=<-0.009 0 0 -1>}
 FACIAL_C_LowerLipRotation {parent=FACIAL_C_FacialRoot, loc=<0.0007 2.3263 1.7413>, rot=<0.3336 0 0 -0.9427>}
 FACIAL_C_MouthLower {parent=FACIAL_C_LowerLipRotation, loc=<-0.0002 -1.3886 5.9162>, rot=<-0.342 0 0 -0.9397>}
 FACIAL_C_LipLower {parent=FACIAL_C_MouthLower, loc=<-0.0001 1.4533 1.6081>, rot=<0.0872 0 0 -0.9962>}
 FACIAL_L_LipLower {parent=FACIAL_C_MouthLower, loc=<0.6771 1.3644 1.4814>, rot=<0.0927 -0.1294 0.0272 -0.9869>}
 FACIAL_R_LipLower {parent=FACIAL_C_MouthLower, loc=<-0.6777 1.3653 1.4816>, rot=<0.0927 0.1294 -0.0272 -0.9869>}
 FACIAL_L_LipLowerOuter {parent=FACIAL_C_MouthLower, loc=<1.3097 1.1011 1.1738>, rot=<0.0913 -0.2321 0.0318 -0.9679>}
 FACIAL_R_LipLowerOuter {parent=FACIAL_C_MouthLower, loc=<-1.3114 1.1005 1.1725>, rot=<0.0913 0.2321 -0.0318 -0.9679>}
 FACIAL_C_TeethLower {parent=FACIAL_C_LowerLipRotation, loc=<-0.0003 -1.405 7.0296>, rot=<-0.342 0 0 -0.9397>}
 FACIAL_C_Tongue1 {parent=FACIAL_C_TeethLower, loc=<-0.0005 -0.0143 -4.754>, rot=<-0.2588 0 0 -0.9659>}
 FACIAL_C_Tongue2 {parent=FACIAL_C_Tongue1, loc=<0.0003 0.4252 1.3924>, rot=<0.2588 0 0 -0.9659>}
 FACIAL_C_Tongue3 {parent=FACIAL_C_Tongue2, loc=<0.0002 0.2433 1.2423>, rot=<0.1305 0 0 -0.9914>}
 FACIAL_C_Tongue4 {parent=FACIAL_C_Tongue3, loc=<0.0006 -0.1237 1.0632>, rot=<0.0436 0 0 -0.999>}
 FACIAL_C_TongueUpper3 {parent=FACIAL_C_Tongue3, loc=<0 -0.339 -0.2014>, rot=<0 0 0 1>}
 FACIAL_C_TongueLower3 {parent=FACIAL_C_Tongue3, loc=<0 0.8521 0.009>, rot=<0 0 0 1>}
 FACIAL_L_TongueSide3 {parent=FACIAL_C_Tongue3, loc=<1.556 -0.009 0.0793>, rot=<0 0 0 1>}
 FACIAL_R_TongueSide3 {parent=FACIAL_C_Tongue3, loc=<-1.5559 -0.0092 0.0795>, rot=<0 0 0 1>}
 FACIAL_C_TongueUpper2 {parent=FACIAL_C_Tongue2, loc=<-0.0001 -0.6115 -0.2769>, rot=<0 0 0 1>}
 FACIAL_L_TongueSide2 {parent=FACIAL_C_Tongue2, loc=<1.8719 -0.0351 0.0743>, rot=<0 0 0 1>}
 FACIAL_R_TongueSide2 {parent=FACIAL_C_Tongue2, loc=<-1.8718 -0.036 0.0748>, rot=<0 0 0 1>}
 FACIAL_C_Jaw {parent=FACIAL_C_FacialRoot, loc=<0.0007 2.3263 1.7413>, rot=<0.3336 0 0 -0.9427>}
 FACIAL_C_Jawline {parent=FACIAL_C_Jaw, loc=<-0.0008 2.1902 7.613>, rot=<0.2586 0 0 -0.966>}
 FACIAL_L_ChinSide {parent=FACIAL_C_Jaw, loc=<2.4133 0.2993 7.3799>, rot=<-0.2208 -0.3738 0.0499 -0.8995>}
 FACIAL_R_ChinSide {parent=FACIAL_C_Jaw, loc=<-2.4133 0.2982 7.3803>, rot=<-0.2208 0.3738 -0.0499 -0.8995>}
 FACIAL_C_Chin {parent=FACIAL_C_Jaw, loc=<-0.0006 0.7865 7.7913>, rot=<-0.2591 0 0 -0.9659>}
 lip_ring {parent=FACIAL_C_Jaw, loc=<0.3736 -1.2348 7.9016>, rot=<0.89 0 0 0.4559>}
 FACIAL_L_Ear {parent=FACIAL_C_FacialRoot, loc=<6.9788 -0.5351 0.167>, rot=<-0.0353 -0.0004 -0.0262 -0.999>}
 FACIAL_R_Ear {parent=FACIAL_C_FacialRoot, loc=<-6.9748 -0.5304 0.1692>, rot=<-0.0353 0.0005 0.0262 -0.999>}
 FACIAL_L_Jawline {parent=FACIAL_C_FacialRoot, loc=<4.3929 5.6561 4.3875>, rot=<0.2006 0.1582 -0.0532 -0.9653>}
 FACIAL_R_Jawline {parent=FACIAL_C_FacialRoot, loc=<-4.3933 5.6552 4.393>, rot=<0.2006 -0.1582 0.0532 -0.9653>}
 FACIAL_C_NoseBridge {parent=FACIAL_C_FacialRoot, loc=<0.0001 -1.3575 9.3544>, rot=<0.0348 0 0 -0.9994>}
 FACIAL_L_EyelidUpperB {parent=FACIAL_C_FacialRoot, loc=<3.3646 -0.9032 6.3856>, rot=<-0.2337 0 0 -0.9723>}
 FACIAL_L_EyelidLowerB {parent=FACIAL_C_FacialRoot, loc=<3.3646 -0.9032 6.3856>, rot=<0.3088 0 0 -0.9511>}
 FACIAL_R_EyelidLowerB {parent=FACIAL_C_FacialRoot, loc=<-3.3625 -0.9047 6.387>, rot=<0.317 0 0 -0.9484>}
 FACIAL_R_EyelidUpperB {parent=FACIAL_C_FacialRoot, loc=<-3.3625 -0.9047 6.387>, rot=<-0.2337 0 0 -0.9723>}
 clavicle_l {parent=spine_05, loc=<5.4831 -0.1053 -3.2946>, rot=<0.0465 -0.7289 -0.009 -0.6829>}
 upperarm_l {parent=clavicle_l, loc=<9.8658 -1.2136 -0.4105>, rot=<0.0164 -0.3245 -0.0269 -0.9454>}
 upperarm_twist_01_l {parent=upperarm_l, loc=<7.2135 -0.1134 0>, rot=<0 0 0 1>}
 upperarm_twist_02_l {parent=upperarm_l, loc=<19.7998 -0.1134 0>, rot=<0 0 0 1>}
 clavicle_r {parent=spine_05, loc=<5.4831 -0.1053 3.2946>, rot=<0.7289 0.0465 -0.6829 0.009>}
 upperarm_r {parent=clavicle_r, loc=<-9.8658 1.2136 0.4105>, rot=<0.0164 -0.3245 -0.0269 -0.9454>}
 upperarm_twist_01_r {parent=upperarm_r, loc=<-7.2135 0.1134 0>, rot=<0 0 0 -1>}
 sleeve_2_r {parent=upperarm_twist_01_r, loc=<0.4269 5.3259 2.2865>, rot=<0 -1 0 0>}
 sleeve_2_r1 {parent=sleeve_2_r, loc=<5.6414 0 0>, rot=<0 0 0 1>}
 deform_upperarm_bk_base_r {parent=upperarm_twist_01_r, loc=<6.4053 4.3746 -0.314>, rot=<0 0 0 -1>}
 deform_upperarm_bk_r {parent=deform_upperarm_bk_base_r, loc=<0 0 0>, rot=<0 0 0 1>}
 extra_spine_05 {parent=spine_05, loc=<0 0 0>, rot=<0 0 0 1>}
 bag_root {parent=spine_05, loc=<11.1051 -1.6608 -9.3698>, rot=<0.0284 -0.7065 -0.0284 -0.7065>}
 bag_front_01 {parent=bag_root, loc=<-0.8717 5.0598 -1.8569>, rot=<-0.4739 -0.2497 0.7977 0.2769>}
 bag_front_02 {parent=bag_front_01, loc=<7.8242 0 0>, rot=<-0.0125 0.0259 -0.0631 -0.9976>}
 bag_front_03 {parent=bag_front_02, loc=<7.9872 0 0>, rot=<0.061 0.0171 -0.146 -0.9873>}
 bag_front_04 {parent=bag_front_03, loc=<5.7188 0 0>, rot=<-0.1047 0.0202 -0.1879 -0.9764>}
 bag_front_05 {parent=bag_front_04, loc=<8.1103 0 0>, rot=<-0.0658 0.0061 -0.0917 -0.9936>}
 bag_front_06 {parent=bag_front_05, loc=<8.022 0 0>, rot=<0 0 0 1>}
 bag_back_01 {parent=bag_root, loc=<-1.5508 -3.7462 -1.7913>, rot=<-0.4594 0.1832 0.8598 -0.1272>}
 bag_back_02 {parent=bag_back_01, loc=<5.5424 0 0>, rot=<0 0 0.1808 -0.9835>}
 bag_back_03 {parent=bag_back_02, loc=<6.3957 0 0>, rot=<0.0074 0.0005 0.0666 -0.9978>}
 bag_back_04 {parent=bag_back_03, loc=<5.8846 0 0>, rot=<0.1516 0.0023 0.0149 -0.9883>}
 bag_back_05 {parent=bag_back_04, loc=<6.3129 0 0>, rot=<-0.0263 -0.0011 0.0437 -0.9987>}
 bag_back_06 {parent=bag_back_05, loc=<9.4374 0 0>, rot=<0 0 0 1>}
 deform_pec_aim_base_l {parent=spine_04, loc=<3.2456 9.3309 -1.2237>, rot=<0.0603 -0.4033 0.3318 -0.8507>}
 deform_pec_aim_l {parent=deform_pec_aim_base_l, loc=<0 0 0>, rot=<0 0 0 1>}
 deform_pec_base_l {parent=deform_pec_aim_l, loc=<9.3665 0 0>, rot=<0 0 0 1>}
 deform_pec_l {parent=deform_pec_base_l, loc=<0 0 0>, rot=<0 0 0 1>}
 deform_pec_aim_base_r {parent=spine_04, loc=<3.2452 9.3309 1.2237>, rot=<-0.4033 -0.0603 0.8507 0.3318>}
 deform_pec_aim_r {parent=deform_pec_aim_base_r, loc=<0 0 0>, rot=<0 0 0 1>}
 deform_pec_base_r {parent=deform_pec_aim_r, loc=<-9.3667 -0.0003 -0.0004>, rot=<0 0 0 1>}
 deform_pec_r {parent=deform_pec_base_r, loc=<0 0 0>, rot=<0 0 0 1>}
 dyn_bag_sim_chain_01_MSTR {parent=spine_04, loc=<-3.0683 1.178 3.6634>, rot=<0.2209 -0.6271 0.7355 0.1304>}
 dyn_bag_sim_chain_01 {parent=dyn_bag_sim_chain_01_MSTR, loc=<0 0 0>, rot=<0 0 0 1>}
 dyn_bag_sim_chain_02 {parent=dyn_bag_sim_chain_01, loc=<11.1795 0 0>, rot=<0 0 0 1>}
 dyn_bag_sim_chain_03 {parent=dyn_bag_sim_chain_02, loc=<0.3745 -0.2247 -0.0633>, rot=<0.1018 0.0185 0.2338 -0.9668>}
 hedgehog {parent=dyn_bag_sim_chain_03, loc=<15.1963 5.04 9.9258>, rot=<0.082 0.4136 -0.3511 -0.836>}
 eye_1 {parent=hedgehog, loc=<6.0428 -0.9119 0.398>, rot=<0.7705 -0.4222 0.3168 -0.3574>}
 eye_1_end {parent=eye_1, loc=<4.3233 0 0>, rot=<0 0 0 1>}
 eye_2 {parent=hedgehog, loc=<6.0428 -0.9119 0.398>, rot=<0.884 0.1025 0.4435 -0.1065>}
 eye_2_end {parent=eye_2, loc=<4.2253 0 0>, rot=<0 0 0 1>}
 eyelid_upper_2 {parent=hedgehog, loc=<6.0428 -0.9119 0.398>, rot=<-0.3065 0.4541 -0.1679 -0.8196>}
 joint2 {parent=eyelid_upper_2, loc=<4.3674 0 0>, rot=<0 0 0 1>}
 eyelid_upper_1 {parent=hedgehog, loc=<6.0428 -0.9119 0.398>, rot=<-0.1832 0.2913 0.4882 -0.802>}
 joint4 {parent=eyelid_upper_1, loc=<4.471 0 0>, rot=<0 0 0 1>}
 eyelid_lower_2 {parent=hedgehog, loc=<6.0428 -0.9119 0.398>, rot=<-0.2585 0.2304 -0.2219 -0.9115>}
 joint6 {parent=eyelid_lower_2, loc=<4.2473 0 0>, rot=<0 0 0 1>}
 eyelid_lower_1 {parent=hedgehog, loc=<6.0428 -0.9119 0.398>, rot=<-0.3301 0.0912 0.4257 -0.8375>}
 joint8 {parent=eyelid_lower_1, loc=<4.3502 0 0>, rot=<0 0 0 1>}
 zipper {parent=dyn_bag_sim_chain_03, loc=<8.8014 2.572 -10.3533>, rot=<0.6318 0.2623 -0.2672 -0.6787>}
 BASE_crop_top_1 {parent=spine_03, loc=<4.0646 13.0723 -4.9185>, rot=<0.0547 -0.9985 0 0>}
 crop_top_1 {parent=BASE_crop_top_1, loc=<0 0 0>, rot=<0 0 0 1>}
 crop_top_1_end {parent=crop_top_1, loc=<3 0 0>, rot=<0 0 0 1>}
 BASE_crop_top_2 {parent=spine_03, loc=<4.6487 7.7601 -11.1887>, rot=<0.0468 -0.8545 0.5165 -0.0283>}
 crop_top_2 {parent=BASE_crop_top_2, loc=<0 0 0>, rot=<0 0 0 1>}
 crop_top_2_end {parent=crop_top_2, loc=<3 0 0>, rot=<0 0 0 1>}
 BASE_crop_top_3 {parent=spine_03, loc=<5.543 -0.3724 -12.749>, rot=<0.0387 -0.706 0.706 -0.0387>}
 crop_top_3 {parent=BASE_crop_top_3, loc=<0 0 0>, rot=<0 0 0 1>}
 crop_top_3_end {parent=crop_top_3, loc=<3 0 0>, rot=<0 0 0 1>}
 BASE_crop_top_4 {parent=spine_03, loc=<6.1231 -5.6474 -10.3082>, rot=<0.0141 -0.2567 0.9649 -0.0529>}
 crop_top_4 {parent=BASE_crop_top_4, loc=<0 0 0>, rot=<0 0 0 1>}
 crop_top_4_end {parent=crop_top_4, loc=<3 0 0>, rot=<0 0 0 1>}
 BASE_crop_top_5 {parent=spine_03, loc=<6.6111 -8.9415 -4.1089>, rot=<0.003 -0.055 0.997 -0.0547>}
 crop_top_5 {parent=BASE_crop_top_5, loc=<0 0 0>, rot=<0 0 0 1>}
 crop_top_5_end {parent=crop_top_5, loc=<3 0 0>, rot=<0 0 0 1>}
 BASE_crop_top_6 {parent=spine_03, loc=<6.1593 -5.6287 11.1298>, rot=<-0.0249 0.4534 0.8896 -0.0488>}
 crop_top_6 {parent=BASE_crop_top_6, loc=<0 0 0>, rot=<0 0 0 1>}
 crop_top_6_end {parent=crop_top_6, loc=<3 0 0>, rot=<0 0 0 1>}
 BASE_crop_top_7 {parent=spine_03, loc=<5.4153 0.8082 12.9951>, rot=<-0.0381 0.6954 0.7165 -0.0393>}
 crop_top_7 {parent=BASE_crop_top_7, loc=<0 0 0>, rot=<0 0 0 1>}
 crop_top_7_end {parent=crop_top_7, loc=<3 0 0>, rot=<0 0 0 1>}
 BASE_crop_top_8 {parent=spine_03, loc=<4.5051 10.6081 10.0168>, rot=<0.0475 -0.8672 -0.4949 0.0271>}
 crop_top_8 {parent=BASE_crop_top_8, loc=<0 0 0>, rot=<0 0 0 1>}
 crop_top_8_end {parent=crop_top_8, loc=<3 0 0>, rot=<0 0 0 1>}
 sleeve_l {type=Curve, loc=<0 0 0>, rot=<0 0 0 1>}
 sleeve_r {type=Curve, loc=<0 0 0>, rot=<0 0 0 1>}
]
variables [
 Test_Rotate {type="TArray<double>"}
 crouching {type=double, default=0}
 play_f {type=double, default=0}
 test_01 {type=double, default=-2}
 test_02 {type=double, default=2}
]
