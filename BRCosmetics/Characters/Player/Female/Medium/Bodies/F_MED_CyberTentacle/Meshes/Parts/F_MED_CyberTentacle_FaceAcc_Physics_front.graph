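# Physics asset for the front face-accessory tentacles of the female CyberTentacle character, with four chains (left and right, A and B) of four capsule bodies each.
# Each chain hangs from a small BASE body jointed to the neck, with swing and twist limits that mostly tighten toward the tip.
# Head, neck, upper spine, clavicle and upper-arm capsules cover the upper body, with angularly locked torso and arm joints and free head and neck joints.
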PhysicsAsset /BRCosmetics/Characters/Player/Female/Medium/Bodies/F_MED_CyberTentacle/Meshes/Parts/F_MED_CyberTentacle_FaceAcc_Physics_front
bodies {
 BASE_tentacle_front_l_A_01 [{shapeType=Capsule, length=0.505, radius=0.505, shapeName=BASE_tentacle_front_l_A_01_capsule, rotation=<rot 0 -90 0>}]
 BASE_tentacle_front_l_B_01 [{shapeType=Capsule, length=0.505, radius=0.505, shapeName=BASE_tentacle_front_l_B_01_capsule, rotation=<rot 0 -90 0>}]
 BASE_tentacle_front_r_A_01 [{shapeType=Capsule, length=0.505, radius=0.505, shapeName=BASE_tentacle_front_r_A_01_capsule, rotation=<rot 0 -90 0>}]
 BASE_tentacle_front_r_B_01 [{shapeType=Capsule, length=0.505, radius=0.505, shapeName=BASE_tentacle_front_r_B_01_capsule, rotation=<rot 0 -90 0>}]
 clavicle_l [{shapeType=Capsule, length=6.1804, radius=5.8196, shapeName=clavicle_l_capsule, center=<1.8251 -1.3444 0>, rotation=<rot 0 -90 90>}]
 clavicle_r [{shapeType=Capsule, length=8.2352, radius=6.3535, shapeName=clavicle_r_capsule, center=<-5 1.3618 2.0382>, rotation=<rot 0 90 90>}]
 extra_spine_05 [{shapeType=Capsule, length=8.0196, radius=10.759, shapeName=extra_spine_05_capsule, center=<2.4275 -1.9184 0>, rotation=<rot 0 -90 0>}]
 head [
  {shapeType=Capsule, length=6.7534, radius=6.3369, center=<3.1734 2.6019 0.0541>, rotation=<rot 0.5016 -90.415 -88.3249>}
  {
   shapeType=Capsule
   length=3.3889
   radius=6.0811
   shapeName=head_capsule
   center=<1.6133 -0.0375 -0.0239>
   rotation=<rot -89.1519 179.9981 -180>
  }
 ]
 neck_01 [{shapeType=Capsule, length=6.803, radius=4.9466, shapeName=neck_01_capsule, center=<1.3531 3.3125 0>, rotation=<rot 0 -106.7278 90>}]
 spine_04 [{shapeType=Capsule, length=7.0342, radius=11.9758, shapeName=spine_04_capsule, center=<0 2.035 0>, rotation=<rot 0 -90 0>}]
 spine_05 [{shapeType=Capsule, length=15.7958, radius=12.505, shapeName=spine_04_capsule, center=<1.3282 3.0645 0>, rotation=<rot 0 -90 0>}]
 tentacle_front_l_A_01 [{
   shapeType=Capsule
   length=1.5939
   radius=1.8338
   shapeName=tentacle_front_l_A_01_capsule
   center=<2.5993 -0.2326 0.0425>
   rotation=<rot 70.2954 -129.9876 -135.8157>
  }]
 tentacle_front_l_A_02 [{
   shapeType=Capsule
   length=1.1468
   radius=1.9514
   shapeName=tentacle_front_l_A_02_capsule
   center=<2.9999 0.2541 -0.2957>
   rotation=<rot -34.7292 -90.8676 -96.6785>
  }]
 tentacle_front_l_A_03 [{
   shapeType=Capsule
   length=3.7931
   radius=1.8175
   shapeName=tentacle_front_l_A_03_capsule
   center=<3.9892 0.0651 -0.0078>
   rotation=<rot 14.883 93.0822 -89.6523>
  }]
 tentacle_front_l_A_04 [{
   shapeType=Capsule
   length=4.3873
   radius=1.6425
   shapeName=tentacle_front_l_A_04_capsule
   center=<4.7999 -0.7819 -0.3333>
   rotation=<rot 29.0994 -103.6437 -81.7226>
  }]
 tentacle_front_l_B_01 [{
   shapeType=Capsule
   length=3.7875
   radius=2.3543
   shapeName=tentacle_front_l_B_01_capsule
   center=<4.1001 -0.3516 -0.5946>
   rotation=<rot -8.8988 -91.8047 -86.3882>
  }]
 tentacle_front_l_B_02 [{
   shapeType=Capsule
   length=3.4861
   radius=2.267
   shapeName=tentacle_front_l_B_02_capsule
   center=<3.2077 -0.4595 -0.5311>
   rotation=<rot -12.0613 -89.2662 -101.0819>
  }]
 tentacle_front_l_B_03 [{
   shapeType=Capsule
   length=3.809
   radius=1.9995
   shapeName=tentacle_front_l_B_03_capsule
   center=<5.016 0.0917 -0.195>
   rotation=<rot -32.4118 -89.538 -84.8688>
  }]
 tentacle_front_l_B_04 [{
   shapeType=Capsule
   length=4.522
   radius=1.5818
   shapeName=tentacle_front_l_B_04_capsule
   center=<6.1575 -0.1376 -0.3725>
   rotation=<rot 24.9248 -89.9998 -76.1265>
  }]
 tentacle_front_r_A_01 [{
   shapeType=Capsule
   length=3.671
   radius=2.2756
   shapeName=tentacle_front_r_A_01_capsule
   center=<1.2914 -0.0794 -0.213>
   rotation=<rot -30.7456 -89.9954 88.2016>
  }]
 tentacle_front_r_A_02 [{
   shapeType=Capsule
   length=2.8039
   radius=2.3355
   shapeName=tentacle_front_r_A_02_capsule
   center=<3.4977 0.0842 0.0153>
   rotation=<rot 27.6556 -89.9999 96.6722>
  }]
 tentacle_front_r_A_03 [{
   shapeType=Capsule
   length=4.6508
   radius=1.8868
   shapeName=tentacle_front_r_A_03_capsule
   center=<3.7832 0.4719 0.0081>
   rotation=<rot 39.0572 -90 -96.2924>
  }]
 tentacle_front_r_A_04 [{
   shapeType=Capsule
   length=3.9731
   radius=1.5238
   shapeName=tentacle_front_r_A_04_capsule
   center=<5.5225 -0.7269 0.7478>
   rotation=<rot 36.2377 -113.183 79.3625>
  }]
 tentacle_front_r_B_01 [{
   shapeType=Capsule
   length=4.9248
   radius=2.1577
   shapeName=tentacle_front_r_B_01_capsule
   center=<3.5138 0.3997 0.2982>
   rotation=<rot 0 101.9607 -95.3998>
  }]
 tentacle_front_r_B_02 [{
   shapeType=Capsule
   length=5.0108
   radius=2.2269
   shapeName=tentacle_front_r_B_02_capsule
   center=<4.5395 0.4487 0.1392>
   rotation=<rot 2.7158 -90 98.1116>
  }]
 tentacle_front_r_B_03 [{
   shapeType=Capsule
   length=3.0014
   radius=2.0829
   shapeName=tentacle_front_r_B_03_capsule
   center=<5.1099 0.2454 -0.0166>
   rotation=<rot 10.5742 -90 84.2933>
  }]
 tentacle_front_r_B_04 [{
   shapeType=Capsule
   length=5.2068
   radius=1.9352
   shapeName=tentacle_front_r_B_04_capsule
   center=<4.5254 0.1534 -0.0171>
   rotation=<rot 22.983 -90 81.1269>
  }]
 upperarm_twist_01_l [{
   shapeType=Capsule
   length=14.371
   radius=6.1779
   shapeName=upperarm_twist_01_l_capsule
   center=<2.4447 -0.1011 -0.9649>
   rotation=<rot 0 -92.3693 90>
  }]
 upperarm_twist_01_r [{
   shapeType=Capsule
   length=6.3309
   radius=6.1779
   shapeName=upperarm_twist_01_l_capsule
   center=<-1.0962 0.0454 1.3775>
   rotation=<rot 0 -92.3693 90>
  }]
}
constraints [
 {
  bone1Name=spine_05
  bone2Name=spine_04
  swing1LimitDegrees=45
  swing1Motion=Locked
  swing2LimitDegrees=45
  swing2Motion=Locked
  twistLimitDegrees=45
  twistMotion=Locked
 }
 {
  bone1Name=clavicle_r
  bone2Name=spine_05
  swing1LimitDegrees=45
  swing1Motion=Locked
  swing2LimitDegrees=45
  swing2Motion=Locked
  twistLimitDegrees=45
  twistMotion=Locked
 }
 {
  bone1Name=clavicle_l
  bone2Name=spine_05
  swing1LimitDegrees=45
  swing1Motion=Locked
  swing2LimitDegrees=45
  swing2Motion=Locked
  twistLimitDegrees=45
  twistMotion=Locked
 }
 {bone1Name=tentacle_front_l_A_01, bone2Name=BASE_tentacle_front_l_A_01, swing1LimitDegrees=30, swing2LimitDegrees=30, twistLimitDegrees=8}
 {bone1Name=tentacle_front_l_A_02, bone2Name=tentacle_front_l_A_01, swing1LimitDegrees=30, swing2LimitDegrees=30, twistLimitDegrees=8}
 {bone1Name=tentacle_front_l_A_03, bone2Name=tentacle_front_l_A_02, swing1LimitDegrees=30, swing2LimitDegrees=15, twistLimitDegrees=10}
 {bone1Name=tentacle_front_l_A_04, bone2Name=tentacle_front_l_A_03, swing1LimitDegrees=15, swing2LimitDegrees=15, twistLimitDegrees=10}
 {bone1Name=tentacle_front_l_B_01, bone2Name=BASE_tentacle_front_l_B_01, swing1LimitDegrees=15, swing2LimitDegrees=15, twistLimitDegrees=10}
 {bone1Name=tentacle_front_l_B_02, bone2Name=tentacle_front_l_B_01, swing1LimitDegrees=15, swing2LimitDegrees=15, twistLimitDegrees=10}
 {bone1Name=tentacle_front_l_B_03, bone2Name=tentacle_front_l_B_02, swing1LimitDegrees=15, swing2LimitDegrees=15, twistLimitDegrees=10}
 {bone1Name=tentacle_front_l_B_04, bone2Name=tentacle_front_l_B_03, swing1LimitDegrees=15, swing2LimitDegrees=15, twistLimitDegrees=10}
 {bone1Name=tentacle_front_r_A_01, bone2Name=BASE_tentacle_front_r_A_01, swing1LimitDegrees=50, swing2LimitDegrees=30, twistLimitDegrees=15}
 {bone1Name=tentacle_front_r_A_02, bone2Name=tentacle_front_r_A_01, swing1LimitDegrees=50, swing2LimitDegrees=30, twistLimitDegrees=15}
 {bone1Name=tentacle_front_r_A_03, bone2Name=tentacle_front_r_A_02, swing1LimitDegrees=30, swing2LimitDegrees=15, twistLimitDegrees=15}
 {bone1Name=tentacle_front_r_A_04, bone2Name=tentacle_front_r_A_03, swing1LimitDegrees=30, swing2LimitDegrees=15, twistLimitDegrees=10}
 {bone1Name=tentacle_front_r_B_01, bone2Name=BASE_tentacle_front_r_B_01, swing1LimitDegrees=30, swing2LimitDegrees=15, twistLimitDegrees=10}
 {bone1Name=tentacle_front_r_B_02, bone2Name=tentacle_front_r_B_01, swing1LimitDegrees=30, swing2LimitDegrees=15, twistLimitDegrees=10}
 {bone1Name=tentacle_front_r_B_03, bone2Name=tentacle_front_r_B_02, swing1LimitDegrees=15, swing2LimitDegrees=15, twistLimitDegrees=10}
 {bone1Name=tentacle_front_r_B_04, bone2Name=tentacle_front_r_B_03, swing1LimitDegrees=15, swing2LimitDegrees=15, twistLimitDegrees=10}
 {
  bone1Name=extra_spine_05
  bone2Name=spine_05
  swing1LimitDegrees=45
  swing1Motion=Locked
  swing2LimitDegrees=45
  swing2Motion=Locked
  twistLimitDegrees=45
  twistMotion=Locked
 }
 {
  bone1Name=upperarm_twist_01_l
  bone2Name=clavicle_l
  swing1LimitDegrees=45
  swing1Motion=Locked
  swing2LimitDegrees=45
  swing2Motion=Locked
  twistLimitDegrees=45
  twistMotion=Locked
 }
 {
  bone1Name=upperarm_twist_01_r
  bone2Name=clavicle_r
  swing1LimitDegrees=45
  swing1Motion=Locked
  swing2LimitDegrees=45
  swing2Motion=Locked
  twistLimitDegrees=45
  twistMotion=Locked
 }
 {
  bone1Name=head
  bone2Name=neck_01
  swing1LimitDegrees=45
  swing1Motion=Free
  swing2LimitDegrees=45
  swing2Motion=Free
  twistLimitDegrees=45
  twistMotion=Free
 }
 {
  bone1Name=BASE_tentacle_front_r_A_01
  bone2Name=neck_01
  swing1LimitDegrees=45
  swing1Motion=Free
  swing2LimitDegrees=45
  swing2Motion=Free
  twistLimitDegrees=45
  twistMotion=Free
 }
 {
  bone1Name=BASE_tentacle_front_r_B_01
  bone2Name=neck_01
  swing1LimitDegrees=45
  swing1Motion=Free
  swing2LimitDegrees=45
  swing2Motion=Free
  twistLimitDegrees=45
  twistMotion=Free
 }
 {
  bone1Name=BASE_tentacle_front_l_B_01
  bone2Name=neck_01
  swing1LimitDegrees=45
  swing1Motion=Free
  swing2LimitDegrees=45
  swing2Motion=Free
  twistLimitDegrees=45
  twistMotion=Free
 }
 {
  bone1Name=BASE_tentacle_front_l_A_01
  bone2Name=neck_01
  swing1LimitDegrees=45
  swing1Motion=Free
  swing2LimitDegrees=45
  swing2Motion=Free
  twistLimitDegrees=45
  twistMotion=Free
 }
 {
  bone1Name=neck_01
  bone2Name=spine_05
  swing1LimitDegrees=45
  swing1Motion=Free
  swing2LimitDegrees=45
  swing2Motion=Free
  twistLimitDegrees=45
  twistMotion=Free
 }
]
