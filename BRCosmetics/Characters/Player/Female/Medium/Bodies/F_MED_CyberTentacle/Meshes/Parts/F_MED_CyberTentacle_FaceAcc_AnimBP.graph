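# Animation Blueprint for the CyberTentacle female face accessory, adding rig and physics motion over copied character poses.
# Attached-parent pose, with upper arms layered from the body mesh, feeds a crouch-aware Control Rig, two RigidBody nodes, then a second Control Rig.
# Backbling or cape selects spine_02-based "front" physics; otherwise head-based physics runs.
# Event updates set those flags, a jump-plus-tactical-sprint factor, and a Gravity Override variable.
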AnimBlueprint /BRCosmetics/Characters/Player/Female/Medium/Bodies/F_MED_CyberTentacle/Meshes/Parts/F_MED_CyberTentacle_FaceAcc_AnimBP
target_skeleton=&/F_MED_CyberTentacle_FaceAcc_Skeleton
graphs [
 AnimGraph {nodes [
   n1 = Layered blend per bone(
    BasePose=Copy Pose From Mesh() {bUseAttachedParent=true}
    BlendPoses_0=Copy Pose From Mesh(SourceMeshComponent=Get BodySkelMesh() -> (BodySkelMesh:Skeletal Mesh Component Budgeted Object Reference)) {bUseAttachedParent=true}
   ) {layerSetup [{branchFilters [
       {boneName=upperarm_l}
       {boneName=upperarm_r}
       {boneName=upperarm_twist_01_l}
       {boneName=upperarm_twist_01_r}
       {boneName=upperarm_twist_02_l}
       {boneName=upperarm_twist_02_r}
      ]}]}
   n2 = Control Rig!(Source=@n1, Alpha, crouching:double=Get crouching_f() -> (crouching_f:double), play_f:double=Get play_f() -> (play_f:double)) -> (Pose) {controlRigAssetReference {blueprintRigClass=&/F_MED_CyberTentacle_FaceAcc_CtrlRig_02.F_MED_CyberTentacle_FaceAcc_CtrlRig_02_C}}
   n3 = Get backpack_cape_mod_bool() -> (backpack_cape_mod_bool:bool)
   n4 = RigidBody!(
    ComponentPose=Local To Component(LocalPose=@n2)
    bAlphaBoolEnabled:bool=NOT Boolean(A=@n3)
    OverrideWorldGravity:Vector="0,0,-980"
    SimSpaceSettings:Sim Space Settings Structure=(WorldAlpha=1,VelocityScaleZ=0.3,DampingAlpha=1,MaxLinearVelocity=3000,MaxAngularVelocity=1500,MaxLinearAcceleration=1000,MaxAngularAcceleration=65,ExternalLinearDragV=(X=1,Y=1,Z=1),ExternalLinearVelocity=(X=0,Y=0,Z=0),ExternalAngularVelocity=(X=0,Y=0,Z=0))
   ) -> (Pose) {
    alphaInputType=Bool
    bClampLinearTranslationLimitToRefPose=true
    bOverrideWorldGravity=true
    overridePhysicsAsset=&/F_MED_CyberTentacle_FaceAcc_Physics_front
    simulationSpace=BaseBoneSpace
    alphaBoolBlend {blendInTime=0.2, blendOption=Linear, blendOutTime=0.2}
    baseBoneRef {boneName=spine_02}
    overrideWorldGravity {z=-980}
    simSpaceSettings {
     maxAngularAcceleration=65
     maxAngularVelocity=1500
     maxLinearAcceleration=1000
     maxLinearVelocity=3000
     velocityScaleZ=0.3
     worldAlpha=1
     externalLinearDragV=<1 1 1>
    }
   }
   n5 = RigidBody(
    ComponentPose=@n4
    bAlphaBoolEnabled=@n3
    OverrideWorldGravity="0,0,-980"
    SimSpaceSettings=(WorldAlpha=1,VelocityScaleZ=0.3,DampingAlpha=1,MaxLinearVelocity=3000,MaxAngularVelocity=1500,MaxLinearAcceleration=1000,MaxAngularAcceleration=65,ExternalLinearDragV=(X=1,Y=1,Z=1),ExternalLinearVelocity=(X=0,Y=0,Z=0),ExternalAngularVelocity=(X=0,Y=0,Z=0))
   ) {
    alphaInputType=Bool
    bClampLinearTranslationLimitToRefPose=true
    bOverrideWorldGravity=true
    overridePhysicsAsset=&/F_MED_CyberTentacle_FaceAcc_Physics
    simulationSpace=BaseBoneSpace
    alphaBoolBlend {blendInTime=0.2, blendOption=Linear, blendOutTime=0.2}
    baseBoneRef {boneName=head}
    overrideWorldGravity {z=-980}
    simSpaceSettings {
     maxAngularAcceleration=65
     maxAngularVelocity=1500
     maxLinearAcceleration=1000
     maxLinearVelocity=3000
     velocityScaleZ=0.3
     worldAlpha=1
     externalLinearDragV=<1 1 1>
    }
   }
   n6 = Control Rig!(
    Source=Component To Local(ComponentPose=@n5)
    Alpha
    play_f=Get play_f()
    backpack_cape_mod:double=Get backpack_cape_mod() -> (backpack_cape_mod:double)
    backpack_cape_mod_bool:bool=Get backpack_cape_mod_bool()
    jump_sprint:double=Get jump_sprint() -> (jump_sprint:double)
   ) -> (Pose) {controlRigAssetReference {blueprintRigClass=&/F_MED_CyberTentacle_FaceAcc_CtrlRig.F_MED_CyberTentacle_FaceAcc_CtrlRig_C}}
   Output Pose(Result=@n6) {layerGroup=None}
  ]}
 EventGraph {nodes [
   Event BlueprintUpdateAnimation()
   > n1 = Sequence()
   > Set crouching_f(crouching_f:double="To Float (Boolean)"(InBool=Get bIsCrouching())) -> (Output_Get:double)
   n2 = CheckForBackbling()
   > n3 = Set backpack_cape_mod_bool(backpack_cape_mod_bool:bool=NOT Boolean(A=OR Boolean(A=@n2, B=@n2.bIsBackblingACape))) -> (Output_Get:bool)
   > Set backpack_cape_mod(backpack_cape_mod:double="To Float (Boolean)"(InBool=@n3)) -> (Output_Get:double)
   n4 = AND Boolean(A=Get bIsJumping(), B=Get bIsTacticalSprinting())
   > Set jump_sprint(jump_sprint:double="To Float (Boolean)"(InBool=NOT Boolean(A=@n4))) -> (Output_Get:double)
   n5 = "Gravity Override State Function (Parameter Asset) (Nativized)"(
    execute=@n1.then_1
    CharacterPartAnimInstance=Self-Reference()
    GameTimeInSeconds=GetGameTimeInSeconds()
    Parameters=&/F_MED_CyberTentacle_FaceAcc_Parameters
   )
   > Set Gravity Override(Gravity Override=@n5)
   n6 = Event OnCharacterCustomizationLoaded()
   > Set BodySkelMesh(BodySkelMesh:Skeletal Mesh Component Budgeted Object Reference=Get Array Item with EFort Custom Part Type(Array=@n6.CustomMeshParts)) -> (Output_Get:Skeletal Mesh Component Budgeted Object Reference)
   Event BlueprintBeginPlay() {is=title K2Node_Event}
   > n7 = Set play(play:bool=true) -> (Output_Get:bool)
   > Set play_f(play_f:double="To Float (Boolean)"(InBool=@n7)) -> (Output_Get:double)
  ]}
]
variables [
 backpack_cape_mod {type=number}
 backpack_cape_mod_bool {type=boolean}
 BodySkelMesh {type=SkeletalMeshComponentBudgeted}
 crouching_f {type=number}
 Gravity Override {type=Vector}
 jump_sprint {type=number}
 play {type=boolean}
 play_f {type=number}
 Sim Space Settings {
  type=SimSpaceSettings
  default {
   dampingAlpha=1
   maxAngularAcceleration=10000
   maxAngularVelocity=10000
   maxLinearAcceleration=10000
   maxLinearVelocity=10000
   velocityScaleZ=1
   worldAlpha=0
   externalLinearDragV=<0 0 0>
  }
 }
]
